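; BuildingToy is an overridable base for customizable Creative devices that react to minigame state and player-set options.
; On minigame start it saves its damageable state, applies the invulnerable-during-play setting to damage and health bar, then restores both at the end.
; Option changes refresh phase visibility through VisibleInGame, using visual components that children can supply.
; The "Customize" interaction requires creation rights, no vehicle, and no started minigame.
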(event OnAnyPropertyChanged(ToyOptionsComponent)
  ;[0 1 0.1336 0.5] Initialization - Update Collision and Visuals once custom data has loaded.
  (CallFunction|PropertiesUpdated))

(event OnPropertiesLoaded(ToyOptionsComponent)
  ;[0 1 0.1336 0.5] Initialization - Update Collision and Visuals once custom data has loaded.
  (CallFunction|PropertiesUpdated))

(event EventBeginPlay
  ;[1 1 1] Binding Delegates to Events, allowing children to control order of execution if necessary.
  (|CallParentFunction)
  (bind _fortminigameprogress (Variables|Default|GetFortMinigameProgress)
    (:then
      (Minigame|AssignMinigameStartedDelegate _fortminigameprogress (EventDispatchers|CreateEvent :!event Custom|OnMinigameStarted))
      (Minigame|BindEventtoMinigameEndedDelegate _fortminigameprogress (EventDispatchers|CreateEvent :!event Custom|OnMinigameEnded)))
    (:else
      _))
  (bind _visibleingame (Variables|Default|GetVisibleInGame)
    (:then
      (Visibility|SetVisibilityComponents _visibleingame (Visibility|GetVisualComponents) true)
      (Visibility|UpdateVisibilitySettings))
    (:else)))

(event Custom|OnMinigameStarted
  (bind _bisinvulnerableduringminigameplay (Variables|Default|GetbIsInvulnerableDuringMinigamePlay))
  ;[0 1 0.05918 0.5] Minigame Started - Update whether this can be damaged during the mini game based on whether it is invulnerable during minigame play.
  (Variables|Default|SetbPreviousInvulnerable (not (Variables|Actor|GetbCanBeDamaged)))
  (CallFunction|SetSuppressHealthBar _bisinvulnerableduringminigameplay)
  (Networking|SwitchHasAuthority
    (:Authority
      (Fortnite|SetCanbeDamaged self (not _bisinvulnerableduringminigameplay)))))

(event Custom|OnMinigameEnded
  (bind _bpreviousinvulnerable (Variables|Default|GetbPreviousInvulnerable))
  ;[1 0.1625 0 0.5] Minigame Ended - Update whether this can be damaged during creation time.
  (CallFunction|SetSuppressHealthBar _bpreviousinvulnerable)
  (Networking|SwitchHasAuthority
    (:Authority
      (Fortnite|SetCanbeDamaged self (not _bpreviousinvulnerable)))))

(event OnAddedtoMinigame(FortMinigameProgress)
  (params (Minigame :type Fort%20Minigame%20Object%20Reference))
  ;[1 1 1] For Overriding by children
  (Minigame|BindEventtoOnPlayModeChangedDelegate Minigame (delegate Custom|On%20Play%20Mode%20Changed)))

(event Custom|On%20Play%20Mode%20Changed
  (params (Minigame :type Fort%20Minigame%20Object%20Reference) (bIsInPlayMode :type Boolean)))

(fn BlueprintCanInteract
  (params (InteractingPawn :type Fort%20Pawn%20Object%20Reference) (InteractionBeingAttempted :type EInteractionBeingAttempted%20Enum :default "FirstInteraction") (InteractionType :type TInteractionType%20Enum :default "IT_NoInteraction"))
  (bind _returnvalue (Pawn|GetController InteractingPawn))
  (bind _fortminigameprogress (Variables|Default|GetFortMinigameProgress))
  (bind _currentminigame (Variables|/Script/FortniteGame.FortMinigameLogicComponent|GetCurrentMinigame _fortminigameprogress))
  (Utilities|IsinVehicle InteractingPawn
    (:true
      (LocalVariables|SetbInteractSuccess false)
      (return (LocalVariables|GetbInteractSuccess)))
    (:false
      (Utilities|FlowControl|Sequence
        (:then_0
          (Utilities|IsValid (Variables|/Script/FortniteGame.FortActorOptionsComponent|GetInternal_PlayerOptions (Variables|Default|GetToyOptionsComponent))
            (:Is%20Valid
              (Utilities|IsValid InteractingPawn
                (:Is%20Valid
                  (Utilities|IsValid _returnvalue
                    (:Is%20Valid
                      (bind _asfort_player_controller_athena (Utilities|Casting|CastToFortPlayerControllerAthena _returnvalue))
                      (if (Creative|CanPlayerCreateinCurrentExperience _asfort_player_controller_athena)
                        (Utilities|IsValid _currentminigame
                          (:Is%20Valid
                            (LocalVariables|SetbInteractSuccess (not (Minigame|HasStarted _currentminigame :!class /Script/FortniteGame.FortMinigame))))
                          (:Is%20Not%20Valid
                            (LocalVariables|SetbInteractSuccess true)))))))))))
        (:then_1
          (return (LocalVariables|GetbInteractSuccess)))))))

(fn BlueprintGetInteractionString
  (params (InteractingPawn :type Fort%20Pawn%20Object%20Reference) (InteractionBeingAttempted :type EInteractionBeingAttempted%20Enum :default "FirstInteraction"))
  (return "NSLOCTEXT(\"[59ACA76A45CF7E2C53F6F98B026CBB61]\", \"76E7770F4F29F94BC1A61CB1054A919E\", \"Customize\")"))

(fn GetVisualComponents
  (return 0))

(fn IsVisibleDuringMinigame
  (return (Variables|Default|GetbIsVisibleDuringMinigame)))

(fn MinigameIsRunning
  (bind _fortminigameprogress (Variables|Default|GetFortMinigameProgress))
  (bind _currentminigame (Variables|/Script/FortniteGame.FortMinigameLogicComponent|GetCurrentMinigame _fortminigameprogress))
  (Utilities|IsValid _currentminigame
    (:Is%20Valid
      (return (Minigame|HasStarted _currentminigame :!class /Script/FortniteGame.FortMinigame)))))

(fn PropertiesUpdated
  (Visibility|UpdateVisibilitySettings))

(fn UpdateVisibilitySettings
  (Visibility|SetVisiblityDuringPhase (Variables|Default|GetVisibleInGame) (select (Visibility|IsVisibleDuringMinigame) "NewEnumerator3" "NewEnumerator1")))

(fn ConstructionScript)
