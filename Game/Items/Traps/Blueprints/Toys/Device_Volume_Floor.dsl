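; Creative floor trap base for a configurable, team-filtered zone volume, with elimination-zone damage helpers for players, bots, creatures and vehicles.
; User options set zone scale, protected team and visibility, resizing the trigger and its material.
; Overlaps are classified into empty enter/exit hook events, presumably for child devices.
; Minigame state drives the enabled flag, whose OnRep updates trigger collision and visibility; vehicle delegates re-trigger the trap or clear damage effects for registered players.
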(event EventBeginPlay
  (|CallParentFunction)
  (Utilities|FlowControl|Sequence
    (:then_0
      (bind _framemeshvisibility (Variables|Default|GetFrameMeshVisibility)
        (:then
          (Visibility|SetVisibilityComponents _framemeshvisibility (Utilities|Array|MakeArray (Variables|Default|GetFrame)) true))
        (:else)))
    (:then_1
      (bind _zonemeshvisibility (Variables|Default|GetZoneMeshVisibility)
        (:then
          (Visibility|SetVisibilityComponents _zonemeshvisibility (Visibility|GetZoneVisualComponents) true)
          (CallFunction|UpdateZoneVisiblity))
        (:else)))))

(event MinigameStateChangedDelegate(FortMinigameProgress)
  (params (Minigame :type Fort%20Minigame%20Object%20Reference) (NewMinigameState :type EFortMinigameState%20Enum :default "PreGame"))
  (CallFunction|MinigameStateLogic NewMinigameState))

(event EventOnPlaced
  (bind _trigger (Variables|Default|GetTrigger))
  (Utilities|FlowControl|DoOnce false
    (:Completed
      (Trap|AddTriggerComponent _trigger)
      (CallFunction|InitializeDevice)
      (Collision|SetCollisionEnabled _trigger "NoCollision")
      (Audio|PlaySoundatLocation "/Game/Sounds/Creative/Gadgets/Traps/Trap_Elimination_Placed_Cue.Trap_Elimination_Placed_Cue" (Transformation|GetActorLocation)))))

(event Custom|Initialize%20Device
  ;[0 0.375 0.132] Initialization
  (CallFunction|UpdateVolumeSize)
  (CallFunction|InitializeTeam)
  (CallFunction|UpdateZoneVisiblity))

(event GameplayCue|GameplayCue.Abilities.Activation.Traps.ActivateTrap
  (params (EventType :type EGameplayCueEvent%20Enum :default "OnActive") (Parameters :type Gameplay%20Cue%20Parameters%20Structure%20%28by%20ref%29))
  (switch Utilities|FlowControl|Switch|SwitchonEGameplayCueEvent EventType
    (:Executed
      (Audio|PlaySoundatLocation (Variables|Default|GetTrapFireSound) (Transformation|GetActorLocation) "0, 0, 0" 0.5))))

(event Custom|Enable%20Event
  (params (Instigator :type Controller%20Object%20Reference))
  ;[0.2393 0 1 0.25] Gameplay Functions
  (CallFunction|SetEnabled true))

(event Custom|Disable%20Event
  (params (Instigator :type Controller%20Object%20Reference))
  ;[0.2393 0 1 0.25] Gameplay Functions
  (CallFunction|SetEnabled false))

(event OnComponentBeginOverlap(Trigger)
  (params (OverlappedComponent :type Primitive%20Component%20Object%20Reference) (OtherActor :type Actor%20Object%20Reference) (OtherComp :type Primitive%20Component%20Object%20Reference) (OtherBodyIndex :type Integer) (bFromSweep :type Boolean) (SweepResult :type Hit%20Result%20Structure%20%28by%20ref%29))
  (CallFunction|ComponentBeginOverlap OtherActor))

(event Custom|Pawn%20Entered%20Trigger%20Volume
  (params (Pawn :type Pawn%20Object%20Reference) (PlayerController :type Player%20Controller%20Object%20Reference)))

(event PawnNowHasController(UseRestriction)
  (params (FortPlayerPawn :type Fort%20Player%20Pawn%20Object%20Reference) (Player%20Controller :type Fort%20Player%20Controller%20Object%20Reference))
  (if (DEPRECATED|CanPlayerControllerInteract (Variables|Default|GetUseRestriction) Player%20Controller true false)
    (CallFunction|PawnEnteredTriggerVolume FortPlayerPawn Player%20Controller)))

(event Custom|Pawn%20Exited%20Trigger%20Volume
  (params (Pawn :type Pawn%20Object%20Reference) (PlayerController :type Player%20Controller%20Object%20Reference)))

(event Custom|Non-Pawn%20Entered%20Trigger%20Volume
  (params (Actor :type Actor%20Object%20Reference)))

(event Custom|Non-Pawn%20Exited%20Trigger%20Volume
  (params (Actor :type Actor%20Object%20Reference)))

(event OnComponentEndOverlap(Trigger)
  (params (OverlappedComponent :type Primitive%20Component%20Object%20Reference) (OtherActor :type Actor%20Object%20Reference) (OtherComp :type Primitive%20Component%20Object%20Reference) (OtherBodyIndex :type Integer))
  (bind (_is_player_bot _player_pawn _ai_bot_controller) (CallFunction|ActorIsAPlayerBot OtherActor))
  (bind (_true _pawn) (CallFunction|ActorIsACreature OtherActor))
  (bind (_valid_controller _fortplayerpawn _playercontroller) (DEPRECATED|ActorHasValidController (Variables|Default|GetUseRestriction) OtherActor true true))
  (if _valid_controller
    (CallFunction|PawnExitedTriggerVolume _fortplayerpawn _playercontroller)
    (elif _true
      (CallFunction|PawnExitedTriggerVolume _pawn)
      (elif _is_player_bot
        (CallFunction|PlayerBotExitedTriggerVolume _player_pawn _ai_bot_controller)
        (else
          (CallFunction|Non-PawnExitedTriggerVolume OtherActor))))))

(event Custom|Player%20Bot%20Entered%20Trigger%20Volume
  (params (Pawn :type Pawn%20Object%20Reference) (AI%20Bot%20Controller :type Fort%20Athena%20AIBot%20Controller%20Object%20Reference)))

(event Custom|Player%20Bot%20Exited%20Trigger%20Volume
  (params (Pawn :type Pawn%20Object%20Reference) (AI%20Bot%20Controller :type Fort%20Athena%20AIBot%20Controller%20Object%20Reference)))

(fn ActorIsACreature
  (params (Actor :type Actor%20Object%20Reference))
  ;[1 1 1] Does this pass a Pawn test but fail a playerpawn test?
  (Utilities|Casting|CastToFortPlayerPawn Actor
    (:then
      (return false 0))
    (:CastFailed
      (bind _asfort_pawn (Utilities|Casting|CastToFortPawn Actor))
      (return true _asfort_pawn))))

(fn ActorIsAPlayerBot
  (params (Actor :type Actor%20Object%20Reference))
  (bind _returnvalue (AI|GetAIController Actor))
  ;[1 1 1] Check if the actor has an AI controller and then determine if it is AI Bot Controller
  (Utilities|IsValid _returnvalue
    (:Is%20Valid
      (bind _asfort_athena_aibot_controller (Utilities|Casting|CastToFortAthenaAIBotController _returnvalue)
        (:then
          (return true (Variables|/Script/FortniteAI.FortAthenaAIBotController|GetPlayerBotPawn _asfort_athena_aibot_controller) _asfort_athena_aibot_controller))
        (:CastFailed
          _)))
    (:Is%20Not%20Valid
      _))
  (return false 0 0))

(fn ActorIsAVehicle
  (params (Actor :type Actor%20Object%20Reference))
  (Utilities|Casting|CastToFortVehicleInterface Actor
    (:then
      (return true))
    (:CastFailed
      (return false))))

(fn AddAffectedPlayerInVolume
  (params (AffectedPlayer :type Actor%20Object%20Reference%20%28by%20ref%29))
  (Utilities|IsValid AffectedPlayer
    (:Is%20Valid
      (bind _index (Default|RegisterPlayerinVolume AffectedPlayer))
      (if (>= _index 0)
        (Default|BindVehicleDelegates AffectedPlayer)))))

(fn AffectsTarget
  (params (Target%20Actor :type Actor%20Object%20Reference))
  (bind (_is_player_bot _player_pawn) (CallFunction|ActorIsAPlayerBot Target%20Actor))
  (if _is_player_bot
    (if (Variables|UserOptions|GetbDamagePlayerBots)
      (return (StateChecks|IsValidTeam (Variables|Default|GetUseRestriction) _player_pawn))
      (else
        (return false)))
    (else
      (bind _asfort_player_pawn (Utilities|Casting|CastToFortPlayerPawn Target%20Actor)
        (:then
          (if (Variables|UserOptions|GetbDamagePlayers)
            (return (DEPRECATED|CanActorInteract (Variables|Default|GetUseRestriction) _asfort_player_pawn true false))
            (else
              (return false))))
        (:CastFailed
          (bind _asfort_pawn (Utilities|Casting|CastToFortPawn Target%20Actor))
          (if (Variables|UserOptions|GetbDamageCreatures)
            (return (StateChecks|IsValidTeam (Variables|Default|GetUseRestriction) _asfort_pawn))
            (else
              (return false))))))))

(fn BindVehicleDelegates
  (params (Player :type Actor%20Object%20Reference))
  (bind _outputdelegate (EventDispatchers|CreateEvent :!event OnAffectedPlayerEnteredVehicle))
  (bind _output_get (LocalVariables|SetPlayerController (CallFunction|GetPlayerControllerfromActor Player)))
  (Utilities|IsValid _output_get
    (:Is%20Valid
      (Vehicle|BindEventtoOnExitVehicle (LocalVariables|GetPlayerController) (EventDispatchers|CreateEvent :!event OnAffectedPlayerExitedVehicle))
      (Vehicle|BindEventtoOnEnterVehicleDriver (LocalVariables|GetPlayerController) _outputdelegate)
      (Vehicle|BindEventtoOnEnterVehiclePassenger (LocalVariables|GetPlayerController) _outputdelegate)
      (Default|OnAffectedPlayerEnteredVehicle))))

(fn ComponentBeginOverlap
  (params (OtherActor :type Actor%20Object%20Reference))
  (bind (_is_player_bot _player_pawn _ai_bot_controller) (CallFunction|ActorIsAPlayerBot OtherActor))
  (bind (_true _pawn) (CallFunction|ActorIsACreature OtherActor))
  (bind (_valid_controller _fortplayerpawn _playercontroller) (DEPRECATED|ActorHasValidController (Variables|Default|GetUseRestriction) OtherActor true true))
  (if _valid_controller
    (CallFunction|PawnEnteredTriggerVolume _fortplayerpawn _playercontroller)
    (elif _true
      (CallFunction|PawnEnteredTriggerVolume _pawn)
      (elif _is_player_bot
        (CallFunction|PlayerBotEnteredTriggerVolume _player_pawn _ai_bot_controller)
        (else
          (CallFunction|Non-PawnEnteredTriggerVolume OtherActor))))))

(fn EmptyCueParameters
  (return (Ability|GameplayCue|MakeGameplayCueParameters 0.0)))

(fn GetInstigatorForBroadcastIfValidActor
  (params (Actor :type Actor%20Object%20Reference) (Controller :type Controller%20Object%20Reference))
  (Networking|SwitchHasAuthority
    (:Authority
      (bind _shouldaffect (CallFunction|AffectsTarget Actor))
      (if _shouldaffect
        (Utilities|IsValid Controller
          (:Is%20Valid
            (return Controller))
          (:Is%20Not%20Valid
            (bind _asfort_player_pawn_athena (Utilities|Casting|CastToFortPlayerPawnAthena Actor))
            (return (Pawn|GetController _asfort_player_pawn_athena))))))))

(fn GetPlayerControllerFromActor
  (params (Actor :type Actor%20Object%20Reference))
  (Utilities|FlowControl|Sequence
    (:then_0
      (bind _aspawn (Utilities|Casting|CastToPawn (Variables|GetActor)))
      (bind _asfort_player_controller (Utilities|Casting|CastToFortPlayerController (Pawn|GetController _aspawn))
        (:then
          (return _asfort_player_controller))
        (:CastFailed
          ;[0 0 0 1 10] Special case for vehicles, as the driver pawn is sometimes unpossesed and does not have a player state
          (bind _asfort_player_pawn (Utilities|Casting|CastToFortPlayerPawn (Variables|GetActor)))
          (bind _asfort_vehicle_interface (Utilities|Casting|CastToFortVehicleInterface (Vehicle|GetVehicleActor _asfort_player_pawn :!class /Script/FortniteGame.FortPlayerPawn)))
          (Utilities|IsValid (Class|FortVehicleInterface|GetVehicleController _asfort_vehicle_interface _asfort_player_pawn)
            (:Is%20Valid
              (return (Class|FortVehicleInterface|GetVehicleController _asfort_vehicle_interface _asfort_player_pawn)))))))
    (:then_1
      (return (Fortnite|GetFortPlayerControllerfromActor (Variables|GetActor))))))

(fn GetTriggerLocation
  (return (Math|Vector|MakeVector (.x (Variables|/Script/Engine.SceneComponent|GetRelativeLocation (Variables|Default|GetTrigger))) (.y (Variables|/Script/Engine.SceneComponent|GetRelativeLocation (Variables|Default|GetTrigger))) (+ (* (.z (Variables|/Script/Engine.BoxComponent|GetBoxExtent (Variables|Default|GetTrigger))) (Variables|UserOptions|GetZoneZScale)) (Math|Vector|DotProduct (Variables|/Script/Engine.SceneComponent|GetRelativeLocation (Variables|Default|GetFrame)) (Variables|Default|GetTriggerLocationDotProduct))))))

(fn GetVisualComponents
  (return (Utilities|Array|MakeArray (Variables|Debug|GetStaticMeshComponent) (Variables|Default|GetLightRays) (Variables|Default|GetPointLight))))

(fn GetVolumeMID
  (bind _asmaterial_instance_dynamic (Utilities|Casting|CastToMaterialInstanceDynamic (Rendering|Material|GetMaterial (Variables|Default|GetFrame) 0))
    (:then
      (return _asmaterial_instance_dynamic))
    (:CastFailed
      (bind _returnvalue (Rendering|Material|CreateDynamicMaterialInstance (Variables|Default|GetFrame) 0 "/Game/Abilities/Toys/EliminationVolume/M_EliminationVolume.M_EliminationVolume" "None" :!class /Script/Engine.PrimitiveComponent))
      (return _returnvalue))))

(fn GetZoneProtectedTeamValue
  (return (Math|Conversions|ToInteger(Byte) (Creative|TeamOption|ConverttoTeam (Variables|UserOptions|GetZoneProtectedTeam)))))

(fn GetZoneVisualComponents
  (return 0))

(fn InitializeTeam
  (CallFunction|SetTeam (Math|Conversions|ToByte(Integer) (CallFunction|GetZoneProtectedTeamValue)))
  (TeamRestrictions|SetTeamswithAffects (Variables|Default|GetUseRestriction) (Creative|TeamOption|GetLegacyTeamIndex (Variables|UserOptions|GetZoneProtectedTeam)) true))

(fn IsVisibleDuringMinigame
  (return (CallFunction|ShouldShowinGame (Variables|UserOptions|GetBaseMeshVisibleInGame))))

(fn IsZoneVisibleDuringMinigame
  (return (CallFunction|ShouldShowinGame (Variables|UserOptions|GetZoneVisibleInGame))))

(fn IsPlayerRegistered
  (params (PlayerToCheck :type Actor%20Object%20Reference%20%28by%20ref%29))
  (return (Utilities|Array|ContainsItem (Variables|Default|GetAffectedPlayersInVolume) PlayerToCheck)))

(fn MinigameStateLogic
  (params (State :type EFortMinigameState%20Enum :default "PreGame"))
  (Networking|SwitchHasAuthority
    (:Authority
      (bind _currentminigame (Variables|/Script/FortniteGame.FortMinigameLogicComponent|GetCurrentMinigame (Variables|Default|GetFortMinigameProgress))
        (:then
          (if (or (Utilities|Enum|Equal(Enum) (Variables|/Script/FortniteGame.FortMinigame|GetCurrentState _currentminigame) "Warmup") (Utilities|Enum|Equal(Enum) (Variables|/Script/FortniteGame.FortMinigame|GetCurrentState _currentminigame) "InProgress"))
            (CallFunction|SetEnabled (Variables|UserOptions|GetEnabledOnStart))
            (elif (Minigame|IsPregame _currentminigame)
              (CallFunction|SetEnabled false))))
        (:else)))))

(fn OnAffectedPlayerEnteredVehicle
  (Trap|TriggerifAppropriate))

(fn OnAffectedPlayerExitedVehicle
  ;[1 1 1] Clear the GE from vehicles if there are no more affected pawns in vehicles
  (bind _output_get (LocalVariables|SetOverlappingTargets (Trap|GetTouchingDamageTargets (Variables|Trap|GetTriggerFilter))))
  (for _array_element _output_get
    (bind _asfort_vehicle_interface (Utilities|Casting|CastToFortVehicleInterface _array_element))
    (if (and (== (Utilities|Array|Length (Vehicle|GetAllPassengers _asfort_vehicle_interface)) 0) (not (Variables|UserOptions|GetbDamageEmptyVehicles)))
      (Utilities|IsValid (Ability|GetAbilitySystemComponent _array_element)
        (:Is%20Valid
          (GameplayEffects|RemoveActiveEffectswithTags (Ability|GetAbilitySystemComponent _array_element) "(GameplayTags=((TagName=\"Gameplay.DamageSource.EliminationZone\")))")))
      (else
        ;[0 0 0 1 10] Needed for when a player goes from driver to gunner in the Mech. When this elimination zone thinks the player has left the volume, but they are still in it.
        (Trap|TriggerifAppropriate)))))

(fn OnRep_bEnabled
  (CallFunction|UpdateVolumeState)
  (CallFunction|UpdateZoneVisiblity)
  (Visibility|UpdateVisibilitySettings))

(fn OnRep_ZoneSizePresetIndex
  (CallFunction|UpdateVolumeSize))

(fn PropertiesUpdated
  (|CallParentFunction)
  (CallFunction|InitializeDevice))

(fn RegisterPlayerInVolume
  (params (PlayerToRegister :type Actor%20Object%20Reference%20%28by%20ref%29))
  (Utilities|IsValid PlayerToRegister
    (:Is%20Valid
      (bind _returnvalue (Utilities|Array|AddUnique (Variables|Default|GetAffectedPlayersInVolume) PlayerToRegister))
      (return _returnvalue))
    (:Is%20Not%20Valid
      (return -1))))

(fn RemoveAffectedPlayerInVolume
  (params (AffectedPlayer :type Actor%20Object%20Reference%20%28by%20ref%29))
  (bind _success (Default|UnregisterPlayerinVolume AffectedPlayer))
  (if _success
    (Default|UnbindVehicleDelegates (Variables|GetAffectedPlayer))))

(fn SetEnabled
  (params (New%20Enabled :type Boolean))
  ;[1 1 1] Switches enabled to new state if it's not already.
  ; This avoids calling the OnRep on the server if it's already in that state.
  (if (Math|Boolean|NotEqual(Boolean) New%20Enabled (Variables|Default|GetbEnabled))
    (Variables|Default|SetbEnabled (Variables|GetNew%20Enabled))))

(fn ShouldShowInGame
  (params (DeviceOptionVisualInGame :type Boolean))
  (return (Math|Boolean|ANDBoolean (Variables|Default|GetbEnabled) DeviceOptionVisualInGame (not (Replay|GetIsPlayingReplay)))))

(fn TransmitifValidActor
  (params (Transmitter :type Fort%20Gameplay%20Trigger%20Message%20Component%20Object%20Reference) (Actor :type Actor%20Object%20Reference) (Controller :type Controller%20Object%20Reference))
  (Networking|SwitchHasAuthority
    (:Authority
      (bind _shouldaffect (CallFunction|AffectsTarget Actor))
      (if _shouldaffect
        (bind _controller (Variables|GetController)
          (:then
            (Class|FortGameplayTriggerMessageComponent|SendTriggerMessage (Variables|GetTransmitter) _controller))
          (:else
            (bind _asfort_player_pawn_athena (Utilities|Casting|CastToFortPlayerPawnAthena (Variables|GetActor)))
            (Class|FortGameplayTriggerMessageComponent|SendTriggerMessage (Variables|GetTransmitter) (Pawn|GetController _asfort_player_pawn_athena))))
        (return true)))))

(fn UnbindVehicleDelegates
  (params (Player :type Actor%20Object%20Reference))
  (bind _outputdelegate (EventDispatchers|CreateEvent :!event OnAffectedPlayerEnteredVehicle))
  (bind _output_get (LocalVariables|SetPlayerController (CallFunction|GetPlayerControllerfromActor Player)))
  (Utilities|IsValid _output_get
    (:Is%20Valid
      (Vehicle|UnbindEventfromOnExitVehicle (LocalVariables|GetPlayerController) (EventDispatchers|CreateEvent :!event OnAffectedPlayerExitedVehicle))
      (Default|OnAffectedPlayerExitedVehicle)
      (Vehicle|UnbindEventfromOnEnterVehicleDriver (LocalVariables|GetPlayerController) _outputdelegate)
      (Vehicle|UnbindEventfromOnEnterVehiclePassenger (LocalVariables|GetPlayerController) _outputdelegate))))

(fn UnregisterPlayerInVolume
  (params (PlayerToUnregister :type Actor%20Object%20Reference%20%28by%20ref%29))
  (bind _returnvalue (Utilities|Array|RemoveItem (Variables|Default|GetAffectedPlayersInVolume) PlayerToUnregister))
  (return _returnvalue))

(fn UpdateVolumeState
  (CallFunction|UpdateTriggerCollision true))

(fn UpdateZoneVisiblity
  (Visibility|SetVisiblityDuringPhase (Variables|Default|GetZoneMeshVisibility) (select (Visibility|IsZoneVisibleDuringMinigame) "NewEnumerator3" "NewEnumerator1")))

(fn UpdateTriggerCollision
  (params (bCheckEnabled :type Boolean))
  (bind _currentminigame (Variables|/Script/FortniteGame.FortMinigameLogicComponent|GetCurrentMinigame (Variables|Default|GetFortMinigameProgress))
    (:then
      (Collision|SetCollisionEnabled (Variables|Default|GetTrigger) (select (and (Minigame|IsinProgress _currentminigame) (select bCheckEnabled (Variables|Default|GetbEnabled) true)) "QueryAndPhysics" "NoCollision")))
    (:else)))

(fn UpdateVolumeSize
  (bind _zonezscale (Variables|UserOptions|GetZoneZScale))
  (bind _zoneyscale (Variables|UserOptions|GetZoneYScale))
  (bind _zonexscale (Variables|UserOptions|GetZoneXScale))
  (bind _returnvalue (Math|Vector|MakeVector _zonexscale _zoneyscale _zonezscale))
  (Rendering|Material|SetVectorParameterValue (CallFunction|GetVolumeMID) "SizeBias" (Math|Conversions|ToLinearColor(Vector) (- (* (Math|Vector|MakeVector 256.0 256.0 384.0) _returnvalue) (Math|Vector|MakeVector 256.0 256.0 384.0))) :!class /Script/Engine.MaterialInstanceDynamic)
  (Rendering|SetBoundsScale (Variables|Default|GetFrame) (* (Math|Float|Max(Float) (Variables|UserOptions|GetZoneXScale) (Variables|UserOptions|GetZoneYScale) (Variables|UserOptions|GetZoneZScale)) 1.5))
  (Transformation|SetWorldScale3D (Variables|Default|GetTrigger) _returnvalue)
  (bind _location (CallFunction|GetTriggerLocation))
  (Transformation|SetRelativeLocation (Variables|Default|GetTrigger) _location false true))

(fn ConstructionScript
  (Networking|CanExecuteCosmeticEvents
    (:True
      (bind _returnvalue (Rendering|Material|CreateDynamicMaterialInstance (Variables|Default|GetFrame) 0 0 "None" :!class /Script/Engine.PrimitiveComponent))
      (Variables|Default|SetVolumeMID _returnvalue))))

(fn VFXParameters
  (bind (_returnvalue_x _returnvalue_y _returnvalue_z) (Math|Vector|BreakVector (* (Math|Vector|MakeVector 512.0 512.0 384.0) (Math|Vector|MakeVector (Variables|UserOptions|GetZoneXScale) (Variables|UserOptions|GetZoneYScale) (Variables|UserOptions|GetZoneZScale)))))
  (bind (_zoneedgessize_x _zoneedgessize_y _zoneedgessize_z) (Math|Vector|BreakVector (Variables|GetZoneEdgesSize_X)))
  (return (Utilities|Array|MakeArray (Utilities|Struct|MakeParticleSysParam "BlocksVolume" "PSPT_VectorRand" 0.0 0.0 (Variables|Default|GetBlocksVolumeHigh) (Variables|Default|GetBlocksVolumeLow)) (Utilities|Struct|MakeParticleSysParam "BlocksSpawnRateScale" "PSPT_Scalar" (Utilities|ClampMaxFloat (/ (Math|Float|float*float _returnvalue_x _returnvalue_y _returnvalue_z) 100663296.0) 1e+01) 0.0 "0.000000,0.000000,0.000000" (0.0 0.0 0.0) "(B=0,G=0,R=0,A=0)" 0 0) (Utilities|Struct|MakeParticleSysParam "MeshSize" "PSPT_Vector" 0.0 0.0 (Math|Vector|MakeVector _zoneedgessize_x _zoneedgessize_y (/ _zoneedgessize_z (Variables|UserOptions|GetZoneZScale))) "0.000000,0.000000,0.000000" "(B=0,G=0,R=0,A=0)" 0 0) (Utilities|Struct|MakeParticleSysParam "BlocksSize" "PSPT_Vector" 0.0 0.0 (Math|Vector|MakeVector (Utilities|Reciprocal (Variables|UserOptions|GetZoneXScale)) (Utilities|Reciprocal (Variables|UserOptions|GetZoneXScale))) "0.000000,0.000000,0.000000" "(B=0,G=0,R=0,A=0)" 0 0) (Utilities|Struct|MakeParticleSysParam "BlocksVelocity" "PSPT_Vector" 0.0 0.0 (Math|Vector|MakeVector 0.0 0.0 (* -2e+01 (Utilities|Reciprocal (Variables|UserOptions|GetZoneZScale)))) "0.000000,0.000000,0.000000" "(B=0,G=0,R=0,A=0)" 0 0))))

(fn WasActivatedByOverlappingActor
  (LocalVariables|SetOverlappingActors (Collision|GetOverlappingActors (Variables|Default|GetTrigger) "/Script/Engine.Pawn" :!class /Script/Engine.PrimitiveComponent))
  (for _array_element (LocalVariables|GetOverlappingActors)
    (bind _asfort_vehicle_interface (Utilities|Casting|CastToFortVehicleInterface _array_element)
      (:then
        (bind _output_get (LocalVariables|SetPassengers (Vehicle|GetAllPassengers _asfort_vehicle_interface)))
        (if (> (Utilities|Array|Length _output_get) 0)
          (if (Variables|UserOptions|GetbDamageVehicles)
            (for _array_element_2 (LocalVariables|GetPassengers)
              (bind _asfort_player_pawn (Utilities|Casting|CastToFortPlayerPawn _array_element_2))
              (if (Math|Byte|NotEqual(Byte) (Team|GetActorTeam _asfort_player_pawn) (Math|Conversions|ToByte(Integer) (CallFunction|GetZoneProtectedTeamValue)))
                (return true))))
          (else
            (if (and (Variables|UserOptions|GetbDamageEmptyVehicles) (Variables|UserOptions|GetbDamageVehicles))
              (return true)))))
      (:CastFailed
        (bind _asfort_pawn (Utilities|Casting|CastToFortPawn _array_element)
          (:then
            (if (Math|Byte|NotEqual(Byte) (Team|GetActorTeam _asfort_pawn) (Math|Conversions|ToByte(Integer) (CallFunction|GetZoneProtectedTeamValue)))
              (Utilities|Casting|CastToFortPlayerPawn _asfort_pawn
                (:then
                  (if (Variables|UserOptions|GetbDamagePlayers)
                    (return true)))
                (:CastFailed
                  (if (Variables|UserOptions|GetbDamageCreatures)
                    (return true))))))
          (:CastFailed
            (Utilities|Casting|CastToBuildingRift _array_element)
            (if (Variables|UserOptions|GetbDamageCreatures)
              (return true)))))))
  (return false))
