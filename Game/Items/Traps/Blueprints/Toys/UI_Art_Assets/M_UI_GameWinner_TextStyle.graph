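# Translucent UI material that, judging by its name, styles game-winner text with a tinted, angled highlight band.
# Two UV gradients rotated half a turn apart by Rotate and offset by Pan are multiplied through a lerp, then sharpened by Tightness into a band.
# The clamped band is added to Color to whiten it, with a separate Alpha parameter.
# No Time node drives it, so band movement depends on changing parameters.
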Material /Game/Items/Traps/Blueprints/Toys/UI_Art_Assets/M_UI_GameWinner_TextStyle
graph {
 properties {blendMode=BLEND_Translucent, materialDomain=MD_UI, pixelDepthOffsetMode=PDOM_Legacy, usedShadingModels=Default Lit}
 nodes [
  n1 = TextureCoordinate()
  n2 = Multiply(A=ScalarParameter() {defaultValue=-0.075, parameterName=Rotate}) {constB=0.125}
  n3 = Rotator(Coordinate=@n1, Time=@n2) {speed=6.2832}
  n4 = ScalarParameter() {defaultValue=1.25, parameterName=Pan}
  n5 = Power(Base=ComponentMask(None=Add(A=@n3, B=@n4)) {r=true}) {constExponent=4}
  n6 = Rotator(Coordinate=@n1, Time=Add(A=@n2, B=Constant() {r=0.5})) {speed=6.2832}
  n7 = OneMinus(None=@n4)
  n8 = Power(Base=ComponentMask(None=Add(A=@n6, B=@n7)) {r=true}) {constExponent=4}
  n9 = ScalarParameter() {defaultValue=2, parameterName=Tightness}
  n10 = Clamp(None=Power(Base=LinearInterpolate(B=@n5, Alpha=@n8), Exp=Multiply(A=@n9) {constB=16}) {constExponent=16})
  Clamp(None=Add(A=Clamp(None=@n10), B=VectorParameter() {parameterName=Color, defaultValue=<rgba 0 1 1 1>}.RGB))
  ScalarParameter() {defaultValue=1, parameterName=Alpha}
 ]
}
