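# Sound cue for a negative BlackMonday stinger in the Creative radio gadget sound set.
# A single wave player plays the Stinger_BlackMonday_Negative_01 wave directly, with no randomization, mixing or modulation.
# It uses the SFX sound class and the shared Creative scoring attenuation asset.
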SoundCue /Game/Sounds/Creative/Gadgets/Radio/Stingers/BlackMonday/Stinger_BlackMonday_Negative_01_Cue
root=WavePlayer_0
nodes [{id=WavePlayer_0, properties {soundWaveAssetPtr=&/Stinger_BlackMonday_Negative_01}}]
properties {
 attenuationSettings=/Game/Sounds/Creative/Attenuations/Fort_Creative_Scoring_Att
 maxDistance=2097152
 soundClassObject=&SFX
}
