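# Sound cue for Halloween bat sounds, filed under the Creative speaker gadget sounds.
# A Random node feeds the output, picking among eight equally weighted bat wave players.
# It uses the SFX sound class and shared tile-ambient attenuation, with volume slightly reduced.
# Its concurrency override settings specify a single instance and prevent new ones rather than cutting off the current sound.
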SoundCue /Game/Sounds/Creative/Gadgets/Speaker/Halloween/Halloween_Bats_Cue
root=Random_0
nodes [
 {
  id=Random_0
  children [WavePlayer_0, WavePlayer_1, WavePlayer_2, WavePlayer_3, WavePlayer_4, WavePlayer_9, WavePlayer_10, WavePlayer_11]
  properties {weights [1,1,1,1,1,1,1,1]}
 }
 {id=WavePlayer_0, properties {soundWaveAssetPtr=&/Halloween_Bats_01}}
 {id=WavePlayer_1, properties {soundWaveAssetPtr=&/Halloween_Bats_02}}
 {id=WavePlayer_10, properties {soundWaveAssetPtr=&/Halloween_Bats_07}}
 {id=WavePlayer_11, properties {soundWaveAssetPtr=&/Halloween_Bats_08}}
 {id=WavePlayer_2, properties {soundWaveAssetPtr=&/Halloween_Bats_03}}
 {id=WavePlayer_3, properties {soundWaveAssetPtr=&/Halloween_Bats_04}}
 {id=WavePlayer_4, properties {soundWaveAssetPtr=&/Halloween_Bats_05}}
 {id=WavePlayer_9, properties {soundWaveAssetPtr=&/Halloween_Bats_06}}
]
properties {
 attenuationSettings=/Game/Sounds/Attenuation/Fort_TileAmbient_Att
 maxDistance=10000
 soundClassObject=&SFX
 volumeMultiplier=0.9
 concurrencyOverrides {maxCount=1, resolutionRule=PreventNew}
}
