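# Sound cue for the Halloween ghost sounds of the Creative speaker gadget.
# A Random node picks one of five equally weighted stereo monster-cry waves drawn from the shared creature ambience library.
# It plays in the SFX sound class with tile-ambient attenuation.
# Its concurrency override allows one instance at a time and prevents new ones from starting while it plays.
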SoundCue /Game/Sounds/Creative/Gadgets/Speaker/Halloween/Halloween_Ghosts_Cue
root=Random_0
nodes [
 {id=Random_0, children [WavePlayer_0,WavePlayer_1,WavePlayer_2,WavePlayer_3,WavePlayer_4], properties {weights [1,1,1,1,1]}}
 {id=WavePlayer_0, properties {soundWaveAssetPtr=/Game/Sounds/Creative/Ambiences/AnimalsAndCreatures/MonsterCryD01_Stereo}}
 {id=WavePlayer_1, properties {soundWaveAssetPtr=/Game/Sounds/Creative/Ambiences/AnimalsAndCreatures/MonsterCryD02_Stereo}}
 {id=WavePlayer_2, properties {soundWaveAssetPtr=/Game/Sounds/Creative/Ambiences/AnimalsAndCreatures/MonsterCryD03_Stereo}}
 {id=WavePlayer_3, properties {soundWaveAssetPtr=/Game/Sounds/Creative/Ambiences/AnimalsAndCreatures/MonsterCryB05_Stereo}}
 {id=WavePlayer_4, properties {soundWaveAssetPtr=/Game/Sounds/Creative/Ambiences/AnimalsAndCreatures/MonsterCryB02_Stereo}}
]
properties {
 attenuationSettings=/Game/Sounds/Attenuation/Fort_TileAmbient_Att
 maxDistance=10000
 soundClassObject=&SFX
 volumeMultiplier=1
 concurrencyOverrides {maxCount=1, resolutionRule=PreventNew}
}
