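# Sound cue for the Halloween female laughing sound in the Creative speaker gadget set.
# A Random node picks one of six equally weighted girl-laugh wave players, drawn from the Creative AbstractVoices ambience recordings.
# It plays on the SFX class with shared tile-ambient attenuation, and its concurrency override allows one instance at a time, stopping the oldest.
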SoundCue /Game/Sounds/Creative/Gadgets/Speaker/Halloween/Halloween_Laughing_Female_Cue
root=Random_0
nodes [
 {
  id=Random_0
  children [WavePlayer_0, WavePlayer_1, WavePlayer_4, WavePlayer_9, WavePlayer_10, WavePlayer_11]
  properties {weights [1,1,1,1,1,1]}
 }
 {id=WavePlayer_0, properties {soundWaveAssetPtr=/Game/Sounds/Creative/Ambiences/AbstractVoices/Voice_GirlLaughA001}}
 {id=WavePlayer_1, properties {soundWaveAssetPtr=/Game/Sounds/Creative/Ambiences/AbstractVoices/Voice_GirlLaughA002}}
 {id=WavePlayer_10, properties {soundWaveAssetPtr=/Game/Sounds/Creative/Ambiences/AbstractVoices/Voice_GirlLaughA007}}
 {id=WavePlayer_11, properties {soundWaveAssetPtr=/Game/Sounds/Creative/Ambiences/AbstractVoices/Voice_GirlLaughA008}}
 {id=WavePlayer_4, properties {soundWaveAssetPtr=/Game/Sounds/Creative/Ambiences/AbstractVoices/Voice_GirlLaughA005}}
 {id=WavePlayer_9, properties {soundWaveAssetPtr=/Game/Sounds/Creative/Ambiences/AbstractVoices/Voice_GirlLaughA006}}
]
properties {
 attenuationSettings=/Game/Sounds/Attenuation/Fort_TileAmbient_Att
 maxDistance=10000
 soundClassObject=&SFX
 volumeMultiplier=1
 concurrencyOverrides {maxCount=1, resolutionRule=StopOldest}
}
