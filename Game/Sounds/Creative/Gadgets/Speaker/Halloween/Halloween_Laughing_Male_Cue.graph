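# Sound cue for a male Halloween laugh, filed under the Creative speaker gadget sounds.
# A Random node picks one of nine laugh wave variations with equal weights.
# It plays in the SFX class with a shared tile-ambient attenuation asset, and its concurrency override holds a single instance, preventing new ones while one plays.
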SoundCue /Game/Sounds/Creative/Gadgets/Speaker/Halloween/Halloween_Laughing_Male_Cue
root=Random_1
nodes [
 {
  id=Random_1
  children [WavePlayer_5, WavePlayer_6, WavePlayer_7, WavePlayer_8, WavePlayer_12, WavePlayer_13, WavePlayer_14, WavePlayer_15, WavePlayer_16]
  properties {weights [1,1,1,1,1,1,1,1,1]}
 }
 {id=WavePlayer_12, properties {soundWaveAssetPtr=&/Halloween_Laughing_Male_05}}
 {id=WavePlayer_13, properties {soundWaveAssetPtr=&/Halloween_Laughing_Male_06}}
 {id=WavePlayer_14, properties {soundWaveAssetPtr=&/Halloween_Laughing_Male_07}}
 {id=WavePlayer_15, properties {soundWaveAssetPtr=&/Halloween_Laughing_Male_08}}
 {id=WavePlayer_16, properties {soundWaveAssetPtr=&/Halloween_Laughing_Male_09}}
 {id=WavePlayer_5, properties {soundWaveAssetPtr=&/Halloween_Laughing_Male_01}}
 {id=WavePlayer_6, properties {soundWaveAssetPtr=&/Halloween_Laughing_Male_02}}
 {id=WavePlayer_7, properties {soundWaveAssetPtr=&/Halloween_Laughing_Male_03}}
 {id=WavePlayer_8, properties {soundWaveAssetPtr=&/Halloween_Laughing_Male_04}}
]
properties {
 attenuationSettings=/Game/Sounds/Attenuation/Fort_TileAmbient_Att
 maxDistance=10000
 soundClassObject=&SFX
 volumeMultiplier=1
 concurrencyOverrides {maxCount=1, resolutionRule=PreventNew}
}
