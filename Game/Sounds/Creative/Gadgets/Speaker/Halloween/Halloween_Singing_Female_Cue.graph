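# Sound cue of female singing for the Halloween set of the Creative speaker gadget.
# A Random node picks one of four equally weighted stereo girl-singing clips from the abstract-voice ambience set.
# It uses tile-ambient attenuation and the SFX sound class, with concurrency override values that specify a single instance and block new ones.
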SoundCue /Game/Sounds/Creative/Gadgets/Speaker/Halloween/Halloween_Singing_Female_Cue
root=Random_1
nodes [
 {id=Random_1, children [WavePlayer_5,WavePlayer_6,WavePlayer_7,WavePlayer_8], properties {weights [1,1,1,1]}}
 {id=WavePlayer_5, properties {soundWaveAssetPtr=/Game/Sounds/Creative/Ambiences/AbstractVoices/Voice_GirlSingsB001_Stereo}}
 {id=WavePlayer_6, properties {soundWaveAssetPtr=/Game/Sounds/Creative/Ambiences/AbstractVoices/Voice_GirlSingsB002_Stereo}}
 {id=WavePlayer_7, properties {soundWaveAssetPtr=/Game/Sounds/Creative/Ambiences/AbstractVoices/Voice_GirlSingsB003_Stereo}}
 {id=WavePlayer_8, properties {soundWaveAssetPtr=/Game/Sounds/Creative/Ambiences/AbstractVoices/Voice_GirlSingsB004_Stereo}}
]
properties {
 attenuationSettings=/Game/Sounds/Attenuation/Fort_TileAmbient_Att
 maxDistance=10000
 soundClassObject=&SFX
 volumeMultiplier=0.9
 concurrencyOverrides {maxCount=1, resolutionRule=PreventNew}
}
