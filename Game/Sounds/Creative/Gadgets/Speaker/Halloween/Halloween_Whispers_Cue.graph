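# Halloween whispers sound cue for the Creative speaker gadget, playing distant howling abstract-voice ambience.
# A Random node picks one of three stereo howl waves with equal weight.
# It routes to the SFX sound class with tile-ambient attenuation and a slightly boosted volume.
# Its concurrency override allows only one instance at a time and blocks new plays while one is active.
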SoundCue /Game/Sounds/Creative/Gadgets/Speaker/Halloween/Halloween_Whispers_Cue
root=Random_2
nodes [
 {id=Random_2, children [WavePlayer_10, WavePlayer_11, WavePlayer_12], properties {weights [1,1,1]}}
 {id=WavePlayer_10, properties {soundWaveAssetPtr=/Game/Sounds/Creative/Ambiences/AbstractVoices/Voice_HowlDistantB001_Stereo}}
 {id=WavePlayer_11, properties {soundWaveAssetPtr=/Game/Sounds/Creative/Ambiences/AbstractVoices/Voice_HowlDistantB004_Stereo}}
 {id=WavePlayer_12, properties {soundWaveAssetPtr=/Game/Sounds/Creative/Ambiences/AbstractVoices/Voice_HowlDistantB006_Stereo}}
]
properties {
 attenuationSettings=/Game/Sounds/Attenuation/Fort_TileAmbient_Att
 maxDistance=10000
 soundClassObject=&SFX
 volumeMultiplier=1.2
 concurrencyOverrides {maxCount=1, resolutionRule=PreventNew}
}
