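# UI_VictoryScore_Cue is a Sound Cue that plays the victory score stinger for the Creative in-game UI.
# A Mixer root feeds from a single Wave Player, with its second input left empty.
# It uses the Victory Royale UI sound class with a raised volume.
# A concurrency override allows only one instance at a time and blocks new plays while one is active.
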SoundCue /Game/Sounds/Creative/IngameUI/UI_VictoryScore_Cue
root=Mixer_0
nodes [
 {id=Mixer_0, children [WavePlayer_1,null], properties {inputVolume [1,1]}}
 {id=WavePlayer_1, properties {soundWaveAssetPtr=&/UI_Victory_Score_Stinger_01}}
]
properties {
 bOverrideConcurrency=true
 maxDistance=2097152
 soundClassObject=/Game/Sounds/FortSoundClassesAndModes/Fort_UI_VictoryRoyale_PSM
 volumeMultiplier=1.25
 concurrencyOverrides {maxCount=1, resolutionRule=PreventNew}
}
