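# Sound cue that plays the UI stinger for a draw result on the Creative in-game victory score.
# A single wave player feeds the draw score stinger wave straight to the output at boosted volume.
# It routes through the Victory Royale UI sound class and overrides concurrency so only one instance plays, blocking new triggers while one is active.
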SoundCue /Game/Sounds/Creative/IngameUI/UI_VictoryScore_Draw_Cue
root=WavePlayer_0
nodes [{id=WavePlayer_0, properties {soundWaveAssetPtr=&/UI_Draw_Score_Stinger_01}}]
properties {
 bOverrideConcurrency=true
 maxDistance=2097152
 soundClassObject=/Game/Sounds/FortSoundClassesAndModes/Fort_UI_VictoryRoyale_PSM
 volumeMultiplier=1.25
 concurrencyOverrides {maxCount=1, resolutionRule=PreventNew}
}
