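# Widget Blueprint for a round-winner victory screen that presents team or player, placement and win-description text over a logo with flare, glow and triangle accents.
# Intro and Intro_3 animate the elements in with staggered scale pops and fade them out, while Intro_NoFade holds the main text and logo for the separate FadeOut animation.
# Later variants also animate material parameters for panning text highlights, outline color flashes and flare intensity.
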WidgetBlueprint /Game/Items/Traps/Blueprints/Toys/RoundsWinner_VictoryScreen
animations [
 Intro {
  binding_count=10
  length_seconds=6
  object=&self:Intro
  sequence {bindings [
    WinDescription {
     id=#1
     possessed_class=CommonTextBlock
     tracks [
      {
       type=MovieScene2DTransformTrack
       sections [{
         type=MovieScene2DTransformSection
         end_frame=40
         start_frame=0
         channels [
          Translation.X {type=MovieSceneScriptingFloatChannel}
          Translation.Y {type=MovieSceneScriptingFloatChannel}
          Angle {type=MovieSceneScriptingFloatChannel}
          Scale.X {type=MovieSceneScriptingFloatChannel, keys [[7.9993,4],[10.9993,0.9],[11.9993,1.125],[13.9993,1]]}
          Scale.Y {type=MovieSceneScriptingFloatChannel}
          Shear.X {type=MovieSceneScriptingFloatChannel}
          Shear.Y {type=MovieSceneScriptingFloatChannel}
         ]
        }]
      }
      {
       type=MovieSceneFloatTrack
       display_name=Render Opacity
       sections [{
         type=MovieSceneFloatSection
         end_frame=100
         start_frame=0
         channels [None {type=MovieSceneScriptingFloatChannel, keys [[7.9993,0],[10.9993,1],[60.9993,1],[80.9997,0]]}]
        }]
      }
     ]
    }
    Logo {
     id=#2
     possessed_class=Image
     tracks [
      {
       type=MovieScene2DTransformTrack
       sections [{
         type=MovieScene2DTransformSection
         end_frame=24
         start_frame=0
         channels [
          Translation.X {type=MovieSceneScriptingFloatChannel}
          Translation.Y {type=MovieSceneScriptingFloatChannel}
          Angle {type=MovieSceneScriptingFloatChannel, keys [[0,16],[6.9993,-16]]}
          Scale.X {type=MovieSceneScriptingFloatChannel, keys [[0,3],[2.9993,0.75],[5,1.125],[8,1]]}
          Scale.Y {type=MovieSceneScriptingFloatChannel, keys [[0,3],[2.9993,0.75],[5,1.125],[8,1]]}
          Shear.X {type=MovieSceneScriptingFloatChannel}
          Shear.Y {type=MovieSceneScriptingFloatChannel}
         ]
        }]
      }
      {
       type=MovieSceneFloatTrack
       display_name=Render Opacity
       sections [{
         type=MovieSceneFloatSection
         end_frame=110
         start_frame=0
         channels [None {type=MovieSceneScriptingFloatChannel, keys [[0,0],[2,0.75],[60,0.75],[80,0]]}]
        }]
      }
     ]
    }
    SmallTriangles {
     id=#3
     possessed_class=SizeBox
     tracks [
      {
       type=MovieScene2DTransformTrack
       sections [{
         type=MovieScene2DTransformSection
         end_frame=24
         start_frame=-1
         channels [
          Translation.X {type=MovieSceneScriptingFloatChannel}
          Translation.Y {type=MovieSceneScriptingFloatChannel}
          Angle {type=MovieSceneScriptingFloatChannel}
          Scale.X {type=MovieSceneScriptingFloatChannel, keys [[0.9993,0.25],[2.9993,0.25],[10.9993,1]]}
          Scale.Y {type=MovieSceneScriptingFloatChannel}
          Shear.X {type=MovieSceneScriptingFloatChannel}
          Shear.Y {type=MovieSceneScriptingFloatChannel}
         ]
        }]
      }
      {
       type=MovieSceneFloatTrack
       display_name=Render Opacity
       sections [{
         type=MovieSceneFloatSection
         end_frame=70
         start_frame=-1
         channels [None {type=MovieSceneScriptingFloatChannel, keys [[0.9993,0],[2.9993,0],[10.9993,1],[40.9993,1],[70.9993,0]]}]
        }]
      }
     ]
    }
    Flare {
     id=#4
     possessed_class=Image
     tracks [
      {
       type=MovieSceneWidgetMaterialTrack
       sections [{
         type=MovieSceneParameterSection
         end_frame=80
         start_frame=0
         channels [
          Hardness {type=MovieSceneScriptingFloatChannel, keys [[2,3],[17,1.5],[35,1.5],[45,0]]}
          Intensity {type=MovieSceneScriptingFloatChannel, keys [[2,24],[17,4],[35,4],[45,2]]}
         ]
        }]
      }
      {
       type=MovieSceneFloatTrack
       display_name=Render Opacity
       sections [{
         type=MovieSceneFloatSection
         end_frame=80
         start_frame=0
         channels [None {type=MovieSceneScriptingFloatChannel, keys [[2,0],[2.9993,1],[30,1],[45,0]]}]
        }]
      }
      {
       type=MovieScene2DTransformTrack
       sections [{
         type=MovieScene2DTransformSection
         end_frame=80
         start_frame=0
         channels [
          Translation.X {type=MovieSceneScriptingFloatChannel}
          Translation.Y {type=MovieSceneScriptingFloatChannel}
          Angle {type=MovieSceneScriptingFloatChannel, const [0,1]}
          Scale.X {type=MovieSceneScriptingFloatChannel, keys [[2,1.125],[11,2]]}
          Scale.Y {type=MovieSceneScriptingFloatChannel, keys [[2,0.125],[5.0007,0.25],[8,1]]}
          Shear.X {type=MovieSceneScriptingFloatChannel}
          Shear.Y {type=MovieSceneScriptingFloatChannel}
         ]
        }]
      }
     ]
    }
    TeamOrPlayer {
     id=#5
     possessed_class=CommonTextBlock
     tracks [
      {
       type=MovieScene2DTransformTrack
       sections [{
         type=MovieScene2DTransformSection
         end_frame=59
         start_frame=0
         channels [
          Translation.X {type=MovieSceneScriptingFloatChannel}
          Translation.Y {type=MovieSceneScriptingFloatChannel}
          Angle {type=MovieSceneScriptingFloatChannel}
          Scale.X {type=MovieSceneScriptingFloatChannel, keys [[2.9997,0],[5,1.25],[7,0.95],[10,1]]}
          Scale.Y {type=MovieSceneScriptingFloatChannel, keys [[2.9997,0],[5,1.25],[7,0.95],[10,1]]}
          Shear.X {type=MovieSceneScriptingFloatChannel}
          Shear.Y {type=MovieSceneScriptingFloatChannel}
         ]
        }]
      }
      {
       type=MovieSceneFloatTrack
       display_name=Render Opacity
       sections [{
         type=MovieSceneFloatSection
         end_frame=120
         start_frame=0
         channels [None {type=MovieSceneScriptingFloatChannel, keys [[2.9997,0],[5,1],[64.9993,1],[85.9993,0]]}]
        }]
      }
     ]
    }
    BigWinTextFaded {
     id=#6
     possessed_class=CommonTextBlock
     tracks [
      {
       type=MovieScene2DTransformTrack
       sections [{
         type=MovieScene2DTransformSection
         end_frame=60
         start_frame=0
         channels [
          Translation.X {type=MovieSceneScriptingFloatChannel}
          Translation.Y {type=MovieSceneScriptingFloatChannel}
          Angle {type=MovieSceneScriptingFloatChannel}
          Scale.X {type=MovieSceneScriptingFloatChannel, keys [[2.9993,1.25],[60.9993,1]]}
          Scale.Y {type=MovieSceneScriptingFloatChannel, keys [[2.9993,1.25],[60.9993,1]]}
          Shear.X {type=MovieSceneScriptingFloatChannel}
          Shear.Y {type=MovieSceneScriptingFloatChannel}
         ]
        }]
      }
      {
       type=MovieSceneFloatTrack
       display_name=Render Opacity
       sections [{
         type=MovieSceneFloatSection
         end_frame=80
         start_frame=0
         channels [None {type=MovieSceneScriptingFloatChannel, keys [[2.9993,0],[30.9993,0.25],[50.9993,0.25],[80.9993,0]]}]
        }]
      }
     ]
    }
    Placement {
     id=#7
     possessed_class=CommonTextBlock
     tracks [
      {
       type=MovieSceneFloatTrack
       display_name=Render Opacity
       sections [{
         type=MovieSceneFloatSection
         end_frame=100
         start_frame=0
         channels [None {type=MovieSceneScriptingFloatChannel, keys [[7.9993,0],[10.9993,1],[60.9993,1],[80.9997,0]]}]
        }]
      }
      {
       type=MovieScene2DTransformTrack
       sections [{
         type=MovieScene2DTransformSection
         end_frame=40
         start_frame=0
         channels [
          Translation.X {type=MovieSceneScriptingFloatChannel}
          Translation.Y {type=MovieSceneScriptingFloatChannel}
          Angle {type=MovieSceneScriptingFloatChannel}
          Scale.X {type=MovieSceneScriptingFloatChannel, keys [[7.9993,4],[10.9993,0.9],[11.9993,1.125],[13.9993,1]]}
          Scale.Y {type=MovieSceneScriptingFloatChannel}
          Shear.X {type=MovieSceneScriptingFloatChannel}
          Shear.Y {type=MovieSceneScriptingFloatChannel}
         ]
        }]
      }
     ]
    }
    LargeTriangles {
     id=#8
     possessed_class=SizeBox
     tracks [{
       type=MovieSceneFloatTrack
       display_name=Render Opacity
       sections [{
         type=MovieSceneFloatSection
         end_frame=80
         start_frame=-1
         channels [None {type=MovieSceneScriptingFloatChannel, keys [[0.9993,0],[15.9993,0.1],[40.9993,0.1],[80.9993,0]]}]
        }]
      }]
    }
    SmallGlow {
     id=#9
     possessed_class=Image
     tracks [{
       type=MovieSceneFloatTrack
       display_name=Render Opacity
       sections [{
         type=MovieSceneFloatSection
         end_frame=80
         start_frame=0
         channels [None {type=MovieSceneScriptingFloatChannel, keys [[3.9993,0],[5.9993,1],[14,0.666],[70,0]]}]
        }]
      }]
    }
    Glow {
     id=#10
     possessed_class=Image
     tracks [{
       type=MovieSceneFloatTrack
       display_name=Render Opacity
       sections [{
         type=MovieSceneFloatSection
         end_frame=80
         start_frame=0
         channels [None {type=MovieSceneScriptingFloatChannel, keys [[3.9993,0],[5.9993,1],[19,0.333],[70,0]]}]
        }]
      }]
    }
   ]}
 }
 Intro_3 {
  binding_count=13
  length_seconds=4.0001
  object=&self:Intro_3
  sequence {
   bindings [
    WinDescription {
     id=#1
     possessed_class=CommonTextBlock
     tracks [
      {
       type=MovieScene2DTransformTrack
       sections [{
         type=MovieScene2DTransformSection
         end_frame=240
         start_frame=0
         channels [
          Translation.X {type=MovieSceneScriptingFloatChannel}
          Translation.Y {type=MovieSceneScriptingFloatChannel}
          Angle {type=MovieSceneScriptingFloatChannel}
          Scale.X {type=MovieSceneScriptingFloatChannel, keys [[55,6],[64,0.9],[67,1.25],[73,1]]}
          Scale.Y {type=MovieSceneScriptingFloatChannel, keys [[64,1],[67,1.25],[73,1]]}
          Shear.X {type=MovieSceneScriptingFloatChannel}
          Shear.Y {type=MovieSceneScriptingFloatChannel}
         ]
        }]
      }
      {
       type=MovieSceneFloatTrack
       display_name=Render Opacity
       sections [{
         type=MovieSceneFloatSection
         end_frame=240
         start_frame=0
         channels [None {type=MovieSceneScriptingFloatChannel, keys [[55,0],[61,1],[172.002,1],[223.006,0]]}]
        }]
      }
      {
       type=MovieSceneWidgetMaterialTrack
       display_name=Font.FontMaterial
       sections [{
         type=MovieSceneParameterSection
         end_frame=240
         start_frame=0
         channels [
          Pan {type=MovieSceneScriptingFloatChannel, keys [[0,1.25],[70,1.25],[196,-0.25]]}
          ColorR {type=MovieSceneScriptingFloatChannel}
          ColorG {type=MovieSceneScriptingFloatChannel}
          ColorB {type=MovieSceneScriptingFloatChannel}
          ColorA {type=MovieSceneScriptingFloatChannel}
         ]
        }]
      }
     ]
    }
    Logo {
     id=#2
     possessed_class=Image
     tracks [
      {
       type=MovieScene2DTransformTrack
       sections [{
         type=MovieScene2DTransformSection
         end_frame=240
         start_frame=0
         channels [
          Translation.X {type=MovieSceneScriptingFloatChannel}
          Translation.Y {type=MovieSceneScriptingFloatChannel}
          Angle {type=MovieSceneScriptingFloatChannel, keys [[0,32],[20.998,-16]]}
          Scale.X {type=MovieSceneScriptingFloatChannel, keys [[0,2.75],[8.998,0.5],[13,1.325],[21,1],[120,1],[225,0.95]]}
          Scale.Y {type=MovieSceneScriptingFloatChannel, keys [[0,2.75],[8.998,0.5],[13,1.125],[21,1],[120,1],[225,0.95]]}
          Shear.X {type=MovieSceneScriptingFloatChannel}
          Shear.Y {type=MovieSceneScriptingFloatChannel}
         ]
        }]
      }
      {
       type=MovieSceneFloatTrack
       display_name=Render Opacity
       sections [{
         type=MovieSceneFloatSection
         end_frame=240
         start_frame=0
         channels [None {type=MovieSceneScriptingFloatChannel, keys [[0,0],[6,0.75],[69,0.75],[225,0]]}]
        }]
      }
      {
       type=MovieSceneWidgetMaterialTrack
       sections [{
         type=MovieSceneParameterSection
         end_frame=240
         start_frame=0
         channels [Alpha {type=MovieSceneScriptingFloatChannel, keys [[0,0],[6,0.5],[15,0.2]]}]
        }]
      }
     ]
    }
    Placement {
     id=#11
     possessed_class=CommonTextBlock
     tracks [
      {
       type=MovieSceneWidgetMaterialTrack
       display_name=Font.Outline Settings.OutlineMaterial
       sections [{
         type=MovieSceneParameterSection
         end_frame=240
         start_frame=0
         channels [
          Alpha {type=MovieSceneScriptingFloatChannel, keys [[15,0],[21,1],[62,0]]}
          Pan {type=MovieSceneScriptingFloatChannel, keys [[0,1.25],[11.998,1.25],[91.998,-0.25]]}
          ColorR {type=MovieSceneScriptingFloatChannel, const [1,2]}
          ColorG {type=MovieSceneScriptingFloatChannel, keys [[21,0.9167],[62,0.5]]}
          ColorB {type=MovieSceneScriptingFloatChannel, const [0,2]}
          ColorA {type=MovieSceneScriptingFloatChannel}
         ]
        }]
      }
      {
       type=MovieSceneFloatTrack
       display_name=Render Opacity
       sections [{
         type=MovieSceneFloatSection
         end_frame=240
         start_frame=0
         channels [None {type=MovieSceneScriptingFloatChannel, keys [[8.999,0],[15,1],[177.002,1],[240.002,0]]}]
        }]
      }
      {
       type=MovieScene2DTransformTrack
       sections [{
         type=MovieScene2DTransformSection
         end_frame=240
         start_frame=-1
         channels [
          Translation.X {type=MovieSceneScriptingFloatChannel}
          Translation.Y {type=MovieSceneScriptingFloatChannel}
          Angle {type=MovieSceneScriptingFloatChannel}
          Scale.X {type=MovieSceneScriptingFloatChannel, keys [[8.999,0],[15,1.25],[21,0.925],[30,1]]}
          Scale.Y {type=MovieSceneScriptingFloatChannel, keys [[8.999,0],[15,1.2],[21,0.975],[30,1]]}
          Shear.X {type=MovieSceneScriptingFloatChannel}
          Shear.Y {type=MovieSceneScriptingFloatChannel}
         ]
        }]
      }
     ]
    }
    SmallTriangles_Left {
     id=#12
     possessed_class=Image
     tracks [{
       type=MovieSceneWidgetMaterialTrack
       sections [{
         type=MovieSceneParameterSection
         end_frame=240
         start_frame=0
         channels [OpacityOverride {type=MovieSceneScriptingFloatChannel, keys [[7,0],[15,4],[31,2]]}]
        }]
      }]
    }
    SmallTriangles {
     id=#3
     possessed_class=SizeBox
     tracks [
      {
       type=MovieScene2DTransformTrack
       sections [{
         type=MovieScene2DTransformSection
         end_frame=240
         start_frame=-1
         channels [
          Translation.X {type=MovieSceneScriptingFloatChannel}
          Translation.Y {type=MovieSceneScriptingFloatChannel}
          Angle {type=MovieSceneScriptingFloatChannel}
          Scale.X {type=MovieSceneScriptingFloatChannel, keys [[7,0.25],[16,1]]}
          Scale.Y {type=MovieSceneScriptingFloatChannel, keys [[7,0.5],[16,1]]}
          Shear.X {type=MovieSceneScriptingFloatChannel}
          Shear.Y {type=MovieSceneScriptingFloatChannel}
         ]
        }]
      }
      {
       type=MovieSceneFloatTrack
       display_name=Render Opacity
       sections [{
         type=MovieSceneFloatSection
         end_frame=240
         start_frame=-1
         channels [None {type=MovieSceneScriptingFloatChannel, keys [[7,0],[10,1],[21,0.9],[66,0]]}]
        }]
      }
     ]
    }
    Flare {
     id=#4
     possessed_class=Image
     tracks [
      {
       type=MovieSceneWidgetMaterialTrack
       sections [{
         type=MovieSceneParameterSection
         end_frame=240
         start_frame=0
         channels [
          Hardness {type=MovieSceneScriptingFloatChannel, keys [[3,4],[36,1.5],[54,1.5],[63,4]]}
          Intensity {type=MovieSceneScriptingFloatChannel, keys [[3,24],[36,4],[54,2.5],[63,4]]}
          VTile {type=MovieSceneScriptingFloatChannel, keys [[0,1.5],[54,1.5],[63,1]]}
         ]
        }]
      }
      {
       type=MovieSceneFloatTrack
       display_name=Render Opacity
       sections [{type=MovieSceneFloatSection, end_frame=240, start_frame=0, channels [None {type=MovieSceneScriptingFloatChannel, keys [[3,0],[5,1]]}]}]
      }
      {
       type=MovieScene2DTransformTrack
       sections [{
         type=MovieScene2DTransformSection
         end_frame=240
         start_frame=0
         channels [
          Translation.X {type=MovieSceneScriptingFloatChannel}
          Translation.Y {type=MovieSceneScriptingFloatChannel, keys [[54,0],[63,76]]}
          Angle {type=MovieSceneScriptingFloatChannel, const [0,1]}
          Scale.X {type=MovieSceneScriptingFloatChannel, keys [[3,1.125],[15,2],[54,2],[63,0]]}
          Scale.Y {type=MovieSceneScriptingFloatChannel, keys [[3,0.125],[10,0.2],[17,1],[54,1],[63,0.5]]}
          Shear.X {type=MovieSceneScriptingFloatChannel}
          Shear.Y {type=MovieSceneScriptingFloatChannel}
         ]
        }]
      }
     ]
    }
    TeamOrPlayer {
     id=#5
     possessed_class=CommonTextBlock
     tracks [
      {
       type=MovieScene2DTransformTrack
       sections [{
         type=MovieScene2DTransformSection
         end_frame=240
         start_frame=-1
         channels [
          Translation.X {type=MovieSceneScriptingFloatChannel}
          Translation.Y {type=MovieSceneScriptingFloatChannel}
          Angle {type=MovieSceneScriptingFloatChannel}
          Scale.X {type=MovieSceneScriptingFloatChannel, keys [[8.999,0],[15,1.25],[21,0.925],[30,1]]}
          Scale.Y {type=MovieSceneScriptingFloatChannel, keys [[8.999,0],[15,1.2],[21,0.975],[30,1]]}
          Shear.X {type=MovieSceneScriptingFloatChannel}
          Shear.Y {type=MovieSceneScriptingFloatChannel}
         ]
        }]
      }
      {
       type=MovieSceneFloatTrack
       display_name=Render Opacity
       sections [{
         type=MovieSceneFloatSection
         end_frame=240
         start_frame=0
         channels [None {type=MovieSceneScriptingFloatChannel, keys [[8.999,0],[15,1],[177.002,1],[240.002,0]]}]
        }]
      }
      {
       type=MovieSceneWidgetMaterialTrack
       display_name=Font.Outline Settings.OutlineMaterial
       sections [{
         type=MovieSceneParameterSection
         end_frame=240
         start_frame=0
         channels [
          Alpha {type=MovieSceneScriptingFloatChannel, keys [[15,0],[21,1],[62,0]]}
          Pan {type=MovieSceneScriptingFloatChannel, keys [[0,1.25],[11.998,1.25],[91.998,-0.25]]}
          ColorR {type=MovieSceneScriptingFloatChannel, const [1,2]}
          ColorG {type=MovieSceneScriptingFloatChannel, keys [[21,0.9167],[62,0.5]]}
          ColorB {type=MovieSceneScriptingFloatChannel, const [0,2]}
          ColorA {type=MovieSceneScriptingFloatChannel}
         ]
        }]
      }
     ]
    }
    BigWinTextFaded {
     id=#6
     possessed_class=CommonTextBlock
     tracks [
      {
       type=MovieScene2DTransformTrack
       sections [{
         type=MovieScene2DTransformSection
         end_frame=240
         start_frame=0
         channels [
          Translation.X {type=MovieSceneScriptingFloatChannel}
          Translation.Y {type=MovieSceneScriptingFloatChannel}
          Angle {type=MovieSceneScriptingFloatChannel}
          Scale.X {type=MovieSceneScriptingFloatChannel, keys [[11.998,1.25],[240.003,1]]}
          Scale.Y {type=MovieSceneScriptingFloatChannel, keys [[11.998,1.25],[240.003,1]]}
          Shear.X {type=MovieSceneScriptingFloatChannel}
          Shear.Y {type=MovieSceneScriptingFloatChannel}
         ]
        }]
      }
      {
       type=MovieSceneFloatTrack
       display_name=Render Opacity
       sections [{
         type=MovieSceneFloatSection
         end_frame=240
         start_frame=0
         channels [None {type=MovieSceneScriptingFloatChannel, keys [[11.998,0],[27,0.25],[62,0]]}]
        }]
      }
     ]
    }
    LargeTriangles {
     id=#8
     possessed_class=SizeBox
     tracks [
      {
       type=MovieScene2DTransformTrack
       sections [{
         type=MovieScene2DTransformSection
         end_frame=240
         start_frame=0
         channels [
          Translation.X {type=MovieSceneScriptingFloatChannel}
          Translation.Y {type=MovieSceneScriptingFloatChannel}
          Angle {type=MovieSceneScriptingFloatChannel}
          Scale.X {type=MovieSceneScriptingFloatChannel, keys [[12,0],[23,1],[72,1.25]]}
          Scale.Y {type=MovieSceneScriptingFloatChannel, keys [[12,0.5],[20,1],[72,1.25]]}
          Shear.X {type=MovieSceneScriptingFloatChannel}
          Shear.Y {type=MovieSceneScriptingFloatChannel}
         ]
        }]
      }
      {
       type=MovieSceneFloatTrack
       display_name=Render Opacity
       sections [{
         type=MovieSceneFloatSection
         end_frame=240
         start_frame=0
         channels [None {type=MovieSceneScriptingFloatChannel, keys [[12,0],[20,0.5],[72,0]]}]
        }]
      }
     ]
    }
    SmallTriangles_Right {
     id=#13
     possessed_class=Image
     tracks [{
       type=MovieSceneWidgetMaterialTrack
       sections [{
         type=MovieSceneParameterSection
         end_frame=240
         start_frame=0
         channels [OpacityOverride {type=MovieSceneScriptingFloatChannel, keys [[7,0],[15,4],[31,2]]}]
        }]
      }]
    }
    SmallGlow {
     id=#9
     possessed_class=Image
     tracks [
      {
       type=MovieSceneFloatTrack
       display_name=Render Opacity
       sections [{
         type=MovieSceneFloatSection
         end_frame=240
         start_frame=0
         channels [None {type=MovieSceneScriptingFloatChannel, keys [[10,0],[16,1],[90,0]]}]
        }]
      }
      {
       type=MovieScene2DTransformTrack
       sections [{
         type=MovieScene2DTransformSection
         end_frame=240
         start_frame=0
         channels [
          Translation.X {type=MovieSceneScriptingFloatChannel}
          Translation.Y {type=MovieSceneScriptingFloatChannel}
          Angle {type=MovieSceneScriptingFloatChannel}
          Scale.X {type=MovieSceneScriptingFloatChannel, keys [[0,1],[45,1],[90,0.5]]}
          Scale.Y {type=MovieSceneScriptingFloatChannel, keys [[0,1],[45,1],[90,0.5]]}
          Shear.X {type=MovieSceneScriptingFloatChannel}
          Shear.Y {type=MovieSceneScriptingFloatChannel}
         ]
        }]
      }
     ]
    }
    Glow {
     id=#10
     possessed_class=Image
     tracks [
      {
       type=MovieSceneFloatTrack
       display_name=Render Opacity
       sections [{
         type=MovieSceneFloatSection
         end_frame=240
         start_frame=0
         channels [None {type=MovieSceneScriptingFloatChannel, keys [[10,0],[16,1],[90,0]]}]
        }]
      }
      {
       type=MovieScene2DTransformTrack
       sections [{
         type=MovieScene2DTransformSection
         end_frame=240
         start_frame=0
         channels [
          Translation.X {type=MovieSceneScriptingFloatChannel}
          Translation.Y {type=MovieSceneScriptingFloatChannel}
          Angle {type=MovieSceneScriptingFloatChannel}
          Scale.X {type=MovieSceneScriptingFloatChannel, keys [[0,1],[54,1],[90,0.25]]}
          Scale.Y {type=MovieSceneScriptingFloatChannel, keys [[0,1],[54,1],[90,0.25]]}
          Shear.X {type=MovieSceneScriptingFloatChannel}
          Shear.Y {type=MovieSceneScriptingFloatChannel}
         ]
        }]
      }
     ]
    }
   ]
   display_rate {numerator=60}
  }
 }
 Intro_NoFade {
  binding_count=11
  length_seconds=3.5
  object=&self:Intro_NoFade
  sequence {
   bindings [
    WinDescription {
     id=#1
     possessed_class=CommonTextBlock
     tracks [
      {
       type=MovieScene2DTransformTrack
       sections [{
         type=MovieScene2DTransformSection
         end_frame=209
         start_frame=0
         channels [
          Translation.X {type=MovieSceneScriptingFloatChannel}
          Translation.Y {type=MovieSceneScriptingFloatChannel}
          Angle {type=MovieSceneScriptingFloatChannel}
          Scale.X {type=MovieSceneScriptingFloatChannel, keys [[55,6],[64,0.9],[67,1.25],[73,1]]}
          Scale.Y {type=MovieSceneScriptingFloatChannel, keys [[64,1],[67,1.25],[73,1]]}
          Shear.X {type=MovieSceneScriptingFloatChannel}
          Shear.Y {type=MovieSceneScriptingFloatChannel}
         ]
        }]
      }
      {
       type=MovieSceneFloatTrack
       display_name=Render Opacity
       sections [{
         type=MovieSceneFloatSection
         end_frame=210
         start_frame=0
         channels [None {type=MovieSceneScriptingFloatChannel, keys [[55,0],[61,1],[172.002,1]]}]
        }]
      }
      {
       type=MovieSceneWidgetMaterialTrack
       display_name=Font.FontMaterial
       sections [{
         type=MovieSceneParameterSection
         end_frame=210
         start_frame=0
         channels [
          Pan {type=MovieSceneScriptingFloatChannel, keys [[0,1.25],[70,1.25],[196,-0.25]]}
          ColorR {type=MovieSceneScriptingFloatChannel}
          ColorG {type=MovieSceneScriptingFloatChannel}
          ColorB {type=MovieSceneScriptingFloatChannel}
          ColorA {type=MovieSceneScriptingFloatChannel}
         ]
        }]
      }
     ]
    }
    Logo {
     id=#2
     possessed_class=Image
     tracks [
      {
       type=MovieScene2DTransformTrack
       sections [{
         type=MovieScene2DTransformSection
         end_frame=210
         start_frame=0
         channels [
          Translation.X {type=MovieSceneScriptingFloatChannel}
          Translation.Y {type=MovieSceneScriptingFloatChannel}
          Angle {type=MovieSceneScriptingFloatChannel, keys [[0,32],[20.998,-16]]}
          Scale.X {type=MovieSceneScriptingFloatChannel, keys [[0,2.75],[8.998,0.5],[13,1.325],[21,1],[120,1]]}
          Scale.Y {type=MovieSceneScriptingFloatChannel, keys [[0,2.75],[8.998,0.5],[13,1.125],[21,1],[120,1]]}
          Shear.X {type=MovieSceneScriptingFloatChannel}
          Shear.Y {type=MovieSceneScriptingFloatChannel}
         ]
        }]
      }
      {
       type=MovieSceneFloatTrack
       display_name=Render Opacity
       sections [{
         type=MovieSceneFloatSection
         end_frame=210
         start_frame=0
         channels [None {type=MovieSceneScriptingFloatChannel, keys [[0,0],[6,0.75],[120,0.75]]}]
        }]
      }
      {
       type=MovieSceneWidgetMaterialTrack
       sections [{
         type=MovieSceneParameterSection
         end_frame=210
         start_frame=0
         channels [Alpha {type=MovieSceneScriptingFloatChannel, keys [[0,0],[6,0.5],[15,0.2]]}]
        }]
      }
     ]
    }
    SmallTriangles_Left {
     id=#12
     possessed_class=Image
     tracks [{
       type=MovieSceneWidgetMaterialTrack
       sections [{
         type=MovieSceneParameterSection
         end_frame=210
         start_frame=0
         channels [OpacityOverride {type=MovieSceneScriptingFloatChannel, keys [[7,0],[15,4],[31,2]]}]
        }]
      }]
    }
    SmallTriangles {
     id=#3
     possessed_class=SizeBox
     tracks [
      {
       type=MovieScene2DTransformTrack
       sections [{
         type=MovieScene2DTransformSection
         end_frame=210
         start_frame=-1
         channels [
          Translation.X {type=MovieSceneScriptingFloatChannel}
          Translation.Y {type=MovieSceneScriptingFloatChannel}
          Angle {type=MovieSceneScriptingFloatChannel}
          Scale.X {type=MovieSceneScriptingFloatChannel, keys [[7,0.25],[16,1]]}
          Scale.Y {type=MovieSceneScriptingFloatChannel, keys [[7,0.5],[16,1]]}
          Shear.X {type=MovieSceneScriptingFloatChannel}
          Shear.Y {type=MovieSceneScriptingFloatChannel}
         ]
        }]
      }
      {
       type=MovieSceneFloatTrack
       display_name=Render Opacity
       sections [{
         type=MovieSceneFloatSection
         end_frame=210
         start_frame=-1
         channels [None {type=MovieSceneScriptingFloatChannel, keys [[7,0],[10,1],[21,0.9],[66,0]]}]
        }]
      }
     ]
    }
    Flare {
     id=#4
     possessed_class=Image
     tracks [
      {
       type=MovieSceneWidgetMaterialTrack
       sections [{
         type=MovieSceneParameterSection
         end_frame=210
         start_frame=0
         channels [
          Hardness {type=MovieSceneScriptingFloatChannel, keys [[3,4],[36,1.5],[54,1.5],[63,4]]}
          Intensity {type=MovieSceneScriptingFloatChannel, keys [[3,24],[36,4],[54,2.5],[63,4]]}
          VTile {type=MovieSceneScriptingFloatChannel, keys [[0,1.5],[54,1.5],[63,1]]}
         ]
        }]
      }
      {
       type=MovieSceneFloatTrack
       display_name=Render Opacity
       sections [{type=MovieSceneFloatSection, end_frame=210, start_frame=0, channels [None {type=MovieSceneScriptingFloatChannel, keys [[3,0],[5,1]]}]}]
      }
      {
       type=MovieScene2DTransformTrack
       sections [{
         type=MovieScene2DTransformSection
         end_frame=210
         start_frame=0
         channels [
          Translation.X {type=MovieSceneScriptingFloatChannel}
          Translation.Y {type=MovieSceneScriptingFloatChannel, keys [[54,0],[63,76]]}
          Angle {type=MovieSceneScriptingFloatChannel, const [0,1]}
          Scale.X {type=MovieSceneScriptingFloatChannel, keys [[3,1.125],[15,2],[54,2],[63,0]]}
          Scale.Y {type=MovieSceneScriptingFloatChannel, keys [[3,0.125],[10,0.2],[17,1],[54,1],[63,0.5]]}
          Shear.X {type=MovieSceneScriptingFloatChannel}
          Shear.Y {type=MovieSceneScriptingFloatChannel}
         ]
        }]
      }
     ]
    }
    Placement {
     id=#14
     possessed_class=CommonTextBlock
     tracks [
      {
       type=MovieSceneWidgetMaterialTrack
       display_name=Font.Outline Settings.OutlineMaterial
       sections [{
         type=MovieSceneParameterSection
         end_frame=209
         start_frame=0
         channels [
          Alpha {type=MovieSceneScriptingFloatChannel, keys [[15,0],[21,1],[62,0]]}
          Pan {type=MovieSceneScriptingFloatChannel, keys [[0,1.25],[11.998,1.25],[91.998,-0.25]]}
          ColorR {type=MovieSceneScriptingFloatChannel, const [1,2]}
          ColorG {type=MovieSceneScriptingFloatChannel, keys [[21,0.9167],[62,0.5]]}
          ColorB {type=MovieSceneScriptingFloatChannel, const [0,2]}
          ColorA {type=MovieSceneScriptingFloatChannel}
         ]
        }]
      }
      {
       type=MovieSceneFloatTrack
       display_name=Render Opacity
       sections [{
         type=MovieSceneFloatSection
         end_frame=210
         start_frame=0
         channels [None {type=MovieSceneScriptingFloatChannel, keys [[8.999,0],[15,1],[177.002,1]]}]
        }]
      }
      {
       type=MovieScene2DTransformTrack
       sections [{
         type=MovieScene2DTransformSection
         end_frame=210
         start_frame=-1
         channels [
          Translation.X {type=MovieSceneScriptingFloatChannel}
          Translation.Y {type=MovieSceneScriptingFloatChannel}
          Angle {type=MovieSceneScriptingFloatChannel}
          Scale.X {type=MovieSceneScriptingFloatChannel, keys [[8.999,0],[15,1.25],[21,0.925],[30,1]]}
          Scale.Y {type=MovieSceneScriptingFloatChannel, keys [[8.999,0],[15,1.2],[21,0.975],[30,1]]}
          Shear.X {type=MovieSceneScriptingFloatChannel}
          Shear.Y {type=MovieSceneScriptingFloatChannel}
         ]
        }]
      }
     ]
    }
    TeamOrPlayer {
     id=#5
     possessed_class=CommonTextBlock
     tracks [
      {
       type=MovieScene2DTransformTrack
       sections [{
         type=MovieScene2DTransformSection
         end_frame=210
         start_frame=-1
         channels [
          Translation.X {type=MovieSceneScriptingFloatChannel}
          Translation.Y {type=MovieSceneScriptingFloatChannel}
          Angle {type=MovieSceneScriptingFloatChannel}
          Scale.X {type=MovieSceneScriptingFloatChannel, keys [[8.999,0],[15,1.25],[21,0.925],[30,1]]}
          Scale.Y {type=MovieSceneScriptingFloatChannel, keys [[8.999,0],[15,1.2],[21,0.975],[30,1]]}
          Shear.X {type=MovieSceneScriptingFloatChannel}
          Shear.Y {type=MovieSceneScriptingFloatChannel}
         ]
        }]
      }
      {
       type=MovieSceneFloatTrack
       display_name=Render Opacity
       sections [{
         type=MovieSceneFloatSection
         end_frame=210
         start_frame=0
         channels [None {type=MovieSceneScriptingFloatChannel, keys [[8.999,0],[15,1],[177.002,1]]}]
        }]
      }
      {
       type=MovieSceneWidgetMaterialTrack
       display_name=Font.Outline Settings.OutlineMaterial
       sections [{
         type=MovieSceneParameterSection
         end_frame=209
         start_frame=0
         channels [
          Alpha {type=MovieSceneScriptingFloatChannel, keys [[15,0],[21,1],[62,0]]}
          Pan {type=MovieSceneScriptingFloatChannel, keys [[0,1.25],[11.998,1.25],[91.998,-0.25]]}
          ColorR {type=MovieSceneScriptingFloatChannel, const [1,2]}
          ColorG {type=MovieSceneScriptingFloatChannel, keys [[21,0.9167],[62,0.5]]}
          ColorB {type=MovieSceneScriptingFloatChannel, const [0,2]}
          ColorA {type=MovieSceneScriptingFloatChannel}
         ]
        }]
      }
     ]
    }
    LargeTriangles {
     id=#8
     possessed_class=SizeBox
     tracks [
      {
       type=MovieScene2DTransformTrack
       sections [{
         type=MovieScene2DTransformSection
         end_frame=210
         start_frame=0
         channels [
          Translation.X {type=MovieSceneScriptingFloatChannel}
          Translation.Y {type=MovieSceneScriptingFloatChannel}
          Angle {type=MovieSceneScriptingFloatChannel}
          Scale.X {type=MovieSceneScriptingFloatChannel, keys [[12,0],[23,1],[72,1.25]]}
          Scale.Y {type=MovieSceneScriptingFloatChannel, keys [[12,0.5],[20,1],[72,1.25]]}
          Shear.X {type=MovieSceneScriptingFloatChannel}
          Shear.Y {type=MovieSceneScriptingFloatChannel}
         ]
        }]
      }
      {
       type=MovieSceneFloatTrack
       display_name=Render Opacity
       sections [{
         type=MovieSceneFloatSection
         end_frame=210
         start_frame=0
         channels [None {type=MovieSceneScriptingFloatChannel, keys [[12,0],[20,0.5],[72,0]]}]
        }]
      }
     ]
    }
    SmallTriangles_Right {
     id=#13
     possessed_class=Image
     tracks [{
       type=MovieSceneWidgetMaterialTrack
       sections [{
         type=MovieSceneParameterSection
         end_frame=210
         start_frame=0
         channels [OpacityOverride {type=MovieSceneScriptingFloatChannel, keys [[7,0],[15,4],[31,2]]}]
        }]
      }]
    }
    SmallGlow {
     id=#9
     possessed_class=Image
     tracks [
      {
       type=MovieSceneFloatTrack
       display_name=Render Opacity
       sections [{
         type=MovieSceneFloatSection
         end_frame=210
         start_frame=0
         channels [None {type=MovieSceneScriptingFloatChannel, keys [[10,0],[16,1],[90,0]]}]
        }]
      }
      {
       type=MovieScene2DTransformTrack
       sections [{
         type=MovieScene2DTransformSection
         end_frame=210
         start_frame=0
         channels [
          Translation.X {type=MovieSceneScriptingFloatChannel}
          Translation.Y {type=MovieSceneScriptingFloatChannel}
          Angle {type=MovieSceneScriptingFloatChannel}
          Scale.X {type=MovieSceneScriptingFloatChannel, keys [[0,1],[45,1],[90,0.5]]}
          Scale.Y {type=MovieSceneScriptingFloatChannel, keys [[0,1],[45,1],[90,0.5]]}
          Shear.X {type=MovieSceneScriptingFloatChannel}
          Shear.Y {type=MovieSceneScriptingFloatChannel}
         ]
        }]
      }
     ]
    }
    Glow {
     id=#10
     possessed_class=Image
     tracks [
      {
       type=MovieSceneFloatTrack
       display_name=Render Opacity
       sections [{
         type=MovieSceneFloatSection
         end_frame=210
         start_frame=0
         channels [None {type=MovieSceneScriptingFloatChannel, keys [[10,0],[16,1],[90,0]]}]
        }]
      }
      {
       type=MovieScene2DTransformTrack
       sections [{
         type=MovieScene2DTransformSection
         end_frame=210
         start_frame=0
         channels [
          Translation.X {type=MovieSceneScriptingFloatChannel}
          Translation.Y {type=MovieSceneScriptingFloatChannel}
          Angle {type=MovieSceneScriptingFloatChannel}
          Scale.X {type=MovieSceneScriptingFloatChannel, keys [[0,1],[54,1],[90,0.25]]}
          Scale.Y {type=MovieSceneScriptingFloatChannel, keys [[0,1],[54,1],[90,0.25]]}
          Shear.X {type=MovieSceneScriptingFloatChannel}
          Shear.Y {type=MovieSceneScriptingFloatChannel}
         ]
        }]
      }
     ]
    }
   ]
   display_rate {numerator=60}
  }
 }
 FadeOut {
  binding_count=11
  length_seconds=1
  object=&self:FadeOut
  sequence {
   bindings [
    WinDescription {
     id=#1
     possessed_class=CommonTextBlock
     tracks [
      {
       type=MovieScene2DTransformTrack
       sections [{
         type=MovieScene2DTransformSection
         end_frame=59
         start_frame=0
         channels [
          Translation.X {type=MovieSceneScriptingFloatChannel}
          Translation.Y {type=MovieSceneScriptingFloatChannel}
          Angle {type=MovieSceneScriptingFloatChannel}
          Scale.X {type=MovieSceneScriptingFloatChannel}
          Scale.Y {type=MovieSceneScriptingFloatChannel}
          Shear.X {type=MovieSceneScriptingFloatChannel}
          Shear.Y {type=MovieSceneScriptingFloatChannel}
         ]
        }]
      }
      {
       type=MovieSceneFloatTrack
       display_name=Render Opacity
       sections [{type=MovieSceneFloatSection, end_frame=60, start_frame=0, channels [None {type=MovieSceneScriptingFloatChannel, keys [[0,1],[45,0]]}]}]
      }
      {
       type=MovieSceneWidgetMaterialTrack
       display_name=Font.FontMaterial
       sections [{
         type=MovieSceneParameterSection
         end_frame=60
         start_frame=0
         channels [
          Pan {type=MovieSceneScriptingFloatChannel, const [-0.25,1]}
          ColorR {type=MovieSceneScriptingFloatChannel}
          ColorG {type=MovieSceneScriptingFloatChannel}
          ColorB {type=MovieSceneScriptingFloatChannel}
          ColorA {type=MovieSceneScriptingFloatChannel}
         ]
        }]
      }
     ]
    }
    Logo {
     id=#2
     possessed_class=Image
     tracks [
      {
       type=MovieScene2DTransformTrack
       sections [{
         type=MovieScene2DTransformSection
         end_frame=60
         start_frame=0
         channels [
          Translation.X {type=MovieSceneScriptingFloatChannel}
          Translation.Y {type=MovieSceneScriptingFloatChannel}
          Angle {type=MovieSceneScriptingFloatChannel, const [-16,1]}
          Scale.X {type=MovieSceneScriptingFloatChannel, const [1,1]}
          Scale.Y {type=MovieSceneScriptingFloatChannel, const [1,1]}
          Shear.X {type=MovieSceneScriptingFloatChannel}
          Shear.Y {type=MovieSceneScriptingFloatChannel}
         ]
        }]
      }
      {
       type=MovieSceneFloatTrack
       display_name=Render Opacity
       sections [{type=MovieSceneFloatSection, end_frame=60, start_frame=0, channels [None {type=MovieSceneScriptingFloatChannel, keys [[0,0.75],[30,0]]}]}]
      }
      {
       type=MovieSceneWidgetMaterialTrack
       sections [{type=MovieSceneParameterSection, end_frame=60, start_frame=0, channels [Alpha {type=MovieSceneScriptingFloatChannel, const [0.2,1]}]}]
      }
     ]
    }
    SmallTriangles_Left {
     id=#12
     possessed_class=Image
     tracks [{
       type=MovieSceneWidgetMaterialTrack
       sections [{type=MovieSceneParameterSection, end_frame=60, start_frame=0, channels [OpacityOverride {type=MovieSceneScriptingFloatChannel, const [2,1]}]}]
      }]
    }
    SmallTriangles {
     id=#3
     possessed_class=SizeBox
     tracks [
      {
       type=MovieScene2DTransformTrack
       sections [{
         type=MovieScene2DTransformSection
         end_frame=60
         start_frame=-1
         channels [
          Translation.X {type=MovieSceneScriptingFloatChannel}
          Translation.Y {type=MovieSceneScriptingFloatChannel}
          Angle {type=MovieSceneScriptingFloatChannel}
          Scale.X {type=MovieSceneScriptingFloatChannel, const [1,1]}
          Scale.Y {type=MovieSceneScriptingFloatChannel, const [1,1]}
          Shear.X {type=MovieSceneScriptingFloatChannel}
          Shear.Y {type=MovieSceneScriptingFloatChannel}
         ]
        }]
      }
      {
       type=MovieSceneFloatTrack
       display_name=Render Opacity
       sections [{type=MovieSceneFloatSection, end_frame=60, start_frame=-1, channels [None {type=MovieSceneScriptingFloatChannel, const [0,1]}]}]
      }
     ]
    }
    Flare {
     id=#4
     possessed_class=Image
     tracks [
      {
       type=MovieSceneWidgetMaterialTrack
       sections [{
         type=MovieSceneParameterSection
         end_frame=60
         start_frame=0
         channels [
          Hardness {type=MovieSceneScriptingFloatChannel, const [4,1]}
          Intensity {type=MovieSceneScriptingFloatChannel, const [4,1]}
          VTile {type=MovieSceneScriptingFloatChannel, const [1,1]}
         ]
        }]
      }
      {
       type=MovieSceneFloatTrack
       display_name=Render Opacity
       sections [{type=MovieSceneFloatSection, end_frame=60, start_frame=0, channels [None {type=MovieSceneScriptingFloatChannel, const [1,1]}]}]
      }
      {
       type=MovieScene2DTransformTrack
       sections [{
         type=MovieScene2DTransformSection
         end_frame=60
         start_frame=0
         channels [
          Translation.X {type=MovieSceneScriptingFloatChannel}
          Translation.Y {type=MovieSceneScriptingFloatChannel, const [76,1]}
          Angle {type=MovieSceneScriptingFloatChannel}
          Scale.X {type=MovieSceneScriptingFloatChannel, const [0,1]}
          Scale.Y {type=MovieSceneScriptingFloatChannel, const [0.5,1]}
          Shear.X {type=MovieSceneScriptingFloatChannel}
          Shear.Y {type=MovieSceneScriptingFloatChannel}
         ]
        }]
      }
     ]
    }
    TeamOrPlayer {
     id=#5
     possessed_class=CommonTextBlock
     tracks [
      {
       type=MovieScene2DTransformTrack
       sections [{
         type=MovieScene2DTransformSection
         end_frame=60
         start_frame=-1
         channels [
          Translation.X {type=MovieSceneScriptingFloatChannel}
          Translation.Y {type=MovieSceneScriptingFloatChannel}
          Angle {type=MovieSceneScriptingFloatChannel}
          Scale.X {type=MovieSceneScriptingFloatChannel, const [1,1]}
          Scale.Y {type=MovieSceneScriptingFloatChannel, const [1,1]}
          Shear.X {type=MovieSceneScriptingFloatChannel}
          Shear.Y {type=MovieSceneScriptingFloatChannel}
         ]
        }]
      }
      {
       type=MovieSceneFloatTrack
       display_name=Render Opacity
       sections [{type=MovieSceneFloatSection, end_frame=60, start_frame=0, channels [None {type=MovieSceneScriptingFloatChannel, keys [[0,1],[60,0]]}]}]
      }
      {
       type=MovieSceneWidgetMaterialTrack
       display_name=Font.Outline Settings.OutlineMaterial
       sections [{
         type=MovieSceneParameterSection
         end_frame=59
         start_frame=0
         channels [
          Alpha {type=MovieSceneScriptingFloatChannel, const [0,1]}
          Pan {type=MovieSceneScriptingFloatChannel, const [-0.25,1]}
          ColorR {type=MovieSceneScriptingFloatChannel, const [1,1]}
          ColorG {type=MovieSceneScriptingFloatChannel, const [0.5,1]}
          ColorB {type=MovieSceneScriptingFloatChannel, const [0,1]}
          ColorA {type=MovieSceneScriptingFloatChannel}
         ]
        }]
      }
     ]
    }
    LargeTriangles {
     id=#8
     possessed_class=SizeBox
     tracks [
      {
       type=MovieScene2DTransformTrack
       sections [{
         type=MovieScene2DTransformSection
         end_frame=60
         start_frame=0
         channels [
          Translation.X {type=MovieSceneScriptingFloatChannel}
          Translation.Y {type=MovieSceneScriptingFloatChannel}
          Angle {type=MovieSceneScriptingFloatChannel}
          Scale.X {type=MovieSceneScriptingFloatChannel, const [1.25,1]}
          Scale.Y {type=MovieSceneScriptingFloatChannel, const [1.25,1]}
          Shear.X {type=MovieSceneScriptingFloatChannel}
          Shear.Y {type=MovieSceneScriptingFloatChannel}
         ]
        }]
      }
      {
       type=MovieSceneFloatTrack
       display_name=Render Opacity
       sections [{type=MovieSceneFloatSection, end_frame=60, start_frame=0, channels [None {type=MovieSceneScriptingFloatChannel, const [0,1]}]}]
      }
     ]
    }
    Placement {
     id=#15
     possessed_class=CommonTextBlock
     tracks [
      {
       type=MovieSceneWidgetMaterialTrack
       display_name=Font.FontMaterial
       sections [{
         type=MovieSceneParameterSection
         end_frame=60
         start_frame=0
         channels [
          Pan {type=MovieSceneScriptingFloatChannel, const [-0.25,1]}
          ColorR {type=MovieSceneScriptingFloatChannel}
          ColorG {type=MovieSceneScriptingFloatChannel}
          ColorB {type=MovieSceneScriptingFloatChannel}
          ColorA {type=MovieSceneScriptingFloatChannel}
         ]
        }]
      }
      {
       type=MovieSceneFloatTrack
       display_name=Render Opacity
       sections [{type=MovieSceneFloatSection, end_frame=60, start_frame=0, channels [None {type=MovieSceneScriptingFloatChannel, keys [[0,1],[45,0]]}]}]
      }
      {
       type=MovieScene2DTransformTrack
       sections [{
         type=MovieScene2DTransformSection
         end_frame=59
         start_frame=0
         channels [
          Translation.X {type=MovieSceneScriptingFloatChannel}
          Translation.Y {type=MovieSceneScriptingFloatChannel}
          Angle {type=MovieSceneScriptingFloatChannel}
          Scale.X {type=MovieSceneScriptingFloatChannel}
          Scale.Y {type=MovieSceneScriptingFloatChannel}
          Shear.X {type=MovieSceneScriptingFloatChannel}
          Shear.Y {type=MovieSceneScriptingFloatChannel}
         ]
        }]
      }
     ]
    }
    SmallTriangles_Right {
     id=#13
     possessed_class=Image
     tracks [{
       type=MovieSceneWidgetMaterialTrack
       sections [{type=MovieSceneParameterSection, end_frame=60, start_frame=0, channels [OpacityOverride {type=MovieSceneScriptingFloatChannel, const [2,1]}]}]
      }]
    }
    SmallGlow {
     id=#9
     possessed_class=Image
     tracks [
      {
       type=MovieSceneFloatTrack
       display_name=Render Opacity
       sections [{type=MovieSceneFloatSection, end_frame=60, start_frame=0, channels [None {type=MovieSceneScriptingFloatChannel, const [0,1]}]}]
      }
      {
       type=MovieScene2DTransformTrack
       sections [{
         type=MovieScene2DTransformSection
         end_frame=60
         start_frame=0
         channels [
          Translation.X {type=MovieSceneScriptingFloatChannel}
          Translation.Y {type=MovieSceneScriptingFloatChannel}
          Angle {type=MovieSceneScriptingFloatChannel}
          Scale.X {type=MovieSceneScriptingFloatChannel, const [0.5,1]}
          Scale.Y {type=MovieSceneScriptingFloatChannel, const [0.5,1]}
          Shear.X {type=MovieSceneScriptingFloatChannel}
          Shear.Y {type=MovieSceneScriptingFloatChannel}
         ]
        }]
      }
     ]
    }
    Glow {
     id=#10
     possessed_class=Image
     tracks [
      {
       type=MovieSceneFloatTrack
       display_name=Render Opacity
       sections [{type=MovieSceneFloatSection, end_frame=60, start_frame=0, channels [None {type=MovieSceneScriptingFloatChannel, const [0,1]}]}]
      }
      {
       type=MovieScene2DTransformTrack
       sections [{
         type=MovieScene2DTransformSection
         end_frame=60
         start_frame=0
         channels [
          Translation.X {type=MovieSceneScriptingFloatChannel}
          Translation.Y {type=MovieSceneScriptingFloatChannel}
          Angle {type=MovieSceneScriptingFloatChannel}
          Scale.X {type=MovieSceneScriptingFloatChannel, const [0.25,1]}
          Scale.Y {type=MovieSceneScriptingFloatChannel, const [0.25,1]}
          Shear.X {type=MovieSceneScriptingFloatChannel}
          Shear.Y {type=MovieSceneScriptingFloatChannel}
         ]
        }]
      }
     ]
    }
   ]
   display_rate {numerator=60}
  }
 }
]
graphs [
 EventGraph {nodes [
   Event PreConstruct()
   Event Construct()
   Event Tick()
  ]}
 FadeoutWinnerAnim {nodes [
   FadeoutWinnerAnim() {is=title K2Node_FunctionEntry}
   > PlayAnimation(InAnimation=Get FadeOut() -> (FadeOut:Widget Animation Object Reference))
  ]}
 PlayWinnerAnim {nodes [
   n1 = PlayWinnerAnim!() -> (
    then
    WinnerName:Text
    HasWinner:bool=~
    IsWinner:bool=~
    IsSpectator:bool=~
    Placement:int
    GameEndCallout:Byte=WinLose
    VictoryAudioIndex:int
    DrawAudioIndex:int
    DefeatAudioIndex:int
    CustomVictoryCallout:Text
    CustomDefeatCallout:Text
   ) {is=title K2Node_FunctionEntry}
   > n2 = Set bAlwaysHasWinner(bAlwaysHasWinner:bool="Equal (Enum)"(
     A:EMinigameGameEndCallout Enum=Byte to Enum EMinigameGameEndCallout!(Byte:Byte=@n1.GameEndCallout) -> (ReturnValue:EMinigameGameEndCallout Enum) {is=title K2Node_CastByteToEnum}
     B:EMinigameGameEndCallout Enum=Cooperative
    )) -> (Output_Get:bool)
   > n3 = Branch(Condition=OR Boolean(A=@n1.HasWinner, B=@n2))
   > SetVisibility(
    self=[
     Get FlareNoise() -> (FlareNoise:Size Box Object Reference)
     Get GlowboxLarge() -> (GlowboxLarge:Size Box Object Reference)
     Get GlowBoxSmall() -> (GlowBoxSmall:Size Box Object Reference)
     Get LargeTriangles() -> (LargeTriangles:Size Box Object Reference)
     Get SmallTriangles() -> (SmallTriangles:Size Box Object Reference)
    ]
    InVisibility=SelfHitTestInvisible
   )
   > SetVisibility(
    self=[Get WinDescription() -> (WinDescription:Common Text Object Reference), Get TeamOrPlayer() -> (TeamOrPlayer:Common Text Object Reference)]
    InVisibility=SelfHitTestInvisible
   )
   > SetVisibility(self=Get Placement() -> (Placement:Common Text Object Reference), InVisibility=Hidden)
   > K2Node_Composite_0 = DetermineResultAnim!(execute, bForceWin:bool=Get bAlwaysHasWinner() -> (bAlwaysHasWinner:bool)) -> (Spectator:Exec, Winner:Exec, Loser:Exec) {is=title K2Node_Composite}
   "SetText (Text)"(execute=@K2Node_Composite_0.Loser, self=Get TeamOrPlayer(), InText="LOCTEXT(\"Defeat\")")
   > n4 = Switch on EMinigameGameEndCallout!(execute, Selection:EMinigameGameEndCallout Enum=Byte to Enum EMinigameGameEndCallout(Byte=Get GameEndCallout() -> (GameEndCallout:Byte=0))) -> (WinLose:Exec, Placement:Exec, Cooperative:Exec) {is=title K2Node_SwitchEnum}
   SetVisibility(execute=@n4.Placement, self=Get Placement(), InVisibility=SelfHitTestInvisible)
   > SetVisibility(self=[Get TeamOrPlayer(), Get WinDescription()], InVisibility=Collapsed)
   n5 = "To String (Integer)"(InInt=Get Placement() -> (Placement:int=0))
   > "SetText (Text)"(self=Get Placement() -> (Placement:Common Text Object Reference), InText="To Text (String)"(InString=Append(A="You Placed ##", B=@n5)))
   > n6 = Set Placement Visuals() {is=title K2Node_CallFunction}
   n7 = Branch(
    execute=[@n11.then, @n6.then]
    Condition="Equal (Integer)"(A=Get DefeatAudioIndex() -> (DefeatAudioIndex:int=0), B=0) {is=title K2Node_CallFunction}
   )
   > n8 = PlaySound2D(Sound=/Game/Sounds/Creative/IngameUI/UI_VictoryScore_Defeat_Cue)
   PlayAnimation(
    execute=[@n15.then, @n8.then, @n16.then, @n9.then, @n25.then, @n26.then]
    InAnimation=Get Intro_NoFade() -> (Intro_NoFade:Widget Animation Object Reference)
   )
   n9 = PlaySound2D(
    execute=@n7.else
    Sound="Get (a copy)"(
     Array:Array of Sound Cue Object References=Get AudioSamples() -> (AudioSamples:Array of Sound Cue Object References)
     Dimension 1=Get DefeatAudioIndex() -> (DefeatAudioIndex=0)
    ) -> (Output:Sound Cue Object Reference)
   )
   n10 = Get CustomDefeatCallout() -> (CustomDefeatCallout:Text)
   "SetText (Text)"(
    execute=@n4.WinLose
    self=Get WinDescription()
    InText=Select(Option 0:Text=@n10, Option 1:Text="LOCTEXT(\"You Lose\")", Index=TextIsEmpty(InText=@n10)) -> (ReturnValue:Text)
   )
   > n11 = SetDefeatVisuals()
   SetVisibility(execute=[@n4.Cooperative, @n21.Cooperative], self=Get Placement(), InVisibility=SelfHitTestInvisible)
   > SetVisibility(self=[Get WinDescription(), Get TeamOrPlayer()], InVisibility=Collapsed)
   n12 = Get CustomVictoryCallout() -> (CustomVictoryCallout:Text)
   > "SetText (Text)"(self=Get Placement(), InText=Select(Option 0=@n12, Option 1="LOCTEXT(\"Thank you for playing!\")", Index=TextIsEmpty(InText=@n12)))
   > n13 = Set Placement Visuals()
   n14 = Branch(execute=[@n20.then, @n13.then], Condition="Equal (Integer)"(A=Get VictoryAudioIndex() -> (VictoryAudioIndex:int=0), B=0))
   > n15 = PlaySound2D(Sound=/Game/Sounds/Creative/IngameUI/UI_VictoryScore_Cue)
   n16 = PlaySound2D(execute=@n14.else, Sound="Get (a copy)"(Array=Get AudioSamples(), Dimension 1=Get VictoryAudioIndex() -> (VictoryAudioIndex=0)))
   "SetText (Text)"(execute=@K2Node_Composite_0.Spectator, self=Get TeamOrPlayer(), InText="LOCTEXT(\"Game Over\")")
   > n17 = Branch(Condition=Get HasWinner() -> (HasWinner:bool))
   n18 = "To String (Text)"(InText=Get WinnerName() -> (WinnerName:Text))
   > "SetText (Text)"(self=Get WinDescription(), InText="To Text (String)"(InString=Append(A=@n18, B=" Wins")))
   n19 = "SetText (Text)"(execute=@n17.else, self=Get WinDescription())
   n20 = SetVictoryVisuals(execute=[@n19.then, @n23.then])
   n21 = Switch on EMinigameGameEndCallout(execute=@K2Node_Composite_0.Winner, Selection=Byte to Enum EMinigameGameEndCallout(Byte=Get GameEndCallout() -> (GameEndCallout=0)))
   "SetText (Text)"(execute=[@n21.Placement, @n21.WinLose], self=Get TeamOrPlayer(), InText="LOCTEXT(\"Victory\")")
   n22 = Get CustomVictoryCallout()
   > n23 = "SetText (Text)"(self=Get WinDescription(), InText=Select(Option 0=@n22, Option 1="LOCTEXT(\"You Win!\")", Index=TextIsEmpty(InText=@n22)))
   "SetText (Text)"(execute=@n3.else, self=Get TeamOrPlayer(), InText="LOCTEXT(\"DRAW!\")")
   > SetVisibility(self=Get WinDescription(), InVisibility=Collapsed)
   > SetVisibility(self=[Get FlareNoise(), Get GlowboxLarge(), Get GlowBoxSmall(), Get LargeTriangles(), Get SmallTriangles()], InVisibility=Hidden)
   > n24 = Branch(Condition="Equal (Integer)"(A=Get DrawAudioIndex() -> (DrawAudioIndex:int=0), B=0))
   > n25 = PlaySound2D(Sound=/Game/Sounds/Creative/IngameUI/UI_VictoryScore_Draw_Cue)
   n26 = PlaySound2D(execute=@n24.else, Sound="Get (a copy)"(Array=Get AudioSamples(), Dimension 1=Get DrawAudioIndex() -> (DrawAudioIndex=0)))
  ]}
 DetermineResultAnim {
  path=PlayWinnerAnim.K2Node_Composite_0.DetermineResultAnim
  nodes [
   n1 = Inputs!() -> (execute, bForceWin:bool=~)
   > n2 = Branch(Condition=@n1)
   Outputs!(Spectator:Exec=@n3.then, Winner:Exec=[@n2.then, @n4.then], Loser:Exec=@n4.else)
   n3 = Branch(execute=@n2.else, Condition=Get IsSpectator() -> (IsSpectator:bool))
   n4 = Branch(execute=@n3.else, Condition=Get IsWinner() -> (IsWinner:bool))
  ]
 }
 Set Placement Visuals {nodes [
   Set Placement Visuals() {is=title K2Node_FunctionEntry}
   > SetBrushFromMaterial(
    self=[Get Glow() -> (Glow:Image Object Reference), Get SmallGlow() -> (SmallGlow:Image Object Reference)]
    Material=/Game/Items/Traps/Blueprints/Toys/UI_Art_Assets/M_UI_RoundEnd_Marker_Winner
   )
   > SetBrushFromMaterial(
    self=[
     Get LargeTriangles_Left() -> (LargeTriangles_Left:Image Object Reference)
     Get LargeTriangles_Right() -> (LargeTriangles_Right:Image Object Reference)
     Get SmallTriangles_Left() -> (SmallTriangles_Left:Image Object Reference)
     Get SmallTriangles_Right() -> (SmallTriangles_Right:Image Object Reference)
    ]
    Material=/Game/UI/Foundation/Materials/M_FX_TriangleScroll_Scoreboard_Small
   )
   > SetBrushFromMaterial(self=Get Flare() -> (Flare:Image Object Reference), Material=/Game/UI/Frontend/HeroLoadout/Art/Materials/FX/M_UI_FlareNoise_Test)
   > n1 = GetDynamicFontMaterial!(execute, self:Text Object Reference=Get Placement()) -> (then, ReturnValue:Material Instance Dynamic Object Reference)
   > SetVectorParameterValue(self=@n1, ParameterName=Color, Value=(R=0,G=1,B=1,A=1))
   > Set Outline Color!(In:Exec, CommonTextWidget:Common Text Object Reference=Get Placement(), Color:Linear Color Structure=(R=0,G=0.2502,B=1,A=0.25)) -> (Out:Exec) {is=title K2Node_MacroInstance}
  ]}
 SetDefeatVisuals {nodes [
   SetDefeatVisuals() {is=title K2Node_FunctionEntry}
   > SetBrushFromMaterial(self=[Get Glow(), Get SmallGlow()], Material=/Game/Items/Traps/Blueprints/Toys/UI_Art_Assets/M_UI_RoundEnd_Marker_Defeat)
   > SetBrushFromMaterial(
    self=[Get LargeTriangles_Left(), Get LargeTriangles_Right(), Get SmallTriangles_Left(), Get SmallTriangles_Right()]
    Material=/Game/UI/Foundation/Materials/M_FX_TriangleScroll_Scoreboard_Small_Defeat
   )
   > SetBrushFromMaterial(self=Get Flare(), Material=/Game/UI/Frontend/HeroLoadout/Art/Materials/FX/M_UI_FlareNoise_Defeat)
   > n1 = GetDynamicFontMaterial(self=Get WinDescription())
   > SetVectorParameterValue(self=@n1, ParameterName=Color, Value=(R=1,G=0,B=0.0333,A=1))
   > Set Outline Color(CommonTextWidget=Get WinDescription(), Color=(R=0.75,G=0.025,B=0,A=0.25))
  ]}
 SetVictoryVisuals {nodes [
   SetVictoryVisuals() {is=title K2Node_FunctionEntry}
   > SetBrushFromMaterial(self=[Get Glow(), Get SmallGlow()], Material=&M_UI_RoundEnd_Marker_Winner)
   > SetBrushFromMaterial(
    self=[Get LargeTriangles_Left(), Get LargeTriangles_Right(), Get SmallTriangles_Left(), Get SmallTriangles_Right()]
    Material=&M_FX_TriangleScroll_Scoreboard_Small
   )
   > SetBrushFromMaterial(self=Get Flare(), Material=&M_UI_FlareNoise_Test)
   > n1 = GetDynamicFontMaterial(self=Get WinDescription())
   > SetVectorParameterValue(self=@n1, ParameterName=Color, Value=(R=0,G=1,B=1,A=1))
   > Set Outline Color(CommonTextWidget=Get WinDescription(), Color=(R=0,G=0.2502,B=1,A=0.25))
  ]}
]
variables [AudioSamples {
  type="array<SoundCue>"
  default [
   null
   /Game/Sounds/Creative/Gadgets/Radio/Stingers/Stinger_Accent_01_Cue
   /Game/Sounds/Creative/Gadgets/Radio/Stingers/Stinger_Accent_02_Cue
   /Game/Sounds/Creative/Gadgets/Radio/Stingers/Stinger_Accent_03_Cue
   /Game/Sounds/Creative/Gadgets/Radio/Stingers/Stinger_Bump_01_Cue
   /Game/Sounds/Creative/Gadgets/Radio/Stingers/Stinger_Bump_02_Cue
   /Game/Sounds/Creative/Gadgets/Radio/Stingers/Stinger_End_01_Cue
   /Game/Sounds/Creative/Gadgets/Radio/Stingers/Stinger_End_02_Cue
   /Game/Sounds/Creative/Gadgets/Radio/Stingers/Stinger_Failure_01_Cue
   /Game/Sounds/Creative/Gadgets/Radio/Stingers/Stinger_Failure_02_Cue
   /Game/Sounds/Creative/Gadgets/Radio/Stingers/Stinger_Flashback_01_Cue
   /Game/Sounds/Creative/Gadgets/Radio/Stingers/Stinger_GameOver_01_Cue
   /Game/Sounds/Creative/Gadgets/Radio/Stingers/Stinger_Magic_01_Cue
   /Game/Sounds/Creative/Gadgets/Radio/Stingers/Stinger_Mystery_01_Cue
   /Game/Sounds/Creative/Gadgets/Radio/Stingers/Stinger_Mystery_02_Cue
   /Game/Sounds/Creative/Gadgets/Radio/Stingers/Stinger_Mystery_03_Cue
   /Game/Sounds/Creative/Gadgets/Radio/Stingers/Stinger_Nuetral_01_Cue
   /Game/Sounds/Creative/Gadgets/Radio/Stingers/Stinger_Positive_01_Cue
   /Game/Sounds/Creative/Gadgets/Radio/Stingers/Stinger_Reward_01_Cue
   /Game/Sounds/Creative/Gadgets/Radio/Stingers/Stinger_Reward_02_Cue
   /Game/Sounds/Creative/Gadgets/Radio/Stingers/Stinger_Scare_01_Cue
   /Game/Sounds/Creative/Gadgets/Radio/Stingers/Stinger_Scare_02_Cue
   /Game/Sounds/Creative/Gadgets/Radio/Stingers/Stinger_Success_01_Cue
   /Game/Sounds/Creative/Gadgets/Radio/Stingers/Stinger_Success_02_Cue
   /Game/Sounds/Creative/Gadgets/Radio/Stingers/Stinger_Success_03_Cue
   /Game/Sounds/Creative/Gadgets/Radio/Stingers/Stinger_Suprise_01_Cue
   /Game/Sounds/Creative/Gadgets/Radio/Stingers/Stinger_Suprise_02_Cue
   /Game/Sounds/Creative/Gadgets/Radio/Stingers/Stinger_Threat_01_Cue
   /Game/Sounds/Creative/Gadgets/Radio/Stingers/Stinger_Transition_01_Cue
   /Game/Sounds/Creative/Gadgets/Radio/Stingers/Stinger_Transition_02_Cue
   /Game/Sounds/Creative/Gadgets/Radio/Stingers/Stinger_Unlock_01_Cue
   /Game/Sounds/Creative/Gadgets/Radio/Stingers/Stinger_Wash_01_Cue
   /Game/Sounds/Creative/Gadgets/Radio/Stingers/Halloween/Stinger_Halloween_Alert_01_Cue
   /Game/Sounds/Creative/Gadgets/Radio/Stingers/Halloween/Stinger_Halloween_Bonus_01_Cue
   /Game/Sounds/Creative/Gadgets/Radio/Stingers/Halloween/Stinger_Halloween_Discovery_01_Cue
   /Game/Sounds/Creative/Gadgets/Radio/Stingers/Halloween/Stinger_Halloween_End_01_Cue
   /Game/Sounds/Creative/Gadgets/Radio/Stingers/Halloween/Stinger_Halloween_Lose_01_Cue
   /Game/Sounds/Creative/Gadgets/Radio/Stingers/Halloween/Stinger_Halloween_Mysterious_01_Cue
   /Game/Sounds/Creative/Gadgets/Radio/Stingers/Halloween/Stinger_Halloween_Negative_01_Cue
   /Game/Sounds/Creative/Gadgets/Radio/Stingers/Halloween/Stinger_Halloween_Positive_01_Cue
   /Game/Sounds/Creative/Gadgets/Radio/Stingers/Halloween/Stinger_Halloween_Scare_01_Cue
   /Game/Sounds/Creative/Gadgets/Radio/Stingers/Halloween/Stinger_Halloween_Scare_02_Cue
   /Game/Sounds/Creative/Gadgets/Radio/Stingers/Halloween/Stinger_Halloween_Secret_01_Cue
   /Game/Sounds/Creative/Gadgets/Radio/Stingers/Halloween/Stinger_Halloween_Start_01_Cue
   /Game/Sounds/Creative/Gadgets/Radio/Stingers/Halloween/Stinger_Halloween_Win_01_Cue
   /Game/Sounds/Creative/Gadgets/Speaker/Halloween/Halloween_Bats_Cue
   /Game/Sounds/Creative/Gadgets/Speaker/Halloween/Halloween_Ghosts_Cue
   /Game/Sounds/Creative/Gadgets/Speaker/Halloween/Halloween_Laughing_Female_Cue
   /Game/Sounds/Creative/Gadgets/Speaker/Halloween/Halloween_Laughing_Male_Cue
   /Game/Sounds/Creative/Gadgets/Speaker/Halloween/Halloween_Singing_Female_Cue
   /Game/Sounds/Creative/Gadgets/Speaker/Halloween/Halloween_Whispers_Cue
   /Game/Sounds/Creative/Gadgets/Radio/Stingers/BlackMonday/Stinger_BlackMonday_Accent_01_Cue
   /Game/Sounds/Creative/Gadgets/Radio/Stingers/BlackMonday/Stinger_BlackMonday_Accent_02_Cue
   /Game/Sounds/Creative/Gadgets/Radio/Stingers/BlackMonday/Stinger_BlackMonday_Accent_03_Cue
   /Game/Sounds/Creative/Gadgets/Radio/Stingers/BlackMonday/Stinger_BlackMonday_Alert_01_Cue
   /Game/Sounds/Creative/Gadgets/Radio/Stingers/BlackMonday/Stinger_BlackMonday_Bonus_01_Cue
   /Game/Sounds/Creative/Gadgets/Radio/Stingers/BlackMonday/Stinger_BlackMonday_Bump_01_Cue
   /Game/Sounds/Creative/Gadgets/Radio/Stingers/BlackMonday/Stinger_BlackMonday_Bump_02_Cue
   /Game/Sounds/Creative/Gadgets/Radio/Stingers/BlackMonday/Stinger_BlackMonday_Discovery_01_Cue
   /Game/Sounds/Creative/Gadgets/Radio/Stingers/BlackMonday/Stinger_BlackMonday_End_01_Cue
   /Game/Sounds/Creative/Gadgets/Radio/Stingers/BlackMonday/Stinger_BlackMonday_Lose_01_Cue
   /Game/Sounds/Creative/Gadgets/Radio/Stingers/BlackMonday/Stinger_BlackMonday_Negative_01_Cue
   /Game/Sounds/Creative/Gadgets/Radio/Stingers/BlackMonday/Stinger_BlackMonday_Positive_01_Cue
   /Game/Sounds/Creative/Gadgets/Radio/Stingers/BlackMonday/Stinger_BlackMonday_Scare_01_Cue
   /Game/Sounds/Creative/Gadgets/Radio/Stingers/BlackMonday/Stinger_BlackMonday_Secret_01_Cue
   /Game/Sounds/Creative/Gadgets/Radio/Stingers/BlackMonday/Stinger_BlackMonday_Start_01_Cue
   /Game/Sounds/Creative/Gadgets/Radio/Stingers/BlackMonday/Stinger_BlackMonday_Surprise_01_Cue
   /Game/Sounds/Creative/Gadgets/Radio/Stingers/BlackMonday/Stinger_BlackMonday_Thrill_01_Cue
   /Game/Sounds/Creative/Gadgets/Radio/Stingers/BlackMonday/Stinger_BlackMonday_Win_01_Cue
  ]
 }]
widgets {
 info {inheritedWidgetCount=0, parentClass=&UserWidget, rootWidgetClass=&Overlay, widgetCount=25}
 widgets [
  Overlay_33 {widgetClassPath=&Overlay, widget=Overlay_33}
  BackingForAnimTest {
   widgetClassPath=&Image
   bIsVariable=true
   parent=Overlay_33
   widget=BackingForAnimTest
   properties {
    visibility=Collapsed
    brush {
     bIsDynamicallyLoaded=false
     drawAs=Image
     imageType=NoImage
     mirroring=NoMirror
     resourceName=None
     resourceObject=null
     tiling=NoTile
     imageSize=<32 32>
     margin=<ltrb 0 0 0 0>
     outlineSettings {
      bUseBrushTransparency=false
      roundingType=HalfHeightRadius
      width=0
      color {colorUseRule=UseColor_Specified, specifiedColor=<rgba 0 0 0 0>}
      cornerRadii=<0 0 0 0>
     }
     tintColor {colorUseRule=UseColor_Specified, specifiedColor=<rgba 0 0 0 1>}
     uVRegion {bIsValid=false, max=<0 0>, min=<0 0>}
    }
   }
   slot {class=OverlaySlot, properties {horizontalAlignment=HAlign_Fill, verticalAlignment=VAlign_Fill}}
  }
  All {
   widgetClassPath=&Overlay
   parent=Overlay_33
   widget=All
   properties {clipping=ClipToBounds, renderTransform {angle=0, scale=<1 1>, shear=<0 0>, translation=<0 -128>}}
   slot {class=OverlaySlot, properties {horizontalAlignment=HAlign_Fill, verticalAlignment=VAlign_Fill}}
  }
  GlowboxLarge {
   widgetClassPath=&SizeBox
   bIsVariable=true
   parent=All
   widget=GlowboxLarge
   properties {bOverride_HeightOverride=true, heightOverride=896, renderTransform {angle=0, scale=<1.5 1>, shear=<0 0>, translation=<0 0>}}
   slot {class=OverlaySlot, properties {horizontalAlignment=HAlign_Fill, verticalAlignment=VAlign_Center}}
  }
  Glow {
   widgetClassPath=&Image
   bIsVariable=true
   parent=GlowboxLarge
   widget=Glow
   properties {
    renderOpacity=0.333
    brush {
     bIsDynamicallyLoaded=false
     drawAs=Image
     imageType=NoImage
     mirroring=NoMirror
     resourceName=None
     resourceObject=&M_UI_RoundEnd_Marker_Winner
     tiling=NoTile
     imageSize=<192 32>
     margin=<ltrb 0 0 0 0>
     outlineSettings {
      bUseBrushTransparency=false
      roundingType=HalfHeightRadius
      width=0
      color {colorUseRule=UseColor_Specified, specifiedColor=<rgba 0 0 0 0>}
      cornerRadii=<0 0 0 0>
     }
     tintColor {colorUseRule=UseColor_Specified, specifiedColor=<rgba 1 1 1 1>}
     uVRegion {bIsValid=false, max=<0 0>, min=<0 0>}
    }
   }
   slot {class=SizeBoxSlot}
  }
  GlowBoxSmall {
   widgetClassPath=&SizeBox
   bIsVariable=true
   parent=All
   widget=GlowBoxSmall
   properties {bOverride_HeightOverride=true, heightOverride=512, renderTransform {angle=0, scale=<1.5 1>, shear=<0 0>, translation=<0 0>}}
   slot {class=OverlaySlot, properties {horizontalAlignment=HAlign_Fill, verticalAlignment=VAlign_Center}}
  }
  SmallGlow {
   widgetClassPath=&Image
   bIsVariable=true
   parent=GlowBoxSmall
   widget=SmallGlow
   properties {
    renderOpacity=0.666
    brush {
     bIsDynamicallyLoaded=false
     drawAs=Image
     imageType=NoImage
     mirroring=NoMirror
     resourceName=None
     resourceObject=&M_UI_RoundEnd_Marker_Winner
     tiling=NoTile
     imageSize=<192 32>
     margin=<ltrb 0 0 0 0>
     outlineSettings {
      bUseBrushTransparency=false
      roundingType=HalfHeightRadius
      width=0
      color {colorUseRule=UseColor_Specified, specifiedColor=<rgba 0 0 0 0>}
      cornerRadii=<0 0 0 0>
     }
     tintColor {colorUseRule=UseColor_Specified, specifiedColor=<rgba 1 1 1 1>}
     uVRegion {bIsValid=false, max=<0 0>, min=<0 0>}
    }
   }
   slot {class=SizeBoxSlot}
  }
  FlareNoise {
   widgetClassPath=&SizeBox
   bIsVariable=true
   parent=All
   widget=FlareNoise
   properties {bOverride_HeightOverride=true, heightOverride=256, renderOpacity=0.9}
   slot {class=OverlaySlot, properties {horizontalAlignment=HAlign_Fill, verticalAlignment=VAlign_Center}}
  }
  Flare {
   widgetClassPath=&Image
   bIsVariable=true
   parent=FlareNoise
   widget=Flare
   properties {
    visibility=SelfHitTestInvisible
    brush {
     bIsDynamicallyLoaded=false
     drawAs=Image
     imageType=NoImage
     mirroring=NoMirror
     resourceName=None
     resourceObject=&M_UI_FlareNoise_Test
     tiling=NoTile
     imageSize=<32 32>
     margin=<ltrb 0 0 0 0>
     outlineSettings {
      bUseBrushTransparency=false
      roundingType=HalfHeightRadius
      width=0
      color {colorUseRule=UseColor_Specified, specifiedColor=<rgba 0 0 0 0>}
      cornerRadii=<0 0 0 0>
     }
     tintColor {colorUseRule=UseColor_Specified, specifiedColor=<rgba 1 1 1 1>}
     uVRegion {bIsValid=false, max=<0 0>, min=<0 0>}
    }
    renderTransform {angle=0, scale=<2 1>, shear=<0 0>, translation=<0 0>}
   }
   slot {class=SizeBoxSlot}
  }
  Logo {
   widgetClassPath=&Image
   bIsVariable=true
   parent=All
   widget=Logo
   properties {
    clipping=ClipToBoundsWithoutIntersecting
    renderOpacity=0.75
    brush {
     bIsDynamicallyLoaded=false
     drawAs=Image
     imageType=NoImage
     mirroring=NoMirror
     resourceName=None
     resourceObject=/Game/Athena/HUD/Art/Materials/M_UI_CreativeLogo
     tiling=NoTile
     imageSize=<480 480>
     margin=<ltrb 0 0 0 0>
     outlineSettings {
      bUseBrushTransparency=false
      roundingType=HalfHeightRadius
      width=0
      color {colorUseRule=UseColor_Specified, specifiedColor=<rgba 0 0 0 0>}
      cornerRadii=<0 0 0 0>
     }
     tintColor {colorUseRule=UseColor_Specified, specifiedColor=<rgba 1 1 1 1>}
     uVRegion {bIsValid=false, max=<0 0>, min=<0 0>}
    }
    renderTransform {angle=-16, scale=<1 1>, shear=<0 0>, translation=<0 0>}
   }
   slot {class=OverlaySlot, properties {horizontalAlignment=HAlign_Center, verticalAlignment=VAlign_Center}}
  }
  BigWinTextFaded {
   widgetClassPath=&CommonTextBlock
   parent=All
   widget=BigWinTextFaded
   properties {
    justification=Center
    renderOpacity=0.25
    text=Team 13
    textTransformPolicy=ToUpper
    visibility=Collapsed
    colorAndOpacity {colorUseRule=UseColor_Specified, specifiedColor=<rgba 0 0 0 0.15>}
    font {
     bForceMonospaced=false
     bMaterialIsStencil=false
     fontMaterial=null
     fontObject=&Burbank
     letterSpacing=0
     monospacedWidth=1
     size=256
     skewAmount=0
     typefaceFontName=Burbank-Black
     outlineSettings {
      bApplyOutlineToDropShadows=false
      bMiteredCorners=false
      bSeparateFillAlpha=true
      outlineMaterial=null
      outlineSize=8
      outlineColor=<rgba 1 0.3333 0 0.1>
     }
    }
    renderTransform {angle=0, scale=<1 1>, shear=<-12 0>, translation=<0 12>}
    shadowOffset=<0 0>
   }
   slot {class=OverlaySlot, properties {horizontalAlignment=HAlign_Fill, verticalAlignment=VAlign_Center}}
  }
  LargeTriangles {
   widgetClassPath=&SizeBox
   bIsVariable=true
   parent=All
   widget=LargeTriangles
   properties {bOverride_HeightOverride=true, bOverride_WidthOverride=true, heightOverride=384, renderOpacity=0.1, widthOverride=1260}
   slot {class=OverlaySlot, properties {horizontalAlignment=HAlign_Center, verticalAlignment=VAlign_Center}}
  }
  HorizontalBox_1 {widgetClassPath=&HorizontalBox, parent=LargeTriangles, widget=HorizontalBox_1, slot {class=SizeBoxSlot}}
  LargeTriangles_Left {
   widgetClassPath=&Image
   bIsVariable=true
   parent=HorizontalBox_1
   widget=LargeTriangles_Left
   properties {
    brush {
     bIsDynamicallyLoaded=false
     drawAs=Image
     imageType=NoImage
     mirroring=NoMirror
     resourceName=None
     resourceObject=&M_FX_TriangleScroll_Scoreboard_Small
     tiling=NoTile
     imageSize=<32 32>
     margin=<ltrb 0 0 0 0>
     outlineSettings {
      bUseBrushTransparency=false
      roundingType=HalfHeightRadius
      width=0
      color {colorUseRule=UseColor_Specified, specifiedColor=<rgba 0 0 0 0>}
      cornerRadii=<0 0 0 0>
     }
     tintColor {colorUseRule=UseColor_Specified, specifiedColor=<rgba 1 1 1 1>}
     uVRegion {bIsValid=false, max=<0 0>, min=<0 0>}
    }
    renderTransform {angle=0, scale=<-1 1>, shear=<0 0>, translation=<0 0>}
   }
   slot {class=HorizontalBoxSlot, properties {size {sizeRule=Fill, value=1}}}
  }
  LargeTriangles_Right {
   widgetClassPath=&Image
   bIsVariable=true
   parent=HorizontalBox_1
   widget=LargeTriangles_Right
   properties {
    brush {
     bIsDynamicallyLoaded=false
     drawAs=Image
     imageType=NoImage
     mirroring=NoMirror
     resourceName=None
     resourceObject=&M_FX_TriangleScroll_Scoreboard_Small
     tiling=NoTile
     imageSize=<32 32>
     margin=<ltrb 0 0 0 0>
     outlineSettings {
      bUseBrushTransparency=false
      roundingType=HalfHeightRadius
      width=0
      color {colorUseRule=UseColor_Specified, specifiedColor=<rgba 0 0 0 0>}
      cornerRadii=<0 0 0 0>
     }
     tintColor {colorUseRule=UseColor_Specified, specifiedColor=<rgba 1 1 1 1>}
     uVRegion {bIsValid=false, max=<0 0>, min=<0 0>}
    }
    renderTransform {angle=0, scale=<1 -1>, shear=<0 0>, translation=<0 0>}
   }
   slot {class=HorizontalBoxSlot, properties {size {sizeRule=Fill, value=1}}}
  }
  SmallTriangles {
   widgetClassPath=&SizeBox
   bIsVariable=true
   parent=All
   widget=SmallTriangles
   properties {bOverride_HeightOverride=true, bOverride_WidthOverride=true, heightOverride=280, widthOverride=960}
   slot {class=OverlaySlot, properties {horizontalAlignment=HAlign_Center, verticalAlignment=VAlign_Center}}
  }
  HorizontalBox_0 {widgetClassPath=&HorizontalBox, parent=SmallTriangles, widget=HorizontalBox_0, slot {class=SizeBoxSlot}}
  SmallTriangles_Left {
   widgetClassPath=&Image
   bIsVariable=true
   parent=HorizontalBox_0
   widget=SmallTriangles_Left
   properties {
    brush {
     bIsDynamicallyLoaded=false
     drawAs=Image
     imageType=NoImage
     mirroring=NoMirror
     resourceName=None
     resourceObject=&M_FX_TriangleScroll_Scoreboard_Small
     tiling=NoTile
     imageSize=<32 32>
     margin=<ltrb 0 0 0 0>
     outlineSettings {
      bUseBrushTransparency=false
      roundingType=HalfHeightRadius
      width=0
      color {colorUseRule=UseColor_Specified, specifiedColor=<rgba 0 0 0 0>}
      cornerRadii=<0 0 0 0>
     }
     tintColor {colorUseRule=UseColor_Specified, specifiedColor=<rgba 1 1 1 1>}
     uVRegion {bIsValid=false, max=<0 0>, min=<0 0>}
    }
    renderTransform {angle=0, scale=<-1 -1>, shear=<0 0>, translation=<0 0>}
   }
   slot {class=HorizontalBoxSlot, properties {size {sizeRule=Fill, value=1}}}
  }
  SmallTriangles_Right {
   widgetClassPath=&Image
   bIsVariable=true
   parent=HorizontalBox_0
   widget=SmallTriangles_Right
   properties {brush {
     bIsDynamicallyLoaded=false
     drawAs=Image
     imageType=NoImage
     mirroring=NoMirror
     resourceName=None
     resourceObject=&M_FX_TriangleScroll_Scoreboard_Small
     tiling=NoTile
     imageSize=<32 32>
     margin=<ltrb 0 0 0 0>
     outlineSettings {
      bUseBrushTransparency=false
      roundingType=HalfHeightRadius
      width=0
      color {colorUseRule=UseColor_Specified, specifiedColor=<rgba 0 0 0 0>}
      cornerRadii=<0 0 0 0>
     }
     tintColor {colorUseRule=UseColor_Specified, specifiedColor=<rgba 1 1 1 1>}
     uVRegion {bIsValid=false, max=<0 0>, min=<0 0>}
    }}
   slot {class=HorizontalBoxSlot, properties {size {sizeRule=Fill, value=1}}}
  }
  WinningText {
   widgetClassPath=&VerticalBox
   parent=All
   widget=WinningText
   slot {class=OverlaySlot, properties {horizontalAlignment=HAlign_Center, verticalAlignment=VAlign_Center}}
  }
  SizeBox_7 {
   widgetClassPath=&SizeBox
   parent=WinningText
   widget=SizeBox_7
   properties {bOverride_WidthOverride=true, widthOverride=1152}
   slot {class=VerticalBoxSlot, properties {horizontalAlignment=HAlign_Center, verticalAlignment=VAlign_Center}}
  }
  ScaleBox_0 {widgetClassPath=&ScaleBox, parent=SizeBox_7, widget=ScaleBox_0, slot {class=SizeBoxSlot}}
  TeamOrPlayer {
   widgetClassPath=&CommonTextBlock
   bIsVariable=true
   parent=ScaleBox_0
   widget=TeamOrPlayer
   properties {
    justification=Center
    text=Victory
    textTransformPolicy=ToUpper
    font {
     bForceMonospaced=false
     bMaterialIsStencil=false
     fontMaterial=null
     fontObject=&Burbank
     letterSpacing=0
     monospacedWidth=1
     size=128
     skewAmount=0
     typefaceFontName=Burbank-Black
     outlineSettings {
      bApplyOutlineToDropShadows=false
      bMiteredCorners=false
      bSeparateFillAlpha=false
      outlineMaterial=null
      outlineSize=0
      outlineColor=<rgba 1 1 1 1>
     }
    }
    renderTransform {angle=0, scale=<1 1>, shear=<-12 0>, translation=<0 16>}
    shadowOffset=<0 0>
   }
   slot {class=ScaleBoxSlot}
  }
  WinDescription {
   widgetClassPath=&CommonTextBlock
   bIsVariable=true
   parent=WinningText
   widget=WinDescription
   properties {
    justification=Center
    text="You win!"
    textTransformPolicy=ToUpper
    font {
     bForceMonospaced=false
     bMaterialIsStencil=false
     fontMaterial=/Game/Items/Traps/Blueprints/Toys/UI_Art_Assets/M_UI_GameWinner_TextStyle
     fontObject=&BurbankSmall
     letterSpacing=0
     monospacedWidth=1
     size=32
     skewAmount=0
     typefaceFontName=Black
     outlineSettings {
      bApplyOutlineToDropShadows=false
      bMiteredCorners=false
      bSeparateFillAlpha=true
      outlineMaterial=null
      outlineSize=4
      outlineColor=<rgba 0 0.2502 1 0.2471>
     }
    }
    renderTransform {angle=0, scale=<1 1>, shear=<-16 0>, translation=<0 -28>}
    shadowOffset=<0 0>
   }
   slot {class=VerticalBoxSlot, properties {horizontalAlignment=HAlign_Center, verticalAlignment=VAlign_Center}}
  }
  Placement {
   widgetClassPath=&CommonTextBlock
   bIsVariable=true
   parent=All
   widget=Placement
   properties {
    justification=Center
    text="You Placed ##2"
    textTransformPolicy=ToUpper
    visibility=Hidden
    font {
     bForceMonospaced=false
     bMaterialIsStencil=false
     fontMaterial=&M_UI_GameWinner_TextStyle
     fontObject=&BurbankSmall
     letterSpacing=0
     monospacedWidth=1
     size=50
     skewAmount=0
     typefaceFontName=Black
     outlineSettings {
      bApplyOutlineToDropShadows=false
      bMiteredCorners=false
      bSeparateFillAlpha=true
      outlineMaterial=null
      outlineSize=4
      outlineColor=<rgba 0 0.2502 1 0.2471>
     }
    }
    renderTransform {angle=0, scale=<1 1>, shear=<-16 0>, translation=<0 -28>}
    shadowOffset=<0 0>
   }
   slot {class=OverlaySlot, properties {horizontalAlignment=HAlign_Center, verticalAlignment=VAlign_Center, padding=<ltrb 0 32 0 0>}}
  }
 ]
}
