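# Scoreboard header row widget for minigames, labeling placement, team, total or wins, player and up to five stat columns.
# On Construct it caches stat text and holder arrays, then UpdateColumnDisplay toggles fixed columns from in-game, team, cumulative, wins-display and win-condition variables.
# It then passes DataText to ProvideData, which fills stat labels, collapses unused holders, and logs an error when data exceeds the fixed slots.
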WidgetBlueprint /Game/Items/Traps/Blueprints/Toys/Scoreboard_UI_ColumnLabels
graphs [
 EventGraph {nodes [
   Event Construct()
   > n1 = Sequence!(execute) -> (then_0, then_1, then_2:Exec)
   Set DataHolders(
    execute=@n1.then_1
    DataHolders:Array of Size Box Object References=Make Array!(
     "[0]":Size Box Object Reference=Get DataHolder1() -> (DataHolder1:Size Box Object Reference)
     "[1]":Size Box Object Reference=Get DataHolder2() -> (DataHolder2:Size Box Object Reference)
     "[2]":Size Box Object Reference=Get DataHolder3() -> (DataHolder3:Size Box Object Reference)
     "[3]":Size Box Object Reference=Get DataHolder4() -> (DataHolder4:Size Box Object Reference)
     "[4]":Size Box Object Reference=Get DataHolder5() -> (DataHolder5:Size Box Object Reference)
    ) -> (Array:Array of Size Box Object References)
   ) -> (Output_Get:Array of Size Box Object References)
   UpdateColumnDisplay(execute=@n1.then_2)
   Set DataBoxes(
    execute=@n1.then_0
    DataBoxes:Array of Text Object References=Make Array(
     "[0]":Text Object Reference=Get DataBox1() -> (DataBox1:Common Text Object Reference)
     "[1]":Text Object Reference=Get DataBox2() -> (DataBox2:Common Text Object Reference)
     "[2]":Text Object Reference=Get DataBox3() -> (DataBox3:Common Text Object Reference)
     "[3]":Text Object Reference=Get DataBox4() -> (DataBox4:Common Text Object Reference)
     "[4]":Text Object Reference=Get DataBox5() -> (DataBox5:Common Text Object Reference)
    ) -> (Array:Array of Text Object References)
   ) -> (Output_Get:Array of Text Object References)
   Event PreConstruct()
  ]}
 ProvideData {nodes [
   n1 = ProvideData!() -> (then, InData:Array of Texts) {is=title K2Node_FunctionEntry}
   n2 = Get DataBoxes() -> (DataBoxes:Array of Text Object References)
   > n3 = Branch(Condition="integer <= integer"!(A:int=Length(TargetArray:Array of Texts=@n1), B:int=Length(TargetArray:Array of Text Object References=@n2)) -> (ReturnValue:bool) {is=title K2Node_CallFunction})
   > n4 = For Each Loop(Array:Array of Text Object References=Get DataBoxes()) -> (Array Element:Text Object Reference)
   n5 = Get InData() -> (InData:Array of Texts)
   > n6 = Branch(Condition=Is Valid Index!(TargetArray:Array of Texts=@n5, IndexToTest:int=@n4.Array Index) -> (ReturnValue:bool))
   > "SetText (Text)"(self=@n4, InText="Get (a copy)"(Array:Array of Texts=@n5, Dimension 1=@n4.Array Index) -> (Output:Text))
   n7 = "Get (a copy)"(
    Array:Array of Size Box Object References=Get DataHolders() -> (DataHolders:Array of Size Box Object References)
    Dimension 1=@n4.Array Index
   ) -> (Output:Size Box Object Reference)
   > SetVisibility(self=@n7, InVisibility=SelfHitTestInvisible)
   SetVisibility(execute=@n6.else, self=@n7, InVisibility=Collapsed)
   Log String!(
    Execute:Exec=@n3.else
    Blueprint Log Level:BlueprintLogLevel Enum=NewEnumerator0
    Log Level Instance:BlueprintLogLevel Enum=NewEnumerator0
    Log String:String=ERROR: Too much data for statically sized data row
    Print to Screen:bool
   ) -> (Then:Exec) {is=title K2Node_MacroInstance}
  ]}
 UpdateColumnDisplay {nodes [
   UpdateColumnDisplay() {is=title K2Node_FunctionEntry}
   > SetVisibility(
    self=Get PlacementSizeBox() -> (PlacementSizeBox:Size Box Object Reference)
    InVisibility=Select(Option 0=SelfHitTestInvisible, Option 1=Collapsed, Index=Get bIngameScoreboard() -> (bIngameScoreboard:bool))
   )
   > SetVisibility(
    self=Get TeamgameSpacer() -> (TeamgameSpacer:Size Box Object Reference)
    InVisibility=Select(Option 0=Collapsed, Option 1=SelfHitTestInvisible, Index=Get bIsTeamGame() -> (bIsTeamGame:bool))
   )
   n1 = NOT Boolean(A=Get bShowCumulativeScoreboard() -> (bShowCumulativeScoreboard:bool))
   n2 = "Equal (Enum)"(
    A:EMinigameWinCondition Enum=Get WinCondition() -> (WinCondition:EMinigameWinCondition Enum=MostRoundWins)
    B:EMinigameWinCondition Enum=MostScoreWins
   )
   n3 = Get bDisplayingWins() -> (bDisplayingWins:bool)
   > SetVisibility(
    self=Get TotalColumn() -> (TotalColumn:Size Box Object Reference)
    InVisibility=Select(Option 0=Collapsed, Option 1=SelfHitTestInvisible, Index=AND Boolean!(A=@n3, B=@n2, C:bool=@n1) -> (ReturnValue))
   )
   n4 = "Equal (Enum)"(A=Get WinCondition() -> (WinCondition=MostRoundWins), B=MostRoundWins)
   n5 = Get bDisplayingWins()
   > SetVisibility(
    self=Get WinsColumn() -> (WinsColumn:Size Box Object Reference)
    InVisibility=Select(Option 0=Collapsed, Option 1=SelfHitTestInvisible, Index=AND Boolean!(A=@n5, B=@n4) -> (ReturnValue))
   )
   > ProvideData!(execute, self, InData:Array of Texts=Get DataText() -> (DataText:Array of Texts)) -> (then) {is=title K2Node_CallFunction}
  ]}
]
variables [
 bDisplayingWins {type=boolean}
 bIngameScoreboard {type=boolean}
 bIsTeamGame {type=boolean}
 bShowCumulativeScoreboard {type=boolean}
 DataBoxes {type="array<TextBlock>", default []}
 DataHolders {type="array<SizeBox>", default []}
 DataText {type="array<string>", default []}
 WinCondition {type=EMinigameWinCondition, default=MostRoundWins}
]
widgets {
 info {inheritedWidgetCount=0, parentClass=&UserWidget, rootWidgetClass=&VerticalBox, widgetCount=53}
 widgets [
  VerticalBox_0 {widgetClassPath=&VerticalBox, widget=VerticalBox_0}
  DataHorizontalBox {
   widgetClassPath=&HorizontalBox
   bIsVariable=true
   parent=VerticalBox_0
   widget=DataHorizontalBox
   slot {class=VerticalBoxSlot, properties {size {sizeRule=Fill, value=1}}}
  }
  PlacementSizeBox {
   widgetClassPath=&SizeBox
   bIsVariable=true
   parent=DataHorizontalBox
   widget=PlacementSizeBox
   properties {bOverride_WidthOverride=true, visibility=Collapsed, widthOverride=96}
   slot {class=HorizontalBoxSlot}
  }
  HorizontalBox_7 {widgetClassPath=&HorizontalBox, parent=PlacementSizeBox, widget=HorizontalBox_7, slot {class=SizeBoxSlot}}
  Border_0 {
   widgetClassPath=&Border
   parent=HorizontalBox_7
   widget=Border_0
   properties {verticalAlignment=VAlign_Center, brushColor=<rgba 0.2159 1 1 0.25>, padding=<ltrb 8 0 8 0>}
   slot {class=HorizontalBoxSlot, properties {size {sizeRule=Fill, value=1}}}
  }
  CommonTextBlock_0 {
   widgetClassPath=&CommonTextBlock
   parent=Border_0
   widget=CommonTextBlock_0
   properties {
    autoWrapText=true
    text="##"
    textTransformPolicy=ToUpper
    colorAndOpacity {colorUseRule=UseColor_Specified, specifiedColor=<rgba 0.2159 1 1 0.5>}
    font {
     bForceMonospaced=false
     bMaterialIsStencil=false
     fontMaterial=null
     fontObject=&BurbankSmall
     letterSpacing=0
     monospacedWidth=1
     size=15
     skewAmount=0
     typefaceFontName=Black
     outlineSettings {
      bApplyOutlineToDropShadows=false
      bMiteredCorners=false
      bSeparateFillAlpha=false
      outlineMaterial=null
      outlineSize=0
      outlineColor=<rgba 0 0 0 1>
     }
    }
    shadowOffset=<0 0>
   }
   slot {class=BorderSlot, properties {verticalAlignment=VAlign_Center, padding=<ltrb 8 0 8 0>}}
  }
  Spacer_0 {widgetClassPath=&Spacer, parent=HorizontalBox_7, widget=Spacer_0, properties {size=<8 1>}, slot {class=HorizontalBoxSlot}}
  TeamgameSpacer {
   widgetClassPath=&SizeBox
   bIsVariable=true
   parent=DataHorizontalBox
   widget=TeamgameSpacer
   properties {bOverride_MinDesiredWidth=true, bOverride_WidthOverride=true, heightOverride=24, minDesiredWidth=24, widthOverride=256}
   slot {class=HorizontalBoxSlot}
  }
  HorizontalBox_0 {widgetClassPath=&HorizontalBox, parent=TeamgameSpacer, widget=HorizontalBox_0, slot {class=SizeBoxSlot}}
  Border_136 {
   widgetClassPath=&Border
   parent=HorizontalBox_0
   widget=Border_136
   properties {verticalAlignment=VAlign_Center, brushColor=<rgba 0.2159 1 1 0.25>, padding=<ltrb 8 0 8 0>}
   slot {class=HorizontalBoxSlot, properties {size {sizeRule=Fill, value=1}}}
  }
  CommonTextBlock_183 {
   widgetClassPath=&CommonTextBlock
   parent=Border_136
   widget=CommonTextBlock_183
   properties {
    autoWrapText=true
    text=Team
    textTransformPolicy=ToUpper
    colorAndOpacity {colorUseRule=UseColor_Specified, specifiedColor=<rgba 0.2159 1 1 0.5>}
    font {
     bForceMonospaced=false
     bMaterialIsStencil=false
     fontMaterial=null
     fontObject=&BurbankSmall
     letterSpacing=0
     monospacedWidth=1
     size=15
     skewAmount=0
     typefaceFontName=Black
     outlineSettings {
      bApplyOutlineToDropShadows=false
      bMiteredCorners=false
      bSeparateFillAlpha=false
      outlineMaterial=null
      outlineSize=0
      outlineColor=<rgba 0 0 0 1>
     }
    }
    shadowOffset=<0 0>
   }
   slot {class=BorderSlot, properties {verticalAlignment=VAlign_Center, padding=<ltrb 8 0 8 0>}}
  }
  Spacer_117 {widgetClassPath=&Spacer, parent=HorizontalBox_0, widget=Spacer_117, properties {size=<8 1>}, slot {class=HorizontalBoxSlot}}
  TotalColumn {
   widgetClassPath=&SizeBox
   bIsVariable=true
   parent=DataHorizontalBox
   widget=TotalColumn
   properties {bOverride_MinDesiredWidth=true, bOverride_WidthOverride=true, minDesiredWidth=24, widthOverride=128}
   slot {class=HorizontalBoxSlot}
  }
  HorizontalBox_82 {widgetClassPath=&HorizontalBox, parent=TotalColumn, widget=HorizontalBox_82, slot {class=SizeBoxSlot}}
  Border_91 {
   widgetClassPath=&Border
   parent=HorizontalBox_82
   widget=Border_91
   properties {verticalAlignment=VAlign_Center, brushColor=<rgba 0.2159 1 1 0.25>, padding=<ltrb 8 0 8 0>}
   slot {class=HorizontalBoxSlot, properties {size {sizeRule=Fill, value=1}}}
  }
  TextBlock_84 {
   widgetClassPath=&TextBlock
   parent=Border_91
   widget=TextBlock_84
   properties {
    autoWrapText=true
    text=TOTAL
    colorAndOpacity {colorUseRule=UseColor_Specified, specifiedColor=<rgba 0.2159 1 1 0.5>}
    font {
     bForceMonospaced=false
     bMaterialIsStencil=false
     fontMaterial=null
     fontObject=&BurbankSmall
     letterSpacing=0
     monospacedWidth=1
     size=15
     skewAmount=0
     typefaceFontName=Black
     outlineSettings {
      bApplyOutlineToDropShadows=false
      bMiteredCorners=false
      bSeparateFillAlpha=false
      outlineMaterial=null
      outlineSize=0
      outlineColor=<rgba 0 0 0 1>
     }
    }
   }
   slot {class=BorderSlot, properties {verticalAlignment=VAlign_Center, padding=<ltrb 8 0 8 0>}}
  }
  Spacer_60 {
   widgetClassPath=&Spacer
   parent=HorizontalBox_82
   widget=Spacer_60
   properties {size=<8 1>}
   slot {class=HorizontalBoxSlot, properties {verticalAlignment=VAlign_Top}}
  }
  WinsColumn {
   widgetClassPath=&SizeBox
   bIsVariable=true
   parent=DataHorizontalBox
   widget=WinsColumn
   properties {bOverride_MinDesiredWidth=true, bOverride_WidthOverride=true, heightOverride=24, minDesiredWidth=24, widthOverride=128}
   slot {class=HorizontalBoxSlot}
  }
  HorizontalBox_1 {widgetClassPath=&HorizontalBox, parent=WinsColumn, widget=HorizontalBox_1, slot {class=SizeBoxSlot}}
  Border_89 {
   widgetClassPath=&Border
   parent=HorizontalBox_1
   widget=Border_89
   properties {verticalAlignment=VAlign_Center, brushColor=<rgba 0.2159 1 1 0.25>, padding=<ltrb 8 0 8 0>}
   slot {class=HorizontalBoxSlot, properties {size {sizeRule=Fill, value=1}}}
  }
  WinsText {
   widgetClassPath=&TextBlock
   bIsVariable=true
   parent=Border_89
   widget=WinsText
   properties {
    autoWrapText=true
    text=WINS
    colorAndOpacity {colorUseRule=UseColor_Specified, specifiedColor=<rgba 0.2159 1 1 0.5>}
    font {
     bForceMonospaced=false
     bMaterialIsStencil=false
     fontMaterial=null
     fontObject=&BurbankSmall
     letterSpacing=0
     monospacedWidth=1
     size=15
     skewAmount=0
     typefaceFontName=Black
     outlineSettings {
      bApplyOutlineToDropShadows=false
      bMiteredCorners=false
      bSeparateFillAlpha=false
      outlineMaterial=null
      outlineSize=0
      outlineColor=<rgba 0 0 0 1>
     }
    }
   }
   slot {class=BorderSlot, properties {verticalAlignment=VAlign_Center, padding=<ltrb 8 0 8 0>}}
  }
  Spacer_65 {
   widgetClassPath=&Spacer
   parent=HorizontalBox_1
   widget=Spacer_65
   properties {size=<8 1>}
   slot {class=HorizontalBoxSlot, properties {verticalAlignment=VAlign_Top}}
  }
  PlayerNameBox {
   widgetClassPath=&SizeBox
   parent=DataHorizontalBox
   widget=PlayerNameBox
   properties {bOverride_MinDesiredHeight=true, bOverride_MinDesiredWidth=true, heightOverride=24, minDesiredHeight=24, minDesiredWidth=256}
   slot {class=HorizontalBoxSlot}
  }
  HorizontalBox_2 {widgetClassPath=&HorizontalBox, parent=PlayerNameBox, widget=HorizontalBox_2, slot {class=SizeBoxSlot}}
  Border_87 {
   widgetClassPath=&Border
   parent=HorizontalBox_2
   widget=Border_87
   properties {verticalAlignment=VAlign_Center, brushColor=<rgba 0.2159 1 1 0.25>, padding=<ltrb 8 0 8 0>}
   slot {class=HorizontalBoxSlot, properties {size {sizeRule=Fill, value=1}}}
  }
  CommonTextBlock_120 {
   widgetClassPath=&CommonTextBlock
   parent=Border_87
   widget=CommonTextBlock_120
   properties {
    text=Player
    textTransformPolicy=ToUpper
    colorAndOpacity {colorUseRule=UseColor_Specified, specifiedColor=<rgba 0.2159 1 1 0.5>}
    font {
     bForceMonospaced=false
     bMaterialIsStencil=false
     fontMaterial=null
     fontObject=&BurbankSmall
     letterSpacing=0
     monospacedWidth=1
     size=15
     skewAmount=0
     typefaceFontName=Black
     outlineSettings {
      bApplyOutlineToDropShadows=false
      bMiteredCorners=false
      bSeparateFillAlpha=false
      outlineMaterial=null
      outlineSize=0
      outlineColor=<rgba 0 0 0 1>
     }
    }
    shadowOffset=<0 0>
   }
   slot {class=BorderSlot, properties {verticalAlignment=VAlign_Center, padding=<ltrb 8 0 8 0>}}
  }
  Spacer_212 {widgetClassPath=&Spacer, parent=HorizontalBox_2, widget=Spacer_212, properties {size=<8 1>}, slot {class=HorizontalBoxSlot}}
  StatsDataBox {
   widgetClassPath=&SizeBox
   parent=DataHorizontalBox
   widget=StatsDataBox
   properties {widthOverride=1000}
   slot {class=HorizontalBoxSlot, properties {size {sizeRule=Fill, value=1}}}
  }
  HorizontalBox_8 {widgetClassPath=&HorizontalBox, parent=StatsDataBox, widget=HorizontalBox_8, slot {class=SizeBoxSlot}}
  DataHolder1 {
   widgetClassPath=&SizeBox
   bIsVariable=true
   parent=HorizontalBox_8
   widget=DataHolder1
   properties {bOverride_MinDesiredWidth=true, heightOverride=24, minDesiredWidth=24}
   slot {class=HorizontalBoxSlot, properties {size {sizeRule=Fill, value=1}}}
  }
  HorizontalBox_3 {widgetClassPath=&HorizontalBox, parent=DataHolder1, widget=HorizontalBox_3, slot {class=SizeBoxSlot}}
  Border_1 {
   widgetClassPath=&Border
   parent=HorizontalBox_3
   widget=Border_1
   properties {verticalAlignment=VAlign_Center, brushColor=<rgba 0.2159 1 1 0.25>, padding=<ltrb 8 0 8 0>}
   slot {class=HorizontalBoxSlot, properties {size {sizeRule=Fill, value=1}}}
  }
  DataBox1 {
   widgetClassPath=&CommonTextBlock
   bIsVariable=true
   parent=Border_1
   widget=DataBox1
   properties {
    autoWrapText=true
    scrollStyle=&TextScrollStyle-Base_C
    text=STAT NAME 1
    textTransformPolicy=ToUpper
    colorAndOpacity {colorUseRule=UseColor_Specified, specifiedColor=<rgba 0.2159 1 1 0.5>}
    font {
     bForceMonospaced=false
     bMaterialIsStencil=false
     fontMaterial=null
     fontObject=&BurbankSmall
     letterSpacing=0
     monospacedWidth=1
     size=15
     skewAmount=0
     typefaceFontName=Black
     outlineSettings {
      bApplyOutlineToDropShadows=false
      bMiteredCorners=false
      bSeparateFillAlpha=false
      outlineMaterial=null
      outlineSize=0
      outlineColor=<rgba 0 0 0 1>
     }
    }
    shadowOffset=<0 0>
   }
   slot {class=BorderSlot, properties {verticalAlignment=VAlign_Center, padding=<ltrb 8 0 8 0>}}
  }
  Spacer_318 {widgetClassPath=&Spacer, parent=HorizontalBox_3, widget=Spacer_318, properties {size=<8 1>}, slot {class=HorizontalBoxSlot}}
  DataHolder2 {
   widgetClassPath=&SizeBox
   bIsVariable=true
   parent=HorizontalBox_8
   widget=DataHolder2
   properties {bOverride_MinDesiredWidth=true, heightOverride=24, minDesiredWidth=24}
   slot {class=HorizontalBoxSlot, properties {size {sizeRule=Fill, value=1}}}
  }
  HorizontalBox_5 {widgetClassPath=&HorizontalBox, parent=DataHolder2, widget=HorizontalBox_5, slot {class=SizeBoxSlot}}
  Border_2 {
   widgetClassPath=&Border
   parent=HorizontalBox_5
   widget=Border_2
   properties {verticalAlignment=VAlign_Center, brushColor=<rgba 0.2159 1 1 0.25>, padding=<ltrb 8 0 8 0>}
   slot {class=HorizontalBoxSlot, properties {size {sizeRule=Fill, value=1}}}
  }
  DataBox2 {
   widgetClassPath=&CommonTextBlock
   bIsVariable=true
   parent=Border_2
   widget=DataBox2
   properties {
    autoWrapText=true
    scrollStyle=&TextScrollStyle-Base_C
    text=STAT NAME 2
    textTransformPolicy=ToUpper
    colorAndOpacity {colorUseRule=UseColor_Specified, specifiedColor=<rgba 0.2159 1 1 0.5>}
    font {
     bForceMonospaced=false
     bMaterialIsStencil=false
     fontMaterial=null
     fontObject=&BurbankSmall
     letterSpacing=0
     monospacedWidth=1
     size=15
     skewAmount=0
     typefaceFontName=Black
     outlineSettings {
      bApplyOutlineToDropShadows=false
      bMiteredCorners=false
      bSeparateFillAlpha=false
      outlineMaterial=null
      outlineSize=0
      outlineColor=<rgba 0 0 0 1>
     }
    }
    shadowOffset=<0 0>
   }
   slot {class=BorderSlot, properties {verticalAlignment=VAlign_Center, padding=<ltrb 8 0 8 0>}}
  }
  Spacer_552 {widgetClassPath=&Spacer, parent=HorizontalBox_5, widget=Spacer_552, properties {size=<8 1>}, slot {class=HorizontalBoxSlot}}
  DataHolder3 {
   widgetClassPath=&SizeBox
   bIsVariable=true
   parent=HorizontalBox_8
   widget=DataHolder3
   properties {bOverride_MinDesiredWidth=true, heightOverride=24, minDesiredWidth=24}
   slot {class=HorizontalBoxSlot, properties {size {sizeRule=Fill, value=1}}}
  }
  HorizontalBox_4 {widgetClassPath=&HorizontalBox, parent=DataHolder3, widget=HorizontalBox_4, slot {class=SizeBoxSlot}}
  Border_3 {
   widgetClassPath=&Border
   parent=HorizontalBox_4
   widget=Border_3
   properties {verticalAlignment=VAlign_Center, brushColor=<rgba 0.2159 1 1 0.25>, padding=<ltrb 8 0 8 0>}
   slot {class=HorizontalBoxSlot, properties {size {sizeRule=Fill, value=1}}}
  }
  DataBox3 {
   widgetClassPath=&CommonTextBlock
   bIsVariable=true
   parent=Border_3
   widget=DataBox3
   properties {
    autoWrapText=true
    scrollStyle=&TextScrollStyle-Base_C
    text=STAT NAME 3
    textTransformPolicy=ToUpper
    colorAndOpacity {colorUseRule=UseColor_Specified, specifiedColor=<rgba 0.2159 1 1 0.5>}
    font {
     bForceMonospaced=false
     bMaterialIsStencil=false
     fontMaterial=null
     fontObject=&BurbankSmall
     letterSpacing=0
     monospacedWidth=1
     size=15
     skewAmount=0
     typefaceFontName=Black
     outlineSettings {
      bApplyOutlineToDropShadows=false
      bMiteredCorners=false
      bSeparateFillAlpha=false
      outlineMaterial=null
      outlineSize=0
      outlineColor=<rgba 0 0 0 1>
     }
    }
    shadowOffset=<0 0>
   }
   slot {class=BorderSlot, properties {verticalAlignment=VAlign_Center, padding=<ltrb 8 0 8 0>}}
  }
  Spacer_422 {widgetClassPath=&Spacer, parent=HorizontalBox_4, widget=Spacer_422, properties {size=<8 1>}, slot {class=HorizontalBoxSlot}}
  DataHolder4 {
   widgetClassPath=&SizeBox
   bIsVariable=true
   parent=HorizontalBox_8
   widget=DataHolder4
   properties {bOverride_MinDesiredWidth=true, heightOverride=24, minDesiredWidth=24}
   slot {class=HorizontalBoxSlot, properties {size {sizeRule=Fill, value=1}}}
  }
  HorizontalBox_6 {widgetClassPath=&HorizontalBox, parent=DataHolder4, widget=HorizontalBox_6, slot {class=SizeBoxSlot}}
  Border_4 {
   widgetClassPath=&Border
   parent=HorizontalBox_6
   widget=Border_4
   properties {verticalAlignment=VAlign_Center, brushColor=<rgba 0.2159 1 1 0.25>, padding=<ltrb 8 0 8 0>}
   slot {class=HorizontalBoxSlot, properties {size {sizeRule=Fill, value=1}}}
  }
  DataBox4 {
   widgetClassPath=&CommonTextBlock
   bIsVariable=true
   parent=Border_4
   widget=DataBox4
   properties {
    autoWrapText=true
    scrollStyle=&TextScrollStyle-Base_C
    text=STAT NAME 4
    textTransformPolicy=ToUpper
    colorAndOpacity {colorUseRule=UseColor_Specified, specifiedColor=<rgba 0.2159 1 1 0.5>}
    font {
     bForceMonospaced=false
     bMaterialIsStencil=false
     fontMaterial=null
     fontObject=&BurbankSmall
     letterSpacing=0
     monospacedWidth=1
     size=15
     skewAmount=0
     typefaceFontName=Black
     outlineSettings {
      bApplyOutlineToDropShadows=false
      bMiteredCorners=false
      bSeparateFillAlpha=false
      outlineMaterial=null
      outlineSize=0
      outlineColor=<rgba 0 0 0 1>
     }
    }
    shadowOffset=<0 0>
   }
   slot {class=BorderSlot, properties {verticalAlignment=VAlign_Center, padding=<ltrb 8 0 8 0>}}
  }
  Spacer_663 {widgetClassPath=&Spacer, parent=HorizontalBox_6, widget=Spacer_663, properties {size=<8 1>}, slot {class=HorizontalBoxSlot}}
  DataHolder5 {
   widgetClassPath=&SizeBox
   bIsVariable=true
   parent=HorizontalBox_8
   widget=DataHolder5
   properties {bOverride_MinDesiredWidth=true, heightOverride=24, minDesiredWidth=24}
   slot {class=HorizontalBoxSlot, properties {size {sizeRule=Fill, value=1}}}
  }
  Border_5 {
   widgetClassPath=&Border
   parent=DataHolder5
   widget=Border_5
   properties {verticalAlignment=VAlign_Center, brushColor=<rgba 0.2159 1 1 0.25>, padding=<ltrb 8 0 8 0>}
   slot {class=SizeBoxSlot}
  }
  DataBox5 {
   widgetClassPath=&CommonTextBlock
   bIsVariable=true
   parent=Border_5
   widget=DataBox5
   properties {
    autoWrapText=true
    scrollStyle=&TextScrollStyle-Base_C
    text=STAT NAME 5
    textTransformPolicy=ToUpper
    colorAndOpacity {colorUseRule=UseColor_Specified, specifiedColor=<rgba 0.2159 1 1 0.5>}
    font {
     bForceMonospaced=false
     bMaterialIsStencil=false
     fontMaterial=null
     fontObject=&BurbankSmall
     letterSpacing=0
     monospacedWidth=1
     size=15
     skewAmount=0
     typefaceFontName=Black
     outlineSettings {
      bApplyOutlineToDropShadows=false
      bMiteredCorners=false
      bSeparateFillAlpha=false
      outlineMaterial=null
      outlineSize=0
      outlineColor=<rgba 0 0 0 1>
     }
    }
    shadowOffset=<0 0>
   }
   slot {class=BorderSlot, properties {verticalAlignment=VAlign_Center, padding=<ltrb 8 0 8 0>}}
  }
  Spacer_696 {
   widgetClassPath=&Spacer
   parent=VerticalBox_0
   widget=Spacer_696
   properties {size=<1 12>}
   slot {class=VerticalBoxSlot, properties {verticalAlignment=VAlign_Top}}
  }
 ]
}
