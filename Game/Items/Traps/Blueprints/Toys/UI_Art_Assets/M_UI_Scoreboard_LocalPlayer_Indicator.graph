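# Translucent UI material that, by its name, marks the local player on a scoreboard with a colored border.
# Abs-sine of scaled UVs is split by axis, inverted and boosted into edge bands, which are summed and clamped into a procedural frame mask without textures.
# That mask multiplies a color from MF_CreativeColors and is also passed on alone through a reroute.
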Material /Game/Items/Traps/Blueprints/Toys/UI_Art_Assets/M_UI_Scoreboard_LocalPlayer_Indicator
graph {
 properties {blendMode=BLEND_Translucent, materialDomain=MD_UI, pixelDepthOffsetMode=PDOM_Legacy, usedShadingModels=Default Lit}
 nodes [
  n1 = TextureCoordinate()
  n2 = Abs(None=Sine(None=Multiply(A=@n1) {constB=0.5}))
  n3 = ComponentMask(None=@n2) {r=true}
  n4 = OneMinus(None=Clamp(None=Multiply(A=@n3) {constB=4}))
  n5 = ComponentMask(None=@n2) {g=true}
  n6 = OneMinus(None=Clamp(None=Multiply(A=@n5) {constB=4}))
  n7 = Clamp(None=Add(A=Multiply(A=@n4) {constB=8}, B=Multiply(A=@n6) {constB=8}))
  Multiply(A=MaterialFunctionCall!(Mask=Constant() {r=1}, Index) {materialFunction=/Game/Creative/Materials/MF_CreativeColors}.Result, B=@n7)
  Reroute(None=@n7)
 ]
}
