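# Minigame scoreboard row widget showing one player's name, stats, win count or total, and medal placement.
# UpdatePlayerData stores the row struct and runs UpdateRowVisuals, which dims departed players and toggles the local-player callout.
# FillStatData fills five fixed cells, highlighting flagged values, collapsing unused ones and logging overflow.
# Team games tint via material parameter; others use fixed blue and update medals; WinnerCalloutAnim plays a material-animated sweep.
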WidgetBlueprint /Game/Items/Traps/Blueprints/Toys/Scoreboard_UI_Row
animations [AnimWinnerCallout {
  binding_count=1
  length_seconds=0.75
  object=&self:AnimWinnerCallout
  sequence {bindings [WinnerCallout {
     id=#1
     possessed_class=Image
     tracks [{
       type=MovieSceneWidgetMaterialTrack
       sections [{
         type=MovieSceneParameterSection
         end_frame=15
         start_frame=0
         channels [
          Pan {type=MovieSceneScriptingFloatChannel, keys [[0,1.25],[14.9997,-0.25]]}
          Tightness {type=MovieSceneScriptingFloatChannel, keys [[0,4],[5,1]]}
         ]
        }]
      }]
    }]}
 }]
graphs [
 EventGraph {nodes [
   Event Construct()
   > n1 = Sequence!(execute) -> (then_0, then_1, then_2:Exec)
   > Set DataBoxes(DataBoxes:Array of Text Object References=Make Array!(
     "[0]":Text Object Reference=Get DataBox1() -> (DataBox1:Text Object Reference)
     "[1]":Text Object Reference=Get DataBox2() -> (DataBox2:Text Object Reference)
     "[2]":Text Object Reference=Get DataBox3() -> (DataBox3:Text Object Reference)
     "[3]":Text Object Reference=Get DataBox4() -> (DataBox4:Text Object Reference)
     "[4]":Text Object Reference=Get DataBox5() -> (DataBox5:Text Object Reference)
    ) -> (Array:Array of Text Object References)) -> (Output_Get:Array of Text Object References)
   Set DataHolders(
    execute=@n1.then_1
    DataHolders:Array of Size Box Object References=Make Array(
     "[0]":Size Box Object Reference=Get DataHolder1() -> (DataHolder1:Size Box Object Reference)
     "[1]":Size Box Object Reference=Get DataHolder2() -> (DataHolder2:Size Box Object Reference)
     "[2]":Size Box Object Reference=Get DataHolder3() -> (DataHolder3:Size Box Object Reference)
     "[3]":Size Box Object Reference=Get DataHolder4() -> (DataHolder4:Size Box Object Reference)
     "[4]":Size Box Object Reference=Get DataHolder5() -> (DataHolder5:Size Box Object Reference)
    ) -> (Array:Array of Size Box Object References)
   ) -> (Output_Get:Array of Size Box Object References)
   > Set DataHolderHBOXs(DataHolderHBOXs:Array of Horizontal Box Object References=Make Array(
     "[0]":Horizontal Box Object Reference=Get DataHolderHBOX_1() -> (DataHolderHBOX_1:Horizontal Box Object Reference)
     "[1]":Horizontal Box Object Reference=Get DataHolderHBOX_2() -> (DataHolderHBOX_2:Horizontal Box Object Reference)
     "[2]":Horizontal Box Object Reference=Get DataHolderHBOX_3() -> (DataHolderHBOX_3:Horizontal Box Object Reference)
     "[3]":Horizontal Box Object Reference=Get DataHolderHBOX_4() -> (DataHolderHBOX_4:Horizontal Box Object Reference)
     "[4]":Horizontal Box Object Reference=Get DataHolderHBOX_5() -> (DataHolderHBOX_5:Horizontal Box Object Reference)
    ) -> (Array:Array of Horizontal Box Object References)) -> (Output_Get:Array of Horizontal Box Object References)
   UpdateRowVisuals(execute=@n1.then_2)
   Event PreConstruct()
  ]}
 FillStatData {nodes [
   n1 = FillStatData!() -> (then, StatValues:Array of Minigame Scoreboard Value Structures) {is=title K2Node_FunctionEntry}
   n2 = Get DataBoxes() -> (DataBoxes:Array of Text Object References)
   > n3 = Branch(Condition="integer <= integer"!(A:int=Length(TargetArray:Array of Minigame Scoreboard Value Structures=@n1), B:int=Length(TargetArray:Array of Text Object References=@n2)) -> (ReturnValue:bool) {is=title K2Node_CallFunction})
   > n4 = For Each Loop(Array:Array of Text Object References=Get DataBoxes()) -> (Array Element:Text Object Reference)
   n5 = Get StatValues() -> (StatValues:Array of Minigame Scoreboard Value Structures)
   > n6 = Branch(Condition=Is Valid Index!(TargetArray:Array of Minigame Scoreboard Value Structures=@n5, IndexToTest:int=@n4.Array Index) -> (ReturnValue:bool))
   n7 = "Get (a ref)"(Array:Array of Minigame Scoreboard Value Structures=@n5, Dimension 1=@n4.Array Index) -> (Output:"Minigame Scoreboard Value Structure (by ref)")
   > "SetText (Text)"(self=@n4, InText=Get Stat Value Text!(StatValue:Minigame Scoreboard Value Structure=@n7) -> (StatText:Text) {is=title K2Node_MacroInstance})
   n8 = "Get (a copy)"(
    Array:Array of Horizontal Box Object References=Get DataHolderHBOXs() -> (DataHolderHBOXs:Array of Horizontal Box Object References)
    Dimension 1=@n4.Array Index
   ) -> (Output:Horizontal Box Object Reference)
   > SetVisibility(self=@n8, InVisibility=SelfHitTestInvisible)
   > SetColorAndOpacity(
    self=@n4
    InColorAndOpacity=Select(
     Option 0:Slate Color Structure=Get NormalTextColor() -> (NormalTextColor:Slate Color Structure)
     Option 1:Slate Color Structure=Get HighlightColor() -> (HighlightColor:Slate Color Structure)
     Index=Is Stat Highlighted!(StatValue:Minigame Scoreboard Value Structure=@n7) -> (IsHighlighted:bool=~) {is=title K2Node_MacroInstance}
    ) -> (ReturnValue:Slate Color Structure)
   )
   SetVisibility(execute=@n6.else, self=@n8, InVisibility=Collapsed)
   Log String!(
    Execute:Exec=@n3.else
    Blueprint Log Level:BlueprintLogLevel Enum=NewEnumerator0
    Log Level Instance:BlueprintLogLevel Enum=NewEnumerator0
    Log String:String=ERROR: Too much data for statically sized data row
    Print to Screen:bool
   ) -> (Then:Exec) {is=title K2Node_MacroInstance}
   "Get (a copy)"(
    Array:Array of Size Box Object References=Get DataHolders() -> (DataHolders:Array of Size Box Object References)
    Dimension 1=@n4.Array Index
   ) -> (Output:Size Box Object Reference)
  ]}
 GetStatValueText {nodes [
   n1 = Break Minigame Scoreboard Value!(MinigameScoreboardValue:"Minigame Scoreboard Value Structure (by ref)"=Inputs!() -> (StatValue:Minigame Scoreboard Value Structure)) -> (StatFilter:Fort Minigame Stat Filter Object Reference, Value:int, bHighlight:bool=~) {is=title K2Node_BreakStruct}
   Outputs!(StatText:Text=FormatStat!(InStat:Fort Minigame Stat Filter Object Reference=@n1, InValue:int=@n1.Value) -> (ReturnValue:Text))
  ]}
 IsStatHighlighted {nodes [Outputs!(IsHighlighted:bool=Break Minigame Scoreboard Value(MinigameScoreboardValue=Inputs()) -> (bHighlight=~).bHighlight)]}
 UpdateMedalStanding {nodes [
   UpdateMedalStanding() {is=title K2Node_FunctionEntry}
   n1 = Get bIngameScorerow() -> (bIngameScorerow:bool)
   n2 = Get bIsTeamGame() -> (bIsTeamGame:bool)
   > SetVisibility(
    self=Get PlacementIndicatorBox() -> (PlacementIndicatorBox:Size Box Object Reference)
    InVisibility=Select(Option 0:ESlateVisibility Enum=SelfHitTestInvisible, Option 1:ESlateVisibility Enum=Collapsed, Index=OR Boolean(A=@n1, B=@n2)) -> (ReturnValue:ESlateVisibility Enum)
   )
   n3 = Get PlacementIndicator_Switcher() -> (PlacementIndicator_Switcher:Common Widget Switcher Legacy Object Reference)
   n4 = GetNumWidgets!(self:Widget Switcher Object Reference=@n3) -> (ReturnValue:int=0)
   n5 = Get MedalStanding() -> (MedalStanding:int=0)
   n6 = int - int(A=@n4, B=1) {is=title K2Node_CallFunction}
   > SetActiveWidgetIndex(
    self=@n3
    Index=Select(
     Option 0:int=@n4
     Option 1:int="Clamp (Integer)"(Value=@n5, Min=0, Max=@n6) -> (ReturnValue=0)
     Index="integer >= integer"!(A:int=@n5, B:int=0) -> (ReturnValue:bool) {is=title K2Node_CallFunction}
    ) -> (ReturnValue:int)
   )
  ]}
 UpdatePlayerData {nodes [
   n1 = UpdatePlayerData!() -> (then, PlayerBucketData:Minigame Scoreboard Player Row Structure) {is=title K2Node_FunctionEntry}
   > Set PlayerBucketData(PlayerBucketData:Minigame Scoreboard Player Row Structure=@n1) -> (Output_Get:Minigame Scoreboard Player Row Structure)
   > UpdateRowVisuals()
  ]}
 UpdateRowVisuals {nodes [
   UpdateRowVisuals() {is=title K2Node_FunctionEntry}
   n1 = Break Minigame Scoreboard Player Row!(MinigameScoreboardPlayerRow:"Minigame Scoreboard Player Row Structure (by ref)"=Get PlayerBucketData() -> (PlayerBucketData:Minigame Scoreboard Player Row Structure)) -> (
    PlayerState:Fort Player State Object Reference
    bHasLeftGame:bool=~
    bHideFromDisplay:bool=~
    Name:Text
    TeamColorIndex:int
    Values:Array of Minigame Scoreboard Value Structures
    WinCount:int
    bHighlight:bool=~
   ) {is=title K2Node_BreakStruct}
   > "SetText (Text)"(self=Get NameBox() -> (NameBox:Common Text Object Reference), InText=@n1.Name)
   > n2 = Set ScoreboardValues(ScoreboardValues:Array of Minigame Scoreboard Value Structures=@n1.Values) -> (Output_Get:Array of Minigame Scoreboard Value Structures)
   > FillStatData!(execute, self, StatValues:Array of Minigame Scoreboard Value Structures=@n2) -> (then) {is=title K2Node_CallFunction}
   > SetRenderOpacity(self=Get StatsBorder() -> (StatsBorder:Border Object Reference), InOpacity=SelectFloat(A=0.5, B=1, bPickA=@n1.bHasLeftGame))
   > SetVisibility(
    self=Get LocalPlayerCallout() -> (LocalPlayerCallout:Border Object Reference)
    InVisibility=Select(Option 0:ESlateVisibility Enum=Collapsed, Option 1:ESlateVisibility Enum=SelfHitTestInvisible, Index=@n1.bHighlight) -> (ReturnValue:ESlateVisibility Enum)
   )
   n3 = Get bIsTeamGame()
   > SetVisibility(
    self=Get BottomPad() -> (BottomPad:Spacer Object Reference)
    InVisibility=Select(Option 0=SelfHitTestInvisible, Option 1=Collapsed, Index=@n3)
   )
   > n4 = Branch(Condition=@n3)
   > n5 = GetDynamicMaterial(self:Border Object Reference=Get TeamColor_Backing() -> (TeamColor_Backing:Border Object Reference))
   > n6 = GetDynamicMaterial(self=Get LocalPlayerCallout())
   > n7 = SetScalarParameterValue(self=[@n6, @n5], ParameterName=TeamColor, Value=Get TeamColorParam() -> (TeamColorParam:double))
   n8 = Branch(execute=[@n15.then, @n7.then], Condition=Get bDisplayWins() -> (bDisplayWins:bool))
   > n9 = Branch(Condition="Equal (Enum)"(
     A:EMinigameWinCondition Enum=Get WinCondition() -> (WinCondition:EMinigameWinCondition Enum=MostRoundWins)
     B:EMinigameWinCondition Enum=MostRoundWins
    ))
   n10 = Get WinsDataHolder() -> (WinsDataHolder:Size Box Object Reference)
   > SetVisibility(self=@n10, InVisibility=SelfHitTestInvisible)
   n11 = Get TotalDataHolder() -> (TotalDataHolder:Size Box Object Reference)
   > SetVisibility(self=@n11, InVisibility=Collapsed)
   n12 = "To Text (Integer)"(Value=Break Minigame Scoreboard Player Row(MinigameScoreboardPlayerRow=Get PlayerBucketData()) -> (bHasLeftGame=~, bHideFromDisplay=~, bHighlight=~).WinCount)
   > "SetText (Text)"(self=Get WinsData() -> (WinsData:Text Object Reference), InText=@n12)
   SetVisibility(execute=@n9.else, self=@n10, InVisibility=Collapsed)
   > n13 = Branch(Condition=NOT Boolean(A=Get bShowCumulativeScoreboard() -> (bShowCumulativeScoreboard:bool)))
   > SetVisibility(self=@n11, InVisibility=SelfHitTestInvisible)
   > "SetText (Text)"(self=Get TotalData() -> (TotalData:Text Object Reference), InText=@n12)
   SetVisibility(execute=@n13.else, self=Get TotalDataHolder(), InVisibility=Collapsed)
   SetVisibility(execute=@n8.else, self=[Get TotalDataHolder(), Get WinsDataHolder()], InVisibility=Collapsed)
   n14 = Make Linear Color!(R:float=0, G:float=0.6724, B:float=1, A:float=1) -> (LinearColor:Linear Color Structure) {is=title K2Node_MakeStruct}
   SetBrushColor(execute=@n4.else, self=Get TeamColor_Backing(), InBrushColor=@n14)
   > SetBrushColor(self=Get LocalPlayerCallout(), InBrushColor=@n14)
   > n15 = UpdateMedalStanding() {is=title K2Node_CallFunction}
  ]}
 WinnerCalloutAnim {nodes [
   WinnerCalloutAnim() {is=title K2Node_FunctionEntry}
   > PlayAnimation(InAnimation=Get AnimWinnerCallout() -> (AnimWinnerCallout:Widget Animation Object Reference))
  ]}
]
variables [
 bDisplayWins {type=boolean}
 bIngameScorerow {type=boolean}
 bIsTeamGame {type=boolean}
 bShowCumulativeScoreboard {type=boolean}
 DataBoxes {type="array<TextBlock>", default []}
 DataHolderHBOXs {type="array<HorizontalBox>", default []}
 DataHolders {type="array<SizeBox>", default []}
 HighlightColor {type=SlateColor, default {colorUseRule=UseColor_Specified, specifiedColor=<rgba 1 0.6795 0.0369 1>}}
 MedalStanding {type=integer}
 NormalTextColor {type=SlateColor, default {colorUseRule=UseColor_Specified, specifiedColor=<rgba 1 1 1 0.75>}}
 PlayerBucketData {
  type=MinigameScoreboardPlayerRow
  default {name="", bHasLeftGame=false, bHighlight=false, playerState=null, teamColorIndex=-1, winCount=0, values []}
 }
 ScoreboardValues {type="array<MinigameScoreboardValue>", default []}
 TeamColorParam {type=number}
 WinCondition {type=EMinigameWinCondition, default=MostRoundWins}
]
widgets {
 info {inheritedWidgetCount=0, parentClass=&UserWidget, rootWidgetClass=&VerticalBox, widgetCount=71}
 widgets [
  DataVerticalBox {widgetClassPath=&VerticalBox, widget=DataVerticalBox}
  DataHorizontalBox {
   widgetClassPath=&HorizontalBox
   bIsVariable=true
   parent=DataVerticalBox
   widget=DataHorizontalBox
   slot {class=VerticalBoxSlot, properties {verticalAlignment=VAlign_Top, size {sizeRule=Fill, value=1}}}
  }
  PlacementIndicatorBox {
   widgetClassPath=&SizeBox
   bIsVariable=true
   parent=DataHorizontalBox
   widget=PlacementIndicatorBox
   properties {bOverride_WidthOverride=true, heightOverride=128, visibility=Collapsed, widthOverride=96}
   slot {class=HorizontalBoxSlot}
  }
  HorizontalBox_1 {widgetClassPath=&HorizontalBox, parent=PlacementIndicatorBox, widget=HorizontalBox_1, slot {class=SizeBoxSlot}}
  Overlay_2 {
   widgetClassPath=&Overlay
   parent=HorizontalBox_1
   widget=Overlay_2
   slot {class=HorizontalBoxSlot, properties {size {sizeRule=Fill, value=1}}}
  }
  PlacementIndicator_Switcher {
   widgetClassPath=&CommonWidgetSwitcherLegacy
   bIsVariable=true
   parent=Overlay_2
   widget=PlacementIndicator_Switcher
   properties {activeWidgetIndex=3}
   slot {class=OverlaySlot, properties {horizontalAlignment=HAlign_Fill, verticalAlignment=VAlign_Fill}}
  }
  Scoreboard_Gold {
   widgetClassPath=/Game/Items/Traps/Blueprints/Toys/UI_Art_Assets/Scoreboard_Gold
   bIsVariable=true
   parent=PlacementIndicator_Switcher
   widget=Scoreboard_Gold
   slot {class=WidgetSwitcherSlot}
  }
  Scoreboard_Silver {
   widgetClassPath=/Game/Items/Traps/Blueprints/Toys/UI_Art_Assets/Scoreboard_Silver
   bIsVariable=true
   parent=PlacementIndicator_Switcher
   widget=Scoreboard_Silver
   slot {class=WidgetSwitcherSlot}
  }
  Scoreboard_Bronze {
   widgetClassPath=/Game/Items/Traps/Blueprints/Toys/UI_Art_Assets/Scoreboard_Bronze
   bIsVariable=true
   parent=PlacementIndicator_Switcher
   widget=Scoreboard_Bronze
   slot {class=WidgetSwitcherSlot}
  }
  Placement_None {widgetClassPath=&Overlay, parent=PlacementIndicator_Switcher, widget=Placement_None, slot {class=WidgetSwitcherSlot}}
  PlacementBacking_NOne {
   widgetClassPath=&Border
   parent=Placement_None
   widget=PlacementBacking_NOne
   properties {brushColor=<rgba 0 0.0417 0.5 1>, padding=<ltrb 0 0 0 0>}
   slot {class=OverlaySlot, properties {horizontalAlignment=HAlign_Fill, verticalAlignment=VAlign_Fill}}
  }
  Spacer_76 {
   widgetClassPath=&Spacer
   parent=PlacementBacking_NOne
   widget=Spacer_76
   properties {size=<1 32>}
   slot {class=BorderSlot, properties {padding=<ltrb 0 0 0 0>}}
  }
  SizeBox_1 {
   widgetClassPath=&SizeBox
   parent=Placement_None
   widget=SizeBox_1
   properties {bOverride_HeightOverride=true, bOverride_WidthOverride=true, heightOverride=6, widthOverride=24}
   slot {class=OverlaySlot, properties {horizontalAlignment=HAlign_Center, verticalAlignment=VAlign_Center}}
  }
  Image_65 {
   widgetClassPath=&Image
   bIsVariable=true
   parent=SizeBox_1
   widget=Image_65
   properties {colorAndOpacity=<rgba 0.0021 0.1248 0.6724 1>}
   slot {class=SizeBoxSlot}
  }
  Spacer_67 {widgetClassPath=&Spacer, parent=HorizontalBox_1, widget=Spacer_67, properties {size=<8 1>}, slot {class=HorizontalBoxSlot}}
  TotalDataHolder {
   widgetClassPath=&SizeBox
   bIsVariable=true
   parent=DataHorizontalBox
   widget=TotalDataHolder
   properties {bOverride_WidthOverride=true, widthOverride=128}
   slot {class=HorizontalBoxSlot}
  }
  HorizontalBox {widgetClassPath=&HorizontalBox, parent=TotalDataHolder, widget=HorizontalBox, slot {class=SizeBoxSlot}}
  Border {
   widgetClassPath=&Border
   parent=HorizontalBox
   widget=Border
   properties {horizontalAlignment=HAlign_Center, verticalAlignment=VAlign_Center, brushColor=<rgba 0 0.1329 0.6038 0.9>, padding=<ltrb 0 0 0 0>}
   slot {class=HorizontalBoxSlot, properties {size {sizeRule=Fill, value=1}}}
  }
  TotalData {
   widgetClassPath=&TextBlock
   bIsVariable=true
   parent=Border
   widget=TotalData
   properties {
    justification=Center
    text=000
    font {
     bForceMonospaced=false
     bMaterialIsStencil=false
     fontMaterial=null
     fontObject=&Burbank
     letterSpacing=0
     monospacedWidth=1
     size=18
     skewAmount=0
     typefaceFontName=Burbank-Black
     outlineSettings {
      bApplyOutlineToDropShadows=false
      bMiteredCorners=false
      bSeparateFillAlpha=false
      outlineMaterial=null
      outlineSize=0
      outlineColor=<rgba 0 0 0 0>
     }
    }
    renderTransform {angle=0, scale=<1 1>, shear=<-8 0>, translation=<0 0>}
    shadowColorAndOpacity=<rgba 0.0052 0.0052 0.0056 0.8235>
    shadowOffset=<0 0>
   }
   slot {class=BorderSlot, properties {horizontalAlignment=HAlign_Center, verticalAlignment=VAlign_Center, padding=<ltrb 0 0 0 0>}}
  }
  Spacer {widgetClassPath=&Spacer, parent=HorizontalBox, widget=Spacer, properties {size=<8 1>}, slot {class=HorizontalBoxSlot}}
  WinsDataHolder {
   widgetClassPath=&SizeBox
   bIsVariable=true
   parent=DataHorizontalBox
   widget=WinsDataHolder
   properties {bOverride_WidthOverride=true, widthOverride=128}
   slot {class=HorizontalBoxSlot}
  }
  HorizontalBox_8 {widgetClassPath=&HorizontalBox, parent=WinsDataHolder, widget=HorizontalBox_8, slot {class=SizeBoxSlot}}
  Border_0 {
   widgetClassPath=&Border
   parent=HorizontalBox_8
   widget=Border_0
   properties {horizontalAlignment=HAlign_Center, verticalAlignment=VAlign_Center, brushColor=<rgba 0 0.1329 0.6038 0.9>, padding=<ltrb 0 0 0 0>}
   slot {class=HorizontalBoxSlot, properties {size {sizeRule=Fill, value=1}}}
  }
  WinsData {
   widgetClassPath=&TextBlock
   bIsVariable=true
   parent=Border_0
   widget=WinsData
   properties {
    justification=Center
    text=000
    font {
     bForceMonospaced=false
     bMaterialIsStencil=false
     fontMaterial=null
     fontObject=&Burbank
     letterSpacing=0
     monospacedWidth=1
     size=18
     skewAmount=0
     typefaceFontName=Burbank-Black
     outlineSettings {
      bApplyOutlineToDropShadows=false
      bMiteredCorners=false
      bSeparateFillAlpha=false
      outlineMaterial=null
      outlineSize=0
      outlineColor=<rgba 0 0 0 0>
     }
    }
    renderTransform {angle=0, scale=<1 1>, shear=<-8 0>, translation=<0 0>}
    shadowColorAndOpacity=<rgba 0.0052 0.0052 0.0056 0.8235>
    shadowOffset=<0 0>
   }
   slot {class=BorderSlot, properties {horizontalAlignment=HAlign_Center, verticalAlignment=VAlign_Center, padding=<ltrb 0 0 0 0>}}
  }
  Spacer_132 {widgetClassPath=&Spacer, parent=HorizontalBox_8, widget=Spacer_132, properties {size=<8 1>}, slot {class=HorizontalBoxSlot}}
  StatsBorder {
   widgetClassPath=&Border
   bIsVariable=true
   parent=DataHorizontalBox
   widget=StatsBorder
   properties {brushColor=<rgba 0 0.0578 0.6038 0.5>, padding=<ltrb 0 0 0 0>}
   slot {class=HorizontalBoxSlot, properties {verticalAlignment=VAlign_Top, size {sizeRule=Fill, value=1}}}
  }
  Overlay_0 {widgetClassPath=&Overlay, parent=StatsBorder, widget=Overlay_0, slot {class=BorderSlot, properties {padding=<ltrb 0 0 0 0>}}}
  HorizontalBox_11 {
   widgetClassPath=&HorizontalBox
   parent=Overlay_0
   widget=HorizontalBox_11
   slot {class=OverlaySlot, properties {horizontalAlignment=HAlign_Fill, verticalAlignment=VAlign_Fill}}
  }
  PlayerNameBox {
   widgetClassPath=&SizeBox
   parent=HorizontalBox_11
   widget=PlayerNameBox
   properties {bOverride_WidthOverride=true, minDesiredWidth=256, widthOverride=256}
   slot {class=HorizontalBoxSlot, properties {horizontalAlignment=HAlign_Left}}
  }
  HorizontalBox_10 {
   widgetClassPath=&HorizontalBox
   parent=PlayerNameBox
   widget=HorizontalBox_10
   slot {class=SizeBoxSlot, properties {verticalAlignment=VAlign_Top}}
  }
  HeroPortraitOverlay {widgetClassPath=&Overlay, parent=HorizontalBox_10, widget=HeroPortraitOverlay, slot {class=HorizontalBoxSlot}}
  HeroPortraitSizeBox {
   widgetClassPath=&SizeBox
   parent=HeroPortraitOverlay
   widget=HeroPortraitSizeBox
   properties {bOverride_MinDesiredWidth=true, minDesiredWidth=8}
   slot {class=OverlaySlot, properties {horizontalAlignment=HAlign_Fill, verticalAlignment=VAlign_Fill}}
  }
  TeamColor_Backing {
   widgetClassPath=&Border
   bIsVariable=true
   parent=HeroPortraitSizeBox
   widget=TeamColor_Backing
   properties {
    horizontalAlignment=HAlign_Left
    verticalAlignment=VAlign_Top
    background {
     bIsDynamicallyLoaded=false
     drawAs=Image
     imageType=NoImage
     mirroring=NoMirror
     resourceName=None
     resourceObject=/Game/Items/Traps/Blueprints/Toys/UI_Art_Assets/M_Creative_TeamColor
     tiling=NoTile
     imageSize=<32 32>
     margin=<ltrb 0 0 0 0>
     outlineSettings {
      bUseBrushTransparency=false
      roundingType=HalfHeightRadius
      width=0
      color {colorUseRule=UseColor_Specified, specifiedColor=<rgba 0 0 0 0>}
      cornerRadii=<0 0 0 0>
     }
     tintColor {colorUseRule=UseColor_Specified, specifiedColor=<rgba 1 1 1 1>}
     uVRegion {bIsValid=false, max=<0 0>, min=<0 0>}
    }
    padding=<ltrb 0 0 0 0>
   }
   slot {class=SizeBoxSlot}
  }
  SizeBoxPortraitsizeBox {
   widgetClassPath=&SizeBox
   parent=TeamColor_Backing
   widget=SizeBoxPortraitsizeBox
   properties {bOverride_HeightOverride=true, bOverride_WidthOverride=true, heightOverride=32, widthOverride=32}
   slot {class=BorderSlot, properties {horizontalAlignment=HAlign_Left, verticalAlignment=VAlign_Top, padding=<ltrb 0 0 0 0>}}
  }
  HeroPortrait {
   widgetClassPath=&Image
   bIsVariable=true
   parent=SizeBoxPortraitsizeBox
   widget=HeroPortrait
   properties {
    visibility=Collapsed
    brush {
     bIsDynamicallyLoaded=false
     drawAs=Image
     imageType=NoImage
     mirroring=NoMirror
     resourceName=None
     resourceObject=/Game/UI/Foundation/Textures/Quest/T-Jonesy
     tiling=NoTile
     imageSize=<512 512>
     margin=<ltrb 0 0 0 0>
     outlineSettings {
      bUseBrushTransparency=false
      roundingType=HalfHeightRadius
      width=0
      color {colorUseRule=UseColor_Specified, specifiedColor=<rgba 0 0 0 0>}
      cornerRadii=<0 0 0 0>
     }
     tintColor {colorUseRule=UseColor_Specified, specifiedColor=<rgba 1 1 1 1>}
     uVRegion {bIsValid=false, max=<0 0>, min=<0 0>}
    }
   }
   slot {class=SizeBoxSlot}
  }
  NameBox {
   widgetClassPath=&CommonTextBlock
   bIsVariable=true
   parent=HorizontalBox_10
   widget=NameBox
   properties {
    clipping=ClipToBounds
    scrollStyle=&TextScrollStyle-Base_C
    text=PlayerName16Char
    visibility=Visible
    colorAndOpacity {colorUseRule=UseColor_Specified, specifiedColor=<rgba 0.2159 0.6654 1 1>}
    font {
     bForceMonospaced=false
     bMaterialIsStencil=false
     fontMaterial=null
     fontObject=&BurbankSmall
     letterSpacing=0
     monospacedWidth=1
     size=16
     skewAmount=0
     typefaceFontName=Black
     outlineSettings {
      bApplyOutlineToDropShadows=false
      bMiteredCorners=false
      bSeparateFillAlpha=false
      outlineMaterial=null
      outlineSize=0
      outlineColor=<rgba 0 0 0 0>
     }
    }
    shadowColorAndOpacity=<rgba 0.0052 0.0052 0.0056 0.8235>
    shadowOffset=<0 0>
   }
   slot {class=HorizontalBoxSlot, properties {verticalAlignment=VAlign_Center, padding=<ltrb 8 0 0 0>, size {sizeRule=Fill, value=1}}}
  }
  Divider0 {
   widgetClassPath=&Overlay
   parent=HorizontalBox_10
   widget=Divider0
   slot {class=HorizontalBoxSlot, properties {horizontalAlignment=HAlign_Right}}
  }
  Spacer_49 {widgetClassPath=&Spacer, parent=Divider0, widget=Spacer_49, properties {size=<8 1>}, slot {class=OverlaySlot}}
  Image_60 {
   widgetClassPath=&Image
   bIsVariable=true
   parent=Divider0
   widget=Image_60
   properties {
    brush {
     bIsDynamicallyLoaded=false
     drawAs=Image
     imageType=NoImage
     mirroring=NoMirror
     resourceName=None
     resourceObject=null
     tiling=NoTile
     imageSize=<8 32>
     margin=<ltrb 0 0 0 0>
     outlineSettings {
      bUseBrushTransparency=false
      roundingType=HalfHeightRadius
      width=0
      color {colorUseRule=UseColor_Specified, specifiedColor=<rgba 0 0 0 0>}
      cornerRadii=<0 0 0 0>
     }
     tintColor {colorUseRule=UseColor_Specified, specifiedColor=<rgba 1 1 1 1>}
     uVRegion {bIsValid=false, max=<0 0>, min=<0 0>}
    }
    colorAndOpacity=<rgba 0.0513 0.5271 1 0.2>
   }
   slot {class=OverlaySlot, properties {horizontalAlignment=HAlign_Right, verticalAlignment=VAlign_Fill}}
  }
  StatsDataBox {
   widgetClassPath=&SizeBox
   parent=HorizontalBox_11
   widget=StatsDataBox
   properties {widthOverride=1000}
   slot {class=HorizontalBoxSlot, properties {size {sizeRule=Fill, value=1}}}
  }
  HorizontalBox_0 {widgetClassPath=&HorizontalBox, parent=StatsDataBox, widget=HorizontalBox_0, slot {class=SizeBoxSlot}}
  DataHolderHBOX_1 {
   widgetClassPath=&HorizontalBox
   bIsVariable=true
   parent=HorizontalBox_0
   widget=DataHolderHBOX_1
   slot {class=HorizontalBoxSlot, properties {size {sizeRule=Fill, value=1}}}
  }
  DataHolder1 {
   widgetClassPath=&SizeBox
   bIsVariable=true
   parent=DataHolderHBOX_1
   widget=DataHolder1
   slot {class=HorizontalBoxSlot, properties {verticalAlignment=VAlign_Center, size {sizeRule=Fill, value=1}}}
  }
  DataBox1 {
   widgetClassPath=&TextBlock
   bIsVariable=true
   parent=DataHolder1
   widget=DataBox1
   properties {
    text=0000000
    colorAndOpacity {colorUseRule=UseColor_Specified, specifiedColor=<rgba 1 1 1 0.75>}
    font {
     bForceMonospaced=false
     bMaterialIsStencil=false
     fontMaterial=null
     fontObject=&BurbankSmall
     letterSpacing=0
     monospacedWidth=1
     size=16
     skewAmount=0
     typefaceFontName=Bold
     outlineSettings {
      bApplyOutlineToDropShadows=false
      bMiteredCorners=false
      bSeparateFillAlpha=false
      outlineMaterial=null
      outlineSize=0
      outlineColor=<rgba 0 0 0 0>
     }
    }
    renderTransform {angle=0, scale=<1 1>, shear=<-8 0>, translation=<0 0>}
    shadowColorAndOpacity=<rgba 0.0052 0.0052 0.0056 0.8235>
    shadowOffset=<0 0>
   }
   slot {class=SizeBoxSlot, properties {verticalAlignment=VAlign_Center, padding=<ltrb 8 0 0 0>}}
  }
  Divider1 {widgetClassPath=&Overlay, parent=DataHolderHBOX_1, widget=Divider1, slot {class=HorizontalBoxSlot}}
  Spacer_51 {widgetClassPath=&Spacer, parent=Divider1, widget=Spacer_51, properties {size=<8 1>}, slot {class=OverlaySlot}}
  Image_62 {
   widgetClassPath=&Image
   bIsVariable=true
   parent=Divider1
   widget=Image_62
   properties {
    visibility=SelfHitTestInvisible
    brush {
     bIsDynamicallyLoaded=false
     drawAs=Image
     imageType=NoImage
     mirroring=NoMirror
     resourceName=None
     resourceObject=null
     tiling=NoTile
     imageSize=<8 32>
     margin=<ltrb 0 0 0 0>
     outlineSettings {
      bUseBrushTransparency=false
      roundingType=HalfHeightRadius
      width=0
      color {colorUseRule=UseColor_Specified, specifiedColor=<rgba 0 0 0 0>}
      cornerRadii=<0 0 0 0>
     }
     tintColor {colorUseRule=UseColor_Specified, specifiedColor=<rgba 1 1 1 1>}
     uVRegion {bIsValid=false, max=<0 0>, min=<0 0>}
    }
    colorAndOpacity=<rgba 0.0513 0.5271 1 0.2>
   }
   slot {class=OverlaySlot, properties {horizontalAlignment=HAlign_Fill, verticalAlignment=VAlign_Fill}}
  }
  DataHolderHBOX_2 {
   widgetClassPath=&HorizontalBox
   bIsVariable=true
   parent=HorizontalBox_0
   widget=DataHolderHBOX_2
   slot {class=HorizontalBoxSlot, properties {verticalAlignment=VAlign_Center, size {sizeRule=Fill, value=1}}}
  }
  DataHolder2 {
   widgetClassPath=&SizeBox
   bIsVariable=true
   parent=DataHolderHBOX_2
   widget=DataHolder2
   slot {class=HorizontalBoxSlot, properties {verticalAlignment=VAlign_Center, size {sizeRule=Fill, value=1}}}
  }
  DataBox2 {
   widgetClassPath=&TextBlock
   bIsVariable=true
   parent=DataHolder2
   widget=DataBox2
   properties {
    text=00000000
    colorAndOpacity {colorUseRule=UseColor_Specified, specifiedColor=<rgba 1 1 1 0.75>}
    font {
     bForceMonospaced=false
     bMaterialIsStencil=false
     fontMaterial=null
     fontObject=&BurbankSmall
     letterSpacing=0
     monospacedWidth=1
     size=16
     skewAmount=0
     typefaceFontName=Bold
     outlineSettings {
      bApplyOutlineToDropShadows=false
      bMiteredCorners=false
      bSeparateFillAlpha=false
      outlineMaterial=null
      outlineSize=0
      outlineColor=<rgba 0 0 0 0>
     }
    }
    renderTransform {angle=0, scale=<1 1>, shear=<-8 0>, translation=<0 0>}
    shadowColorAndOpacity=<rgba 0.0052 0.0052 0.0056 0.8235>
    shadowOffset=<0 0>
   }
   slot {class=SizeBoxSlot, properties {verticalAlignment=VAlign_Center, padding=<ltrb 8 0 0 0>}}
  }
  Divider2 {widgetClassPath=&Overlay, parent=DataHolderHBOX_2, widget=Divider2, slot {class=HorizontalBoxSlot}}
  Spacer_52 {widgetClassPath=&Spacer, parent=Divider2, widget=Spacer_52, properties {size=<8 1>}, slot {class=OverlaySlot}}
  Image_196 {
   widgetClassPath=&Image
   bIsVariable=true
   parent=Divider2
   widget=Image_196
   properties {
    visibility=SelfHitTestInvisible
    brush {
     bIsDynamicallyLoaded=false
     drawAs=Image
     imageType=NoImage
     mirroring=NoMirror
     resourceName=None
     resourceObject=null
     tiling=NoTile
     imageSize=<8 32>
     margin=<ltrb 0 0 0 0>
     outlineSettings {
      bUseBrushTransparency=false
      roundingType=HalfHeightRadius
      width=0
      color {colorUseRule=UseColor_Specified, specifiedColor=<rgba 0 0 0 0>}
      cornerRadii=<0 0 0 0>
     }
     tintColor {colorUseRule=UseColor_Specified, specifiedColor=<rgba 1 1 1 1>}
     uVRegion {bIsValid=false, max=<0 0>, min=<0 0>}
    }
    colorAndOpacity=<rgba 0.0513 0.5271 1 0.2>
   }
   slot {class=OverlaySlot, properties {horizontalAlignment=HAlign_Fill, verticalAlignment=VAlign_Fill}}
  }
  DataHolderHBOX_3 {
   widgetClassPath=&HorizontalBox
   bIsVariable=true
   parent=HorizontalBox_0
   widget=DataHolderHBOX_3
   slot {class=HorizontalBoxSlot, properties {verticalAlignment=VAlign_Center, size {sizeRule=Fill, value=1}}}
  }
  DataHolder3 {
   widgetClassPath=&SizeBox
   bIsVariable=true
   parent=DataHolderHBOX_3
   widget=DataHolder3
   slot {class=HorizontalBoxSlot, properties {verticalAlignment=VAlign_Center, size {sizeRule=Fill, value=1}}}
  }
  DataBox3 {
   widgetClassPath=&TextBlock
   bIsVariable=true
   parent=DataHolder3
   widget=DataBox3
   properties {
    text=00000000
    colorAndOpacity {colorUseRule=UseColor_Specified, specifiedColor=<rgba 1 1 1 0.75>}
    font {
     bForceMonospaced=false
     bMaterialIsStencil=false
     fontMaterial=null
     fontObject=&BurbankSmall
     letterSpacing=0
     monospacedWidth=1
     size=16
     skewAmount=0
     typefaceFontName=Bold
     outlineSettings {
      bApplyOutlineToDropShadows=false
      bMiteredCorners=false
      bSeparateFillAlpha=false
      outlineMaterial=null
      outlineSize=0
      outlineColor=<rgba 0 0 0 0>
     }
    }
    renderTransform {angle=0, scale=<1 1>, shear=<-8 0>, translation=<0 0>}
    shadowColorAndOpacity=<rgba 0.0052 0.0052 0.0056 0.8235>
    shadowOffset=<0 0>
   }
   slot {class=SizeBoxSlot, properties {verticalAlignment=VAlign_Center, padding=<ltrb 8 0 0 0>}}
  }
  Divider3 {widgetClassPath=&Overlay, parent=DataHolderHBOX_3, widget=Divider3, slot {class=HorizontalBoxSlot}}
  Spacer_53 {widgetClassPath=&Spacer, parent=Divider3, widget=Spacer_53, properties {size=<8 1>}, slot {class=OverlaySlot}}
  Image_197 {
   widgetClassPath=&Image
   bIsVariable=true
   parent=Divider3
   widget=Image_197
   properties {
    visibility=SelfHitTestInvisible
    brush {
     bIsDynamicallyLoaded=false
     drawAs=Image
     imageType=NoImage
     mirroring=NoMirror
     resourceName=None
     resourceObject=null
     tiling=NoTile
     imageSize=<8 32>
     margin=<ltrb 0 0 0 0>
     outlineSettings {
      bUseBrushTransparency=false
      roundingType=HalfHeightRadius
      width=0
      color {colorUseRule=UseColor_Specified, specifiedColor=<rgba 0 0 0 0>}
      cornerRadii=<0 0 0 0>
     }
     tintColor {colorUseRule=UseColor_Specified, specifiedColor=<rgba 1 1 1 1>}
     uVRegion {bIsValid=false, max=<0 0>, min=<0 0>}
    }
    colorAndOpacity=<rgba 0.0513 0.5271 1 0.2>
   }
   slot {class=OverlaySlot, properties {horizontalAlignment=HAlign_Fill, verticalAlignment=VAlign_Fill}}
  }
  DataHolderHBOX_4 {
   widgetClassPath=&HorizontalBox
   bIsVariable=true
   parent=HorizontalBox_0
   widget=DataHolderHBOX_4
   slot {class=HorizontalBoxSlot, properties {verticalAlignment=VAlign_Center, size {sizeRule=Fill, value=1}}}
  }
  DataHolder4 {
   widgetClassPath=&SizeBox
   bIsVariable=true
   parent=DataHolderHBOX_4
   widget=DataHolder4
   slot {class=HorizontalBoxSlot, properties {verticalAlignment=VAlign_Center, size {sizeRule=Fill, value=1}}}
  }
  DataBox4 {
   widgetClassPath=&TextBlock
   bIsVariable=true
   parent=DataHolder4
   widget=DataBox4
   properties {
    text=00000000
    colorAndOpacity {colorUseRule=UseColor_Specified, specifiedColor=<rgba 1 1 1 0.75>}
    font {
     bForceMonospaced=false
     bMaterialIsStencil=false
     fontMaterial=null
     fontObject=&BurbankSmall
     letterSpacing=0
     monospacedWidth=1
     size=16
     skewAmount=0
     typefaceFontName=Bold
     outlineSettings {
      bApplyOutlineToDropShadows=false
      bMiteredCorners=false
      bSeparateFillAlpha=false
      outlineMaterial=null
      outlineSize=0
      outlineColor=<rgba 0 0 0 0>
     }
    }
    renderTransform {angle=0, scale=<1 1>, shear=<-8 0>, translation=<0 0>}
    shadowColorAndOpacity=<rgba 0.0052 0.0052 0.0056 0.8235>
    shadowOffset=<0 0>
   }
   slot {class=SizeBoxSlot, properties {verticalAlignment=VAlign_Center, padding=<ltrb 8 0 0 0>}}
  }
  Divider4 {widgetClassPath=&Overlay, parent=DataHolderHBOX_4, widget=Divider4, slot {class=HorizontalBoxSlot}}
  Spacer_54 {widgetClassPath=&Spacer, parent=Divider4, widget=Spacer_54, properties {size=<8 1>}, slot {class=OverlaySlot}}
  Image_198 {
   widgetClassPath=&Image
   bIsVariable=true
   parent=Divider4
   widget=Image_198
   properties {
    visibility=SelfHitTestInvisible
    brush {
     bIsDynamicallyLoaded=false
     drawAs=Image
     imageType=NoImage
     mirroring=NoMirror
     resourceName=None
     resourceObject=null
     tiling=NoTile
     imageSize=<8 32>
     margin=<ltrb 0 0 0 0>
     outlineSettings {
      bUseBrushTransparency=false
      roundingType=HalfHeightRadius
      width=0
      color {colorUseRule=UseColor_Specified, specifiedColor=<rgba 0 0 0 0>}
      cornerRadii=<0 0 0 0>
     }
     tintColor {colorUseRule=UseColor_Specified, specifiedColor=<rgba 1 1 1 1>}
     uVRegion {bIsValid=false, max=<0 0>, min=<0 0>}
    }
    colorAndOpacity=<rgba 0.0513 0.5271 1 0.2>
   }
   slot {class=OverlaySlot, properties {horizontalAlignment=HAlign_Fill, verticalAlignment=VAlign_Fill}}
  }
  DataHolderHBOX_5 {
   widgetClassPath=&HorizontalBox
   bIsVariable=true
   parent=HorizontalBox_0
   widget=DataHolderHBOX_5
   slot {class=HorizontalBoxSlot, properties {verticalAlignment=VAlign_Center, size {sizeRule=Fill, value=1}}}
  }
  DataHolder5 {
   widgetClassPath=&SizeBox
   bIsVariable=true
   parent=DataHolderHBOX_5
   widget=DataHolder5
   slot {class=HorizontalBoxSlot, properties {verticalAlignment=VAlign_Center, size {sizeRule=Fill, value=1}}}
  }
  DataBox5 {
   widgetClassPath=&TextBlock
   bIsVariable=true
   parent=DataHolder5
   widget=DataBox5
   properties {
    text=00000000
    colorAndOpacity {colorUseRule=UseColor_Specified, specifiedColor=<rgba 1 1 1 0.75>}
    font {
     bForceMonospaced=false
     bMaterialIsStencil=false
     fontMaterial=null
     fontObject=&BurbankSmall
     letterSpacing=0
     monospacedWidth=1
     size=16
     skewAmount=0
     typefaceFontName=Bold
     outlineSettings {
      bApplyOutlineToDropShadows=false
      bMiteredCorners=false
      bSeparateFillAlpha=false
      outlineMaterial=null
      outlineSize=0
      outlineColor=<rgba 0 0 0 0>
     }
    }
    renderTransform {angle=0, scale=<1 1>, shear=<-8 0>, translation=<0 0>}
    shadowColorAndOpacity=<rgba 0.0052 0.0052 0.0056 0.8235>
    shadowOffset=<0 0>
   }
   slot {class=SizeBoxSlot, properties {verticalAlignment=VAlign_Center, padding=<ltrb 8 0 0 0>}}
  }
  LocalPlayerCallout {
   widgetClassPath=&Border
   bIsVariable=true
   parent=Overlay_0
   widget=LocalPlayerCallout
   properties {
    renderOpacity=0.75
    background {
     bIsDynamicallyLoaded=false
     drawAs=Box
     imageType=NoImage
     mirroring=NoMirror
     resourceName=None
     resourceObject=/Game/Items/Traps/Blueprints/Toys/UI_Art_Assets/M_UI_Scoreboard_LocalPlayer_Indicator
     tiling=NoTile
     imageSize=<32 32>
     margin=<ltrb 0.5 0.5 0.5 0.5>
     outlineSettings {
      bUseBrushTransparency=false
      roundingType=HalfHeightRadius
      width=0
      color {colorUseRule=UseColor_Specified, specifiedColor=<rgba 0 0 0 0>}
      cornerRadii=<0 0 0 0>
     }
     tintColor {colorUseRule=UseColor_Specified, specifiedColor=<rgba 1 1 1 1>}
     uVRegion {bIsValid=false, max=<0 0>, min=<0 0>}
    }
    padding=<ltrb 0 0 0 0>
   }
   slot {class=OverlaySlot, properties {horizontalAlignment=HAlign_Fill, verticalAlignment=VAlign_Fill}}
  }
  WinnerCallout {
   widgetClassPath=&Image
   bIsVariable=true
   parent=Overlay_0
   widget=WinnerCallout
   properties {brush {
     bIsDynamicallyLoaded=false
     drawAs=Image
     imageType=NoImage
     mirroring=NoMirror
     resourceName=None
     resourceObject=/Game/Items/Traps/Blueprints/Toys/UI_Art_Assets/M_UI_RowWinnerCallout
     tiling=NoTile
     imageSize=<32 32>
     margin=<ltrb 0 0 0 0>
     outlineSettings {
      bUseBrushTransparency=false
      roundingType=HalfHeightRadius
      width=0
      color {colorUseRule=UseColor_Specified, specifiedColor=<rgba 0 0 0 0>}
      cornerRadii=<0 0 0 0>
     }
     tintColor {colorUseRule=UseColor_Specified, specifiedColor=<rgba 1 1 1 1>}
     uVRegion {bIsValid=false, max=<0 0>, min=<0 0>}
    }}
   slot {class=OverlaySlot, properties {horizontalAlignment=HAlign_Fill, verticalAlignment=VAlign_Fill}}
  }
  BottomPad {
   widgetClassPath=&Spacer
   bIsVariable=true
   parent=DataVerticalBox
   widget=BottomPad
   properties {size=<1 8>}
   slot {class=VerticalBoxSlot}
  }
 ]
}
