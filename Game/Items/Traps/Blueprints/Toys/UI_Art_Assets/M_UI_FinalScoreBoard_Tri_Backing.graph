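# Translucent UI material that draws an animated triangle-tile backing, named for the final scoreboard.
# Aspect-corrected screen UVs are scaled and rotated, then a time sine crossfades the tile texture's blue and red channels.
# That mask, lifted by NoiseLift, blends two colors and drives opacity, with an optional horizontal sine fade clearing the center.
# Panning noise feeds lerps with identical inputs, so it has no visible effect.
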Material /Game/Items/Traps/Blueprints/Toys/UI_Art_Assets/M_UI_FinalScoreBoard_Tri_Backing
graph {
 properties {blendMode=BLEND_Translucent, materialDomain=MD_UI, pixelDepthOffsetMode=PDOM_Legacy, usedShadingModels=Default Lit}
 nodes [
  n1 = MaterialFunctionCall(Float2=MaterialFunctionCall() {materialFunction=&ScreenResolution}.Visible Resolution) {materialFunction=&BreakOutFloat2Components}
  n2 = Max(A=@n1.G) {constB=0.1}
  n3 = MaterialFunctionCall() {materialFunction=&ScreenAlignedUVs}
  n4 = Multiply(A=AppendVector(A=Divide(A=@n1.R, B=@n2), B=Constant() {r=1}), B=@n3."X 100%, Y 100%") {constB=10}
  n5 = AppendVector(
   A=ScalarParameter() {defaultValue=4, group=TriangleBackground, parameterName=TriangleTiles-U}
   B=ScalarParameter() {defaultValue=2.5, group=TriangleBackground, parameterName=TriangleTiles-V}
  )
  n6 = TextureSample(UVs=Rotator(
    Coordinate=Multiply(A=@n4, B=@n5) {constB=20}
    Time=ScalarParameter() {defaultValue=0.125, group=TriangleBackground, parameterName=TriangleRotation}
   ) {speed=6.2832}) {texture=/Game/UI/Frontend/Campaign/EventCallOut/T-UI-TriangleTile}
  n7 = Multiply(
   A=@n3."X 100%, Y 100%"
   B=AppendVector(
    A=ScalarParameter() {defaultValue=1, group=TriangleBackground, parameterName=TriangleBlendTiles-U}
    B=ScalarParameter() {defaultValue=0.5, group=TriangleBackground, parameterName=TriangleBlendTiles-V}
   )
  )
  n8 = Time()
  n9 = Multiply(A=@n8, B=ScalarParameter() {defaultValue=1, group=TriangleBackground, parameterName=TimeMultiplier})
  n10 = Multiply(
   A=TextureSample(UVs=Panner(Coordinate=@n7, Time=@n9) {speedX=-0.25, speedY=0.125}) {samplerType=SAMPLERTYPE_Grayscale, texture=&LowResBlurredNoise}.R
   B=ScalarParameter() {defaultValue=1, parameterName=NoiseIntensity}
  )
  n11 = LinearInterpolate(
   A=LinearInterpolate(A=@n6.B, B=@n6.B, Alpha=@n10)
   B=LinearInterpolate(A=@n6.R, B=@n6.R, Alpha=@n10)
   Alpha=Clamp(None=Sine(None=@n8) {period=4})
  )
  n12 = ScalarParameter() {parameterName=NoiseLift}
  Clamp(None=@n11, Min=@n12)
  n13 = LinearInterpolate(A=@n11, B=Constant() {r=1}, Alpha=@n12)
  Saturate(None=LinearInterpolate(
    A=VectorParameter() {parameterName=Color2, defaultValue=<rgba 0 0.1667 1 1>}.RGB
    B=VectorParameter() {parameterName=Color1, defaultValue=<rgba 0 0.8333 1 1>}.RGB
    Alpha=@n13
   ))
  n14 = Sine(None=Multiply(A=TextureCoordinate()) {constB=0.5})
  n15 = DotProduct(A=@n13, B=OneMinus(None=ComponentMask(None=@n14) {r=true}))
  n16 = ScalarParameter() {defaultValue=1, parameterName=UseFade}
  Saturate(None=Multiply(A=LinearInterpolate(A=@n13, B=@n15, Alpha=@n16), B=ScalarParameter() {defaultValue=0.5, parameterName=Alpha}))
 ]
}
