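# Translucent UI material that draws a procedural round marker, named for the round-end screen.
# An exponential radial gradient forms the main disc, with edge falloff set by a Softness parameter.
# A slightly larger, hard-coded gradient is capped at low strength and scaled by UseRing, adding an optional faint outer ring to the clamped mask.
# The inner disc mask also blends between two orange-yellow color parameters, tinting disc and ring differently.
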Material /Game/Items/Traps/Blueprints/Toys/UI_Art_Assets/M_UI_RoundEnd_Marker
graph {
 properties {blendMode=BLEND_Translucent, materialDomain=MD_UI, pixelDepthOffsetMode=PDOM_Legacy, usedShadingModels=Default Lit}
 nodes [
  n1 = Clamp(None=MaterialFunctionCall(
    UVs=TextureCoordinate()
    CenterPosition=Constant2Vector() {g=0.5, r=0.5}
    Radius=Constant() {r=0.45}
    Density=ScalarParameter() {defaultValue=32, parameterName=Softness}
   ) {materialFunction=&RadialGradientExponential}.RadialGradientExponential)
  n2 = Clamp(None=MaterialFunctionCall(UVs=TextureCoordinate(), CenterPosition=Constant2Vector() {g=0.5, r=0.5}, Radius=Constant() {r=0.5}, Density=Constant() {r=32}) {materialFunction=&RadialGradientExponential}.RadialGradientExponential) {maxDefault=0.25}
  n3 = ScalarParameter() {defaultValue=1, parameterName=UseRing}
  Clamp(None=Add(A=@n1, B=Multiply(A=@n2, B=@n3)))
  LinearInterpolate(
   A=VectorParameter() {parameterName=Color2, defaultValue=<rgba 1 0.3175 0.025 1>}.RGB
   B=VectorParameter() {parameterName=Color1, defaultValue=<rgba 1 0.5987 0.0369 1>}.RGB
   Alpha=@n1
  )
 ]
}
