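# Scoreboard widget for Fort minigames, showing live standings in-game and round or game-end results with countdown and animated winner banner.
# SetStatsColumns reads tracked stats and minigame rules to configure column labels and team or individual mode.
# Team or player row widgets display the scores, and the scroll box jumps to the owning or spectated player.
# A looping timer refreshes live stats, restarting or clearing as the fullscreen map toggles.
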WidgetBlueprint /Game/Items/Traps/Blueprints/Toys/Scoreboard_UI
animations [
 Intro {
  binding_count=7
  length_seconds=0.7001
  object=&self:Intro
  sequence {
   bindings [
    ScoreboardDisplay {
     id=#1
     possessed_class=VerticalBox
     tracks [{
       type=MovieSceneFloatTrack
       display_name=Render Opacity
       sections [{
         type=MovieSceneFloatSection
         end_frame=42
         start_frame=0
         channels [None {type=MovieSceneScriptingFloatChannel, keys [[0,0],[15.001,0],[42.004,1]]}]
        }]
      }]
    }
    CountdownText {
     id=#2
     possessed_class=CommonTextBlock
     tracks [
      {
       type=MovieScene2DTransformTrack
       sections [{
         type=MovieScene2DTransformSection
         end_frame=42
         start_frame=0
         channels [
          Translation.X {type=MovieSceneScriptingFloatChannel}
          Translation.Y {type=MovieSceneScriptingFloatChannel}
          Angle {type=MovieSceneScriptingFloatChannel}
          Scale.X {type=MovieSceneScriptingFloatChannel, keys [[0,0.65],[27.001,0.65],[33.001,1.1],[36.001,0.95],[42.001,1]]}
          Scale.Y {type=MovieSceneScriptingFloatChannel, keys [[0,0.85],[27.001,0.85],[33.001,1.1],[36.001,0.925],[42.001,1]]}
          Shear.X {type=MovieSceneScriptingFloatChannel}
          Shear.Y {type=MovieSceneScriptingFloatChannel}
         ]
        }]
      }
      {
       type=MovieSceneFloatTrack
       display_name=Render Opacity
       sections [{
         type=MovieSceneFloatSection
         end_frame=42
         start_frame=0
         channels [None {type=MovieSceneScriptingFloatChannel, keys [[0,0],[27.003,0],[33.001,1]]}]
        }]
      }
     ]
    }
    Header {
     id=#3
     display_name=HeaderText
     possessed_class=HorizontalBox
     tracks [
      {
       type=MovieScene2DTransformTrack
       sections [{
         type=MovieScene2DTransformSection
         end_frame=42
         start_frame=0
         channels [
          Translation.X {type=MovieSceneScriptingFloatChannel, keys [[0,-256],[12.001,-256],[27.001,0]]}
          Translation.Y {type=MovieSceneScriptingFloatChannel}
          Angle {type=MovieSceneScriptingFloatChannel}
          Scale.X {type=MovieSceneScriptingFloatChannel}
          Scale.Y {type=MovieSceneScriptingFloatChannel}
          Shear.X {type=MovieSceneScriptingFloatChannel}
          Shear.Y {type=MovieSceneScriptingFloatChannel}
         ]
        }]
      }
      {
       type=MovieSceneFloatTrack
       display_name=Render Opacity
       sections [{
         type=MovieSceneFloatSection
         end_frame=42
         start_frame=0
         channels [None {type=MovieSceneScriptingFloatChannel, keys [[0,0],[12.001,0],[27.001,1]]}]
        }]
      }
     ]
    }
    HeaderDisplay {
     id=#4
     possessed_class=VerticalBox
     tracks [{
       type=MovieSceneFloatTrack
       display_name=Render Opacity
       sections [{type=MovieSceneFloatSection, end_frame=42, start_frame=0, channels [None {type=MovieSceneScriptingFloatChannel, const [1,1]}]}]
      }]
    }
    Image_110 {
     id=#5
     possessed_class=Image
     tracks [
      {
       type=MovieScene2DTransformTrack
       sections [{
         type=MovieScene2DTransformSection
         end_frame=42
         start_frame=0
         channels [
          Translation.X {type=MovieSceneScriptingFloatChannel}
          Translation.Y {type=MovieSceneScriptingFloatChannel, keys [[0,-64],[24.002,-64],[30.002,8],[33.002,-4],[42.004,0]]}
          Angle {type=MovieSceneScriptingFloatChannel}
          Scale.X {type=MovieSceneScriptingFloatChannel}
          Scale.Y {type=MovieSceneScriptingFloatChannel, keys [[0,2],[24.002,2],[30.002,0.9],[33.002,1]]}
          Shear.X {type=MovieSceneScriptingFloatChannel}
          Shear.Y {type=MovieSceneScriptingFloatChannel}
         ]
        }]
      }
      {
       type=MovieSceneFloatTrack
       display_name=Render Opacity
       sections [{
         type=MovieSceneFloatSection
         end_frame=42
         start_frame=0
         channels [None {type=MovieSceneScriptingFloatChannel, keys [[0,0],[24.002,0],[27.002,1]]}]
        }]
      }
     ]
    }
    CountdownHeader {
     id=#6
     possessed_class=VerticalBox
     tracks [
      {
       type=MovieScene2DTransformTrack
       sections [{
         type=MovieScene2DTransformSection
         end_frame=42
         start_frame=0
         channels [
          Translation.X {type=MovieSceneScriptingFloatChannel, keys [[0,-64],[21.002,-64],[33.002,0]]}
          Translation.Y {type=MovieSceneScriptingFloatChannel}
          Angle {type=MovieSceneScriptingFloatChannel}
          Scale.X {type=MovieSceneScriptingFloatChannel}
          Scale.Y {type=MovieSceneScriptingFloatChannel}
          Shear.X {type=MovieSceneScriptingFloatChannel}
          Shear.Y {type=MovieSceneScriptingFloatChannel}
         ]
        }]
      }
      {
       type=MovieSceneFloatTrack
       display_name=Render Opacity
       sections [{
         type=MovieSceneFloatSection
         end_frame=42
         start_frame=0
         channels [None {type=MovieSceneScriptingFloatChannel, keys [[0,0],[21.002,0],[33.002,1]]}]
        }]
      }
     ]
    }
    WinnerDisplay {
     id=#7
     possessed_class=Overlay
     tracks [{
       type=MovieSceneFloatTrack
       display_name=Render Opacity
       sections [{type=MovieSceneFloatSection, end_frame=42, start_frame=0, channels [None {type=MovieSceneScriptingFloatChannel, keys [[0,0],[15,1]]}]}]
      }]
    }
   ]
   display_rate {numerator=60}
  }
 }
 WinnerDisplayIntro {
  binding_count=8
  length_seconds=1
  object=&self:WinnerDisplayIntro
  sequence {
   bindings [
    WinDescription {
     id=#8
     possessed_class=CommonTextBlock
     tracks [
      {
       type=MovieScene2DTransformTrack
       sections [{
         type=MovieScene2DTransformSection
         end_frame=60
         start_frame=0
         channels [
          Translation.X {type=MovieSceneScriptingFloatChannel, keys [[18,-88],[27,0]]}
          Translation.Y {type=MovieSceneScriptingFloatChannel}
          Angle {type=MovieSceneScriptingFloatChannel}
          Scale.X {type=MovieSceneScriptingFloatChannel}
          Scale.Y {type=MovieSceneScriptingFloatChannel}
          Shear.X {type=MovieSceneScriptingFloatChannel}
          Shear.Y {type=MovieSceneScriptingFloatChannel}
         ]
        }]
      }
      {
       type=MovieSceneFloatTrack
       display_name=Render Opacity
       sections [{type=MovieSceneFloatSection, end_frame=60, start_frame=0, channels [None {type=MovieSceneScriptingFloatChannel, keys [[18,0],[27,1]]}]}]
      }
     ]
    }
    Triangles {
     id=#9
     possessed_class=SizeBox
     tracks [
      {
       type=MovieSceneFloatTrack
       display_name=Render Opacity
       sections [{type=MovieSceneFloatSection, end_frame=59, start_frame=0, channels [None {type=MovieSceneScriptingFloatChannel, keys [[15,0],[21,1]]}]}]
      }
      {
       type=MovieScene2DTransformTrack
       sections [{
         type=MovieScene2DTransformSection
         end_frame=60
         start_frame=0
         channels [
          Translation.X {type=MovieSceneScriptingFloatChannel}
          Translation.Y {type=MovieSceneScriptingFloatChannel}
          Angle {type=MovieSceneScriptingFloatChannel}
          Scale.X {type=MovieSceneScriptingFloatChannel, keys [[15,0],[30,1]]}
          Scale.Y {type=MovieSceneScriptingFloatChannel}
          Shear.X {type=MovieSceneScriptingFloatChannel}
          Shear.Y {type=MovieSceneScriptingFloatChannel}
         ]
        }]
      }
     ]
    }
    Glow {
     id=#10
     possessed_class=Image
     tracks [
      {
       type=MovieSceneFloatTrack
       display_name=Render Opacity
       sections [{
         type=MovieSceneFloatSection
         end_frame=60
         start_frame=0
         channels [None {type=MovieSceneScriptingFloatChannel, keys [[15,0],[18,1],[57,0.25]]}]
        }]
      }
      {
       type=MovieScene2DTransformTrack
       sections [{
         type=MovieScene2DTransformSection
         end_frame=60
         start_frame=0
         channels [
          Translation.X {type=MovieSceneScriptingFloatChannel}
          Translation.Y {type=MovieSceneScriptingFloatChannel}
          Angle {type=MovieSceneScriptingFloatChannel}
          Scale.X {type=MovieSceneScriptingFloatChannel, keys [[0,1.125],[15,1.125],[18,1.5],[57,1.125]]}
          Scale.Y {type=MovieSceneScriptingFloatChannel, keys [[15,0.666],[57,1.25]]}
          Shear.X {type=MovieSceneScriptingFloatChannel}
          Shear.Y {type=MovieSceneScriptingFloatChannel}
         ]
        }]
      }
     ]
    }
    Circle {
     id=#11
     possessed_class=SizeBox
     tracks [
      {
       type=MovieSceneFloatTrack
       display_name=Render Opacity
       sections [{type=MovieSceneFloatSection, end_frame=60, start_frame=0, channels [None {type=MovieSceneScriptingFloatChannel, keys [[0,0],[9,0],[13,1]]}]}]
      }
      {
       type=MovieScene2DTransformTrack
       sections [{
         type=MovieScene2DTransformSection
         end_frame=60
         start_frame=0
         channels [
          Translation.X {type=MovieSceneScriptingFloatChannel}
          Translation.Y {type=MovieSceneScriptingFloatChannel}
          Angle {type=MovieSceneScriptingFloatChannel}
          Scale.X {type=MovieSceneScriptingFloatChannel, keys [[9,4],[15,0.9],[19,1.05],[25,1]]}
          Scale.Y {type=MovieSceneScriptingFloatChannel, keys [[9,4],[15,0.9],[19,1.05],[25,1]]}
          Shear.X {type=MovieSceneScriptingFloatChannel}
          Shear.Y {type=MovieSceneScriptingFloatChannel}
         ]
        }]
      }
     ]
    }
    WinnerName {
     id=#12
     possessed_class=CommonTextBlock
     tracks [
      {
       type=MovieScene2DTransformTrack
       sections [{
         type=MovieScene2DTransformSection
         end_frame=60
         start_frame=0
         channels [
          Translation.X {type=MovieSceneScriptingFloatChannel, keys [[12,88],[21,0]]}
          Translation.Y {type=MovieSceneScriptingFloatChannel}
          Angle {type=MovieSceneScriptingFloatChannel}
          Scale.X {type=MovieSceneScriptingFloatChannel}
          Scale.Y {type=MovieSceneScriptingFloatChannel}
          Shear.X {type=MovieSceneScriptingFloatChannel}
          Shear.Y {type=MovieSceneScriptingFloatChannel}
         ]
        }]
      }
      {
       type=MovieSceneFloatTrack
       display_name=Render Opacity
       sections [{type=MovieSceneFloatSection, end_frame=60, start_frame=0, channels [None {type=MovieSceneScriptingFloatChannel, keys [[12,0],[21,1]]}]}]
      }
     ]
    }
    Backing {
     id=#13
     possessed_class=Image
     tracks [{
       type=MovieSceneWidgetMaterialTrack
       sections [{
         type=MovieSceneParameterSection
         end_frame=60
         start_frame=0
         channels [
          Pan {type=MovieSceneScriptingFloatChannel, keys [[15,1.25],[60,-0.25]]}
          AnimCenterAngle {type=MovieSceneScriptingFloatChannel, keys [[0,0],[18,0],[24,1.05],[30,1]]}
          L_Softness {type=MovieSceneScriptingFloatChannel, keys [[2,8],[6,144]]}
          R_Softness {type=MovieSceneScriptingFloatChannel, keys [[2,2],[5,128]]}
          AnimRightSide {type=MovieSceneScriptingFloatChannel, keys [[2,0.5],[6,1]]}
         ]
        }]
      }]
    }
    1 {
     id=#14
     possessed_class=CommonTextBlock
     tracks [
      {
       type=MovieSceneFloatTrack
       display_name=Render Opacity
       sections [{type=MovieSceneFloatSection, end_frame=59, start_frame=0, channels [None {type=MovieSceneScriptingFloatChannel, keys [[10,0],[19,1]]}]}]
      }
      {
       type=MovieScene2DTransformTrack
       sections [{
         type=MovieScene2DTransformSection
         end_frame=59
         start_frame=0
         channels [
          Translation.X {type=MovieSceneScriptingFloatChannel}
          Translation.Y {type=MovieSceneScriptingFloatChannel}
          Angle {type=MovieSceneScriptingFloatChannel}
          Scale.X {type=MovieSceneScriptingFloatChannel, keys [[10,0.25],[19,1.5],[24,0.95],[30,1]]}
          Scale.Y {type=MovieSceneScriptingFloatChannel, keys [[10,0.25],[19,1.75],[24,0.95],[30,1]]}
          Shear.X {type=MovieSceneScriptingFloatChannel}
          Shear.Y {type=MovieSceneScriptingFloatChannel}
         ]
        }]
      }
     ]
    }
    BackingBox {
     id=#15
     possessed_class=SizeBox
     tracks [
      {
       type=MovieSceneFloatTrack
       display_name=Render Opacity
       sections [{type=MovieSceneFloatSection, end_frame=60, start_frame=0, channels [None {type=MovieSceneScriptingFloatChannel, keys [[0,0],[3,1]]}]}]
      }
      {
       type=MovieScene2DTransformTrack
       sections [{
         type=MovieScene2DTransformSection
         end_frame=60
         start_frame=0
         channels [
          Translation.X {type=MovieSceneScriptingFloatChannel, keys [[0,256],[9,0]]}
          Translation.Y {type=MovieSceneScriptingFloatChannel}
          Angle {type=MovieSceneScriptingFloatChannel, const [0,1]}
          Scale.X {type=MovieSceneScriptingFloatChannel, keys [[0,1.75],[8,1]]}
          Scale.Y {type=MovieSceneScriptingFloatChannel, keys [[0,0.25],[6,0.25],[12,1.15],[18,0.975],[21,1]]}
          Shear.X {type=MovieSceneScriptingFloatChannel, keys [[0,-48],[12,0]]}
          Shear.Y {type=MovieSceneScriptingFloatChannel}
         ]
        }]
      }
     ]
    }
   ]
   display_rate {numerator=60}
  }
 }
 HideShowDisplay {
  binding_count=1
  length_seconds=0.3
  object=&self:HideShowDisplay
  sequence {bindings [ScoreboardDisplay {
     id=#16
     display_name=All
     possessed_class=VerticalBox
     tracks [{
       type=MovieSceneFloatTrack
       display_name=Render Opacity
       sections [{type=MovieSceneFloatSection, end_frame=6, start_frame=0, channels [None {type=MovieSceneScriptingFloatChannel, keys [[0,0],[6,1]]}]}]
      }]
    }]}
 }
 BackingFullOpacity {
  binding_count=2
  length_seconds=0.5
  object=&self:BackingFullOpacity
  sequence {bindings [
    Border_4 {
     id=#17
     possessed_class=Border
     tracks [{
       type=MovieSceneWidgetMaterialTrack
       sections [{
         type=MovieSceneParameterSection
         end_frame=10
         start_frame=0
         channels [
          FallOffAlpha {type=MovieSceneScriptingFloatChannel, keys [[0,0.85],[9.9997,1]]}
          HotSpotAlpha {type=MovieSceneScriptingFloatChannel, keys [[0,0.9],[9.9997,1]]}
          Color1R {type=MovieSceneScriptingFloatChannel, const [0,1]}
          Color1G {type=MovieSceneScriptingFloatChannel, keys [[0,0.0091],[10,0.0125]]}
          Color1B {type=MovieSceneScriptingFloatChannel, keys [[0,0.15],[10,0.3]]}
          Color1A {type=MovieSceneScriptingFloatChannel, const [1,1]}
          Color2R {type=MovieSceneScriptingFloatChannel, const [0,1]}
          Color2G {type=MovieSceneScriptingFloatChannel, keys [[0,0.0043],[10,0.0075]]}
          Color2B {type=MovieSceneScriptingFloatChannel, keys [[0,0.0353],[10,0.15]]}
          Color2A {type=MovieSceneScriptingFloatChannel, const [1,1]}
         ]
        }]
      }]
    }
    BackingTris {
     id=#18
     possessed_class=Image
     tracks [{
       type=MovieSceneFloatTrack
       display_name=Render Opacity
       sections [{type=MovieSceneFloatSection, end_frame=10, start_frame=0, channels [None {type=MovieSceneScriptingFloatChannel, keys [[0,0],[10,1]]}]}]
      }]
    }
   ]}
 }
]
graphs [
 AnimateBacking {nodes [
   AnimateBacking() {is=title K2Node_FunctionEntry}
   > PlayAnimation(InAnimation=Get BackingFullOpacity() -> (BackingFullOpacity:Widget Animation Object Reference))
  ]}
 EventGraph {nodes [
   n1 = ShowMinigameEndUI!() -> (OutputDelegate, then, ShowGameTotals:bool=~, ScoreboardState:EMinigameScoreboardStates Enum=RoundEndDisplayWinner, DisplayTime:double=~) {is=title K2Node_CustomEvent}
   > SetStatsColumns()
   > UpdateStats!(execute, self, ShowGameTotals:bool=@n1.ShowGameTotals, ScoreboardState:EMinigameScoreboardStates Enum=@n1.ScoreboardState) -> (then)
   > UpdateRoundDisplay!(execute, self, GameEnd:bool=@n1.ShowGameTotals, ScoreboardState:EMinigameScoreboardStates Enum=@n1.ScoreboardState) -> (then)
   > "SetText (Text)"(
    self=Get CountdownText() -> (CountdownText:Common Text Object Reference)
    InText=Format Text!(Format="LOCTEXT(\"{0}\")", 0:int=Round(A=@n1.DisplayTime) {is=title K2Node_CallFunction}) -> (Result)
   )
   > ShowTargetPlayerStateInScrollBox()
   HideMinigameEndUI() {is=title K2Node_CustomEvent}
   > SetVisibility(self=Self-Reference(), InVisibility=Hidden)
   UpdateActiveStats() {is=title K2Node_CustomEvent}
   > UpdateStats(ShowGameTotals=false, ScoreboardState=Max)
   n2 = FullscreenMapToggle!() -> (OutputDelegate, then, bFullscreenMapVisible:bool=~) {is=title K2Node_CustomEvent}
   > n3 = Branch(Condition=@n2.bFullscreenMapVisible)
   > n4 = Set Timer by Event(Delegate=Create Event(), Time=Get StatUpdateRate() -> (StatUpdateRate:double), bLooping=true)
   > Set UpdateDisplayTimerHandle(UpdateDisplayTimerHandle:Timer Handle Structure=@n4) -> (Output_Get:Timer Handle Structure)
   > SetStatsColumns()
   > UpdateStats(ShowGameTotals=false, ScoreboardState=Max)
   > UpdateRoundDisplay(GameEnd=false, ScoreboardState=RoundEndDisplayWinner)
   > ShowTargetPlayerStateInScrollBox()
   Clear and Invalidate Timer by Handle(execute=@n3.else, Handle=Get UpdateDisplayTimerHandle() -> (UpdateDisplayTimerHandle:Timer Handle Structure))
   Event DoShowUI() {is=title K2Node_Event}
   > Set bInGameScoreboard(bInGameScoreboard:bool=true) -> (Output_Get:bool)
   > SetVisibility(self=Self-Reference(), InVisibility=SelfHitTestInvisible)
   > SetVisibility(self=Get TransitionInfo() -> (TransitionInfo:Horizontal Box Object Reference), InVisibility=Collapsed)
   > n5 = Cast To FortPlayerController(Object=GetOwningPlayer())
   n6 = GetMiniMapComponent!(self:Fort Player Controller Object Reference=@n5) -> (ReturnValue:Fort Controller Component Minimap Object Reference)
   > Is Valid(InputObject=@n6)
   > Bind Event to On Toggle Fullscreen Map(self:Fort Controller Component Minimap Object Reference=@n6, Delegate=Create Event()) {is=title K2Node_AddDelegate}
   > n7 = Set Timer by Event(Delegate=Create Event(), Time=Get StatUpdateRate(), bLooping=true)
   > Set UpdateDisplayTimerHandle(UpdateDisplayTimerHandle=@n7)
   > SetStatsColumns()
   > UpdateStats(ShowGameTotals=false, ScoreboardState=Max)
   > n8 = Get!(execute:Exec) -> (then:Exec, else:Exec, CurrentMinigame:Fort Minigame Object Reference)
   > Bind Event to Minigame Ended Delegate(self:Fort Minigame Object Reference=@n8, Delegate=Create Event()) {is=title K2Node_AddDelegate}
   > UpdateRoundDisplay(GameEnd=false, ScoreboardState=RoundEndDisplayWinner)
   > ShowTargetPlayerStateInScrollBox()
   Event DoHideUI() {is=title K2Node_Event}
   > SetVisibility(InVisibility=Collapsed)
   n9 = Event OnDisplayMinigameEndUI!() -> (OutputDelegate, then, bDisplay:bool) {is=title K2Node_Event}
   > Parent: OnDisplayMinigameEndUI!(execute, bDisplay:bool=@n9.bDisplay) -> (then) {is=title K2Node_CallParentFunction}
   > n10 = Get()
   > n11 = Switch on EFortMinigameState!(
    execute
    Selection:EFortMinigameState Enum=Get CurrentState!(self:Fort Minigame Object Reference=@n10) -> (CurrentState:EFortMinigameState Enum=PreGame)
   ) -> (
    PreGame:Exec
    Setup:Exec
    Transitioning:Exec
    WaitingForCameras:Exec
    Warmup:Exec
    InProgress:Exec
    PostGameTimeDilation:Exec
    PostRoundEnd:Exec
    PostGameEnd:Exec
    PostGameAbandon:Exec
    PostGameReset:Exec
    Teardown:Exec
   ) {is=title K2Node_SwitchEnum}
   Call On Minigame End!(execute:Exec=@n11.PostGameEnd, self:Scoreboard UI Object Reference) -> (then:Exec) {is=title K2Node_CallDelegate}
   n12 = Event PreConstruct()
   > Branch(Condition=@n12.IsDesignTime)
   n13 = Get TestData() -> (TestData:Array of Minigame Scoreboard Player Row Structures)
   > n14 = For Each Loop(Array:Array of Minigame Scoreboard Player Row Structures=@n13) -> (Array Element:Minigame Scoreboard Player Row Structure)
   > n15 = MakePlayerScoreboardRow!(
    execute
    self
    PlayerRow:Minigame Scoreboard Player Row Structure=@n14
    PlayerIndex:int=@n14.Array Index
    ShowWins:bool=Get IsRoundsGame() -> (IsRoundsGame:bool)
    NumPlayers:int=Length(TargetArray:Array of Minigame Scoreboard Player Row Structures=@n13)
    Standing:int=@n14.Array Index
   ) -> (then, ScoreboardRow:Scoreboard UI Row Object Reference)
   > AddChild(self=Get DataScrollBox() -> (DataScrollBox:Scroll Box Object Reference), Content=@n15)
   Event OnDisplayLeavingVolumeUI!() -> (OutputDelegate, then, bDisplay:bool) {is=title K2Node_Event}
   > Call On Minigame End()
   Event Destruct()
   > n16 = Cast To FortPlayerController(Object=GetOwningPlayer())
   n17 = GetMiniMapComponent(self=@n16)
   > Is Valid(InputObject=@n17)
   > Unbind Event from On Toggle Fullscreen Map(self:Fort Controller Component Minimap Object Reference=@n17, Delegate=Create Event()) {is=title K2Node_RemoveDelegate}
  ]}
 GetPlayerRowColor {nodes [
   n1 = Get TeamColors() -> (TeamColors:Array of Linear Color Structures)
   n2 = Length(TargetArray:Array of Linear Color Structures=@n1)
   n3 = Inputs!() -> (PlayerIndex:int)
   Outputs!(RowColor:Linear Color Structure="Get (a ref)"(Array:Array of Linear Color Structures=@n1, Dimension 1="% (Integer)"(A=@n3, B=@n2) -> (ReturnValue=0)) -> (Output:"Linear Color Structure (by ref)"))
  ]}
 GetScoreboardStateNames {nodes [
   GetScoreboardStateNames() {is=title K2Node_FunctionEntry}
   > n1 = Get()
   n2 = GetMinigameStatsComponent!(ContextActor:Actor Object Reference=Get Volume!(self:Fort Minigame Object Reference=@n1) -> (Volume:Actor Object Reference)) -> (ReturnValue:Fort Minigame Stats Component Object Reference)
   > Is Valid(InputObject=@n2)
   > n3 = For Each Loop(Array:Array of Fort Minigame Stat Filter Object References=GetTrackedStatsForScoreboardColumns!(self:Fort Minigame Stats Component Object Reference=@n2, bOnlyPersistentStats:bool=Get bIsCareerScoreboard() -> (bIsCareerScoreboard:bool)) -> (OutScoreboardColumnSortedTrackedStats:Array of Fort Minigame Stat Filter Object References)) -> (Array Element:Fort Minigame Stat Filter Object Reference)
   Return Node!(execute=@n3.Completed, OutNames:Array of Texts=Get NamesArray() -> (NamesArray:Array of Texts))
   > Add(
    TargetArray:Array of Texts=Get NamesArray()
    NewItem:"Text (by ref)"=Get HudShortDescription!(self:Fort Minigame Stat Filter Object Reference=@n3) -> (HudShortDescription:Text)
   ) {is=title K2Node_CallArrayFunction}
  ]}
 GetTargetPlayerState {nodes [
   GetTargetPlayerState() {is=title K2Node_FunctionEntry}
   > n1 = Set OwningPlayer(OwningPlayer:Player Controller Object Reference=GetOwningPlayer()) -> (Output_Get:Player Controller Object Reference)
   > n2 = Is Valid(InputObject=@n1)
   Return Node!(execute=@n2.Is Not Valid, PlayerState:Player State Object Reference)
   > n3 = Set SpectatorPawn(SpectatorPawn:Spectator Pawn Object Reference=GetSpectatorPawn!(self:Player Controller Object Reference=Get OwningPlayer() -> (OwningPlayer:Player Controller Object Reference)) -> (ReturnValue:Spectator Pawn Object Reference)) -> (Output_Get:Spectator Pawn Object Reference)
   > n4 = Is Valid(InputObject=@n3)
   > Return Node(PlayerState=Get PlayerState!(self:Pawn Object Reference=Get SpectatorPawn() -> (SpectatorPawn:Spectator Pawn Object Reference)) -> (PlayerState:Player State Object Reference))
   Return Node(execute=@n4.Is Not Valid, PlayerState=Get PlayerState(self:Controller Object Reference=Get OwningPlayer()))
  ]}
 GetTargetScoreRow {nodes [
   GetTargetScoreRow!() -> (then, TargetPlayerState:Player State Object Reference) {is=title K2Node_FunctionEntry}
   > n1 = For Each Loop(Array:Array of Widget Object References=GetAllChildren(self=Get DataScrollBox())) -> (Array Element:Widget Object Reference)
   Return Node!(execute=@n1.Completed, Widget:Widget Object Reference)
   > n2 = Branch(Condition=Get bIsTeamGame() -> (bIsTeamGame:bool))
   > n3 = Cast To Scoreboard_Team_Widget!(execute, Object=@n1) -> (then, CastFailed, AsScoreboard Team Widget:Scoreboard Team Widget Object Reference) {is=title K2Node_DynamicCast}
   > n4 = GetWidgetForPlayerState!(
    execute
    self:Scoreboard Team Widget Object Reference=@n3
    PlayerState:Player State Object Reference=Get TargetPlayerState() -> (TargetPlayerState:Player State Object Reference)
   ) -> (then, Widget:Widget Object Reference)
   > Is Valid(InputObject=@n4)
   > Return Node(Widget=@n3)
   n5 = Cast To Scoreboard_UI_Row!(execute=@n2.else, Object=@n1) -> (then, CastFailed, AsScoreboard UI Row:Scoreboard UI Row Object Reference) {is=title K2Node_DynamicCast}
   n6 = Get PlayerBucketData!(self:Scoreboard UI Row Object Reference=@n5) -> (PlayerBucketData:Minigame Scoreboard Player Row Structure)
   > Branch(Condition="Equal (Object)"(
     A=Get TargetPlayerState()
     B=Break Minigame Scoreboard Player Row!(MinigameScoreboardPlayerRow:"Minigame Scoreboard Player Row Structure (by ref)"=@n6) -> (PlayerState:Fort Player State Object Reference, bHasLeftGame:bool=~, bHideFromDisplay:bool=~) {is=title K2Node_BreakStruct}
    ) {is=title K2Node_CallFunction})
   > Return Node(Widget=@n5)
  ]}
 GetTeamColor {nodes [
   n1 = Inputs!() -> (BucketNum:int)
   n2 = Get TeamColors()
   Outputs!(Color:Linear Color Structure=Select(
     Option 0:"Linear Color Structure (by ref)"=Make Linear Color!(R:float=0, G:float=0.6667, B:float=1, A:float=1) -> (LinearColor:Linear Color Structure) {is=title K2Node_MakeStruct}
     Option 1:"Linear Color Structure (by ref)"="Get (a ref)"(Array=@n2, Dimension 1=@n1)
     Index=Is Valid Index!(TargetArray:Array of Linear Color Structures=@n2, IndexToTest:int=@n1) -> (ReturnValue:bool)
    ) -> (ReturnValue:"Linear Color Structure (by ref)"))
  ]}
 HandleScoreboardColumnsData {nodes [
   n1 = HandleScoreboardColumnsData!() -> (then, ScoreBoardStatNames:Array of Texts, bTeamGame:bool=~, bDisplayingWins:bool=~) {is=title K2Node_FunctionEntry}
   > n2 = Is Valid(InputObject=Get ScoreboardColumnWidget() -> (ScoreboardColumnWidget:Scoreboard UI Column Labels Object Reference))
   n3 = Create Widget!(
    execute:Exec=@n2.Is Not Valid
    Class:User Widget Class Reference=&/Scoreboard_UI_ColumnLabels.Scoreboard_UI_ColumnLabels_C
    OwningPlayer:Player Controller Object Reference=GetOwningPlayer()
    DataText:Array of Texts=@n1
    bIsTeamGame:bool=@n1.bTeamGame
    bDisplayingWins:bool=@n1.bDisplayingWins
    bIngameScoreboard:bool=Get bInGameScoreboard() -> (bInGameScoreboard:bool)
    WinCondition:EMinigameWinCondition Enum=Get WinCondition() -> (WinCondition:EMinigameWinCondition Enum=MostRoundWins)
    bShowCumulativeScoreboard:bool=Get bShowCumulativeScoreboard() -> (bShowCumulativeScoreboard:bool)
   ) -> (then:Exec, ReturnValue:Scoreboard UI Column Labels Object Reference)
   > n4 = Set ScoreboardColumnWidget(ScoreboardColumnWidget:Scoreboard UI Column Labels Object Reference=@n3) -> (Output_Get:Scoreboard UI Column Labels Object Reference)
   > AddChild(self=Get ColumnHolder() -> (ColumnHolder:Border Object Reference), Content=@n4)
   n5 = Get ScoreboardColumnWidget()
   Set DataText!(
    execute=@n2.Is Valid
    DataText:Array of Texts=Get ScoreBoardStatNames() -> (ScoreBoardStatNames:Array of Texts)
    self:Scoreboard UI Column Labels Object Reference=@n5
   ) -> (then, Output_Get:Array of Texts)
   > Set bIsTeamGame!(execute, bIsTeamGame:bool=Get bTeamGame() -> (bTeamGame:bool), self:Scoreboard UI Column Labels Object Reference=@n5) -> (then, Output_Get:bool)
   > Set bDisplayingWins!(execute, bDisplayingWins:bool=Get bDisplayingWins() -> (bDisplayingWins:bool), self:Scoreboard UI Column Labels Object Reference=@n5) -> (then, Output_Get:bool)
   > Set WinCondition!(
    execute
    WinCondition:EMinigameWinCondition Enum=Get WinCondition() -> (WinCondition=MostRoundWins)
    self:Scoreboard UI Column Labels Object Reference=@n5
   ) -> (then, Output_Get:EMinigameWinCondition Enum=MostRoundWins)
   > Set bShowCumulativeScoreboard!(execute, bShowCumulativeScoreboard:bool=Get bShowCumulativeScoreboard(), self:Scoreboard UI Column Labels Object Reference=@n5) -> (then, Output_Get:bool)
   > UpdateColumnDisplay(self:Scoreboard UI Column Labels Object Reference=@n5)
  ]}
 HideUnusedRows {nodes [
   HideUnusedRows!() -> (then, bTeamGame:bool=~, NumBuckets:int) {is=title K2Node_FunctionEntry}
   > n1 = Sequence()
   > n2 = For Each Loop(Array:Array of Scoreboard Team Widget Object References=Get TeamScoreDisplays() -> (TeamScoreDisplays:Array of Scoreboard Team Widget Object References)) -> (Array Element:Scoreboard Team Widget Object Reference)
   n3 = "integer < integer"!(A:int=@n2.Array Index, B:int=Get NumBuckets() -> (NumBuckets:int=0)) -> (ReturnValue:bool) {is=title K2Node_CallFunction}
   n4 = Get bIsTeamGame()
   > SetVisibility(
    self=@n2
    InVisibility=Select(Option 0:ESlateVisibility Enum=Collapsed, Option 1:ESlateVisibility Enum=HitTestInvisible, Index=AND Boolean(A=@n3, B=@n4)) -> (ReturnValue:ESlateVisibility Enum)
   )
   n5 = For Each Loop(
    Exec=@n1.then_1
    Array:Array of Scoreboard UI Row Object References=Get PlayerScoreDisplays() -> (PlayerScoreDisplays:Array of Scoreboard UI Row Object References)
   ) -> (Array Element:Scoreboard UI Row Object Reference)
   n6 = "integer < integer"(A=@n5.Array Index, B=Get NumBuckets() -> (NumBuckets=0))
   n7 = NOT Boolean(A=Get bIsTeamGame())
   > SetVisibility(self=@n5, InVisibility=Select(Option 0=Collapsed, Option 1=HitTestInvisible, Index=AND Boolean(A=@n6, B=@n7)))
  ]}
 HideWinnerDisplay {nodes [
   HideWinnerDisplay() {is=title K2Node_FunctionEntry}
   > SetVisibility(self=Get WinnerDisplay() -> (WinnerDisplay:Overlay Object Reference), InVisibility=Hidden)
  ]}
 IntroAnim {nodes [
   IntroAnim() {is=title K2Node_FunctionEntry}
   > PlayAnimation(InAnimation=Get Intro())
  ]}
 IsMinigameTeamBased {nodes [
   n1 = IsMinigameTeamBased!() -> (then, MinigameToCheck:Fort Minigame Object Reference, MinigameTeamComponent:Fort Minigame Team Component Object Reference) {is=title K2Node_FunctionEntry}
   n2 = Get bIsCareerScoreboard()
   n3 = Get bForceIndividualScoresInTeamGame!(self:Fort Minigame Scoreboard Component Object Reference=Get ScoreboardComponent!(self:Fort Minigame Object Reference=@n1) -> (ScoreboardComponent:Fort Minigame Scoreboard Component Object Reference)) -> (bForceIndividualScoresInTeamGame:bool)
   > Return Node!(
    execute
    ReturnValue:bool=AND Boolean!(
     A=NOT Boolean(A=@n3)
     B=GetTeamMinigame!(self:Fort Minigame Team Component Object Reference=@n1.MinigameTeamComponent) -> (ReturnValue:bool)
     C:bool=NOT Boolean(A=@n2)
    ) -> (ReturnValue)
   )
  ]}
 MakePlayerScoreboardRow {nodes [
   n1 = MakePlayerScoreboardRow!() -> (then, PlayerRow:Minigame Scoreboard Player Row Structure, PlayerIndex:int, ShowWins:bool=~, NumPlayers:int, Standing:int) {is=title K2Node_FunctionEntry}
   > n2 = Create Widget!(
    execute
    Class=&/Scoreboard_UI_Row.Scoreboard_UI_Row_C
    OwningPlayer=GetOwningPlayer()
    bDisplayWins:bool=@n1.ShowWins
    PlayerBucketData:Minigame Scoreboard Player Row Structure=@n1
    bIngameScorerow:bool=Get bInGameScoreboard()
    bIsTeamGame=Get bIsTeamGame()
    MedalStanding:int=@n1.Standing
    TeamColorParam:double=0
    WinCondition=Get WinCondition() -> (WinCondition=MostRoundWins)
    bShowCumulativeScoreboard=Get bShowCumulativeScoreboard()
   ) -> (then, ReturnValue:Scoreboard UI Row Object Reference)
   > Return Node!(execute, ScoreboardRow:Scoreboard UI Row Object Reference=@n2)
  ]}
 MakeTeamScoreboardRow {nodes [
   n1 = MakeTeamScoreboardRow!() -> (then, MinigameScoreboardBucketRow:"Minigame Scoreboard Bucket Row Structure (by ref)", IsRoundsGame:bool=~) {is=title K2Node_FunctionEntry}
   > n2 = Create Widget!(
    execute
    Class=/Game/Scoreboard_Team_Widget.Scoreboard_Team_Widget_C
    OwningPlayer=GetOwningPlayer()
    bShowVictoryScore:bool=Get IsRoundsGame()
    TeamScoreboardBucket:Minigame Scoreboard Bucket Row Structure=Get MinigameScoreboardBucketRow() -> (MinigameScoreboardBucketRow:"Minigame Scoreboard Bucket Row Structure (by ref)")
    bIngameTeamScore:bool=Get bInGameScoreboard()
    TeamColorParam="To Float (Integer)"(InInt=Break Minigame Scoreboard Bucket Row!(MinigameScoreboardBucketRow:"Minigame Scoreboard Bucket Row Structure (by ref)"=@n1) -> (
      Standing:int
      BucketIndex:int
      PlayerRows:Array of Minigame Scoreboard Player Row Structures
      Name:Text
      TeamColorIndex:int
      Values:Array of Minigame Scoreboard Value Structures
      WinCount:int
      bHighlight:bool=~
     ) {is=title K2Node_BreakStruct}.TeamColorIndex)
   ) -> (then, ReturnValue:Scoreboard Team Widget Object Reference)
   > Return Node!(execute, ReturnValue:Scoreboard Team Widget Object Reference=@n2)
  ]}
 OnBlockMapActionByKey {nodes [
   OnBlockMapActionByKey!() -> (then, InKey:"Key Structure (by ref)") {is=title K2Node_FunctionEntry}
   > Return Node!(execute, bOutMapActionBlocked:bool=false, ReturnValue:bool=false)
  ]}
 OnHandleMapActionKey {nodes [
   OnHandleMapActionKey!() -> (then, InKey:"Key Structure (by ref)", bKeyDown:bool) {is=title K2Node_FunctionEntry}
   > Return Node!(execute, bOutMapActionHandled:bool=false, ReturnValue=false)
  ]}
 OnHandleMapBackAction {nodes [
   OnHandleMapBackAction() {is=title K2Node_FunctionEntry}
   > Return Node!(execute, bOutBackActionHandled:bool=false, ReturnValue=false)
  ]}
 OnMinigameEnd {nodes [OnMinigameEnd() {is=title K2Node_FunctionEntry}]}
 OnShouldCreateMapScreenTab {nodes [
   OnShouldCreateMapScreenTab!() -> (then, WorldContextObject:Object Reference) {is=title K2Node_FunctionEntry}
   > Return Node!(execute, bOutShouldCreateMapScreenTab:bool=false, ReturnValue=false)
  ]}
 OnShouldDisplayBackActionInActionBar {nodes [
   OnShouldDisplayBackActionInActionBar() {is=title K2Node_FunctionEntry}
   > Return Node!(execute, bOutShouldDisplayBackActionInActionBar:bool=false, ReturnValue=false)
  ]}
 OnShouldShowTabOnMapScreenForMinigame {nodes [
   OnShouldShowTabOnMapScreenForMinigame!() -> (then, Minigame:Fort Minigame Object Reference) {is=title K2Node_FunctionEntry}
   > Return Node!(execute, bOutShouldShowTabOnMapScreenForMinigame:bool=false, ReturnValue=false)
  ]}
 ResetScoreDisplays {nodes [
   ResetScoreDisplays() {is=title K2Node_FunctionEntry}
   > n1 = For Each Loop(Array:Array of Scoreboard Team Widget Object References=Get TeamScoreDisplays()) -> (Array Element:Scoreboard Team Widget Object Reference)
   Clear(execute=@n1.Completed, TargetArray:Array of Scoreboard Team Widget Object References=Get TeamScoreDisplays())
   > Clear(TargetArray:Array of Scoreboard UI Row Object References=Get PlayerScoreDisplays())
   > ClearChildren(self=Get DataScrollBox())
   > Clear and Invalidate Timer by Handle(Handle=Get UpdateDisplayTimerHandle())
   ClearTeamPlayers(execute=@n1.LoopBody, self:Scoreboard Team Widget Object Reference=@n1)
  ]}
 SetCountdownText {nodes [
   n1 = SetCountdownText!() -> (then, Text:Text) {is=title K2Node_FunctionEntry}
   > "SetText (Text)"(self=Get CountdownText(), InText=@n1)
  ]}
 SetIsCareerScoreboard {nodes [
   n1 = SetIsCareerScoreboard!() -> (then, bInValue:bool=~) {is=title K2Node_FunctionEntry}
   > n2 = Set bIsCareerScoreboard(bIsCareerScoreboard:bool=@n1) -> (Output_Get:bool)
   > SetVisibility(
    self=Get Header() -> (Header:Horizontal Box Object Reference)
    InVisibility=Select(Option 0=SelfHitTestInvisible, Option 1=Collapsed, Index=@n2)
   )
  ]}
 SetStatsColumns {nodes [
   SetStatsColumns() {is=title K2Node_FunctionEntry}
   > Set ScoreboardStatNames(ScoreboardStatNames:Array of Texts=GetScoreboardStateNames!() -> (OutNames:Array of Texts) {is=title K2Node_CallFunction}) -> (Output_Get:Array of Texts)
   > n1 = Get()
   n2 = GetMinigameTeamComponent!(self:Fort Minigame Object Reference=@n1) -> (ReturnValue:Fort Minigame Team Component Object Reference)
   > Is Valid(InputObject=@n2)
   > n3 = Set bIsTeamGame!(
    execute
    bIsTeamGame=IsMinigameTeamBased!(MinigameToCheck:Fort Minigame Object Reference=@n1, MinigameTeamComponent:Fort Minigame Team Component Object Reference=@n2) -> (ReturnValue:bool) {is=title K2Node_CallFunction}
   ) -> (then, Output_Get)
   > Set IsRoundsGame(IsRoundsGame:bool="integer > integer"(A=Get TotalRounds!(self:Fort Minigame Object Reference=@n1) -> (TotalRounds:int=0), B=1) {is=title K2Node_CallFunction}) -> (Output_Get:bool)
   > Set WinCondition!(execute, WinCondition=Get WinCondition!(self:Fort Minigame Object Reference=@n1) -> (WinCondition=MostRoundWins)) -> (then, Output_Get=MostRoundWins)
   > Set bShowCumulativeScoreboard!(
    execute
    bShowCumulativeScoreboard=Get bShowCumulativeScoreboard!(self:Fort Minigame Scoreboard Component Object Reference=Get ScoreboardComponent(self=@n1)) -> (bShowCumulativeScoreboard)
   ) -> (then, Output_Get)
   > HandleScoreboardColumnsData!(
    execute
    self
    ScoreBoardStatNames:Array of Texts=Get ScoreboardStatNames() -> (ScoreboardStatNames:Array of Texts)
    bTeamGame:bool=@n3
    bDisplayingWins:bool=ShowWinsColumn!() -> (ReturnValue:bool)
   ) -> (then) {is=title K2Node_CallFunction}
  ]}
 ShowTargetPlayerStateInScrollBox {nodes [
   ShowTargetPlayerStateInScrollBox() {is=title K2Node_FunctionEntry}
   > n1 = GetTargetPlayerState!(execute, self) -> (then, PlayerState:Player State Object Reference) {is=title K2Node_CallFunction}
   > Is Valid(InputObject=@n1)
   > n2 = GetTargetScoreRow!(execute, self, TargetPlayerState:Player State Object Reference=@n1) -> (then, Widget:Widget Object Reference) {is=title K2Node_CallFunction}
   > Is Valid(InputObject=@n2)
   > ScrollWidgetIntoView!(
    execute
    self:Scroll Box Object Reference=Get DataScrollBox()
    WidgetToFind:Widget Object Reference=@n2
    AnimateScroll:bool=true
    ScrollDestination:EDescendantScrollDestination Enum=IntoView
    Padding:float=0
   ) -> (then)
  ]}
 ShowWinnerDisplay {nodes [
   n1 = ShowWinnerDisplay!() -> (then, Winner Name:Text, Win Description:Text, GameHasWinner:bool=~) {is=title K2Node_FunctionEntry}
   > n2 = Branch(Condition=@n1.GameHasWinner)
   > "SetText (Text)"(self=Get WinnerName() -> (WinnerName:Common Text Object Reference), InText=@n1)
   > "SetText (Text)"(self=Get WinDescription() -> (WinDescription:Common Text Object Reference), InText=@n1.Win Description)
   > SetContentColorAndOpacity!(
    execute
    self:Border Object Reference=Get WinDescriptionColor() -> (WinDescriptionColor:Border Object Reference)
    InContentColorAndOpacity:Linear Color Structure=(R=0.5029,G=0.117,B=0.0123,A=1)
   ) -> (then)
   > SetVisibility(self=Get PlacementMarker() -> (PlacementMarker:Overlay Object Reference), InVisibility=SelfHitTestInvisible)
   > n3 = GetDynamicMaterial(self=Get Backing() -> (Backing:Image Object Reference))
   > SetVectorParameterValue(self=@n3, ParameterName=Color1, Value=(R=0.9823,G=0.5583,B=0.0482,A=1))
   > n4 = SetVectorParameterValue(self=@n3, ParameterName=Color2, Value=(R=0.9387,G=0.2623,B=0.0395,A=1))
   SetVisibility(execute=[@n4.then, @n6.then], self=Get WinnerDisplay(), InVisibility=SelfHitTestInvisible)
   > PlayAnimation(InAnimation=Get WinnerDisplayIntro() -> (WinnerDisplayIntro:Widget Animation Object Reference))
   "SetText (Text)"(execute=@n2.else, self=Get WinnerName(), InText=@n1.Win Description)
   > SetVisibility(self=Get WinDescription(), InVisibility=Collapsed)
   > SetContentColorAndOpacity(self=Get WinnerNameColor() -> (WinnerNameColor:Border Object Reference), InContentColorAndOpacity=(R=0,G=0.0125,B=0.15,A=1))
   > SetVisibility(self=Get PlacementMarker(), InVisibility=Collapsed)
   > n5 = GetDynamicMaterial(self=Get Backing())
   > SetVectorParameterValue(self=@n5, ParameterName=Color1, Value=(R=0,G=0.0578,B=0.6038,A=1))
   > n6 = SetVectorParameterValue(self=@n5, ParameterName=Color2, Value=(R=0,G=0.0578,B=0.6038,A=1))
  ]}
 ShowWinsColumn {nodes [
   ShowWinsColumn() {is=title K2Node_FunctionEntry}
   > n1 = Branch(Condition=Get IsRoundsGame())
   > n2 = Branch(Condition=Get bIsCareerScoreboard())
   > n3 = Get()
   n4 = Get Volume(self=@n3)
   > Return Node!(
    execute
    ReturnValue=IsStatPersistent!(
     self:Fort Minigame Stats Component Object Reference=GetMinigameStatsComponent(ContextActor=@n4)
     InFilterClass:Fort Minigame Stat Filter Class Reference=/Game/Athena/Playset/Minigames/StatFilters/MinigameStat_RoundWins.MinigameStat_RoundWins_C
    ) -> (ReturnValue:bool)
   )
   Return Node(execute=@n3.else, ReturnValue=false)
   Return Node(execute=@n2.else, ReturnValue=true)
   Return Node(execute=@n1.else, ReturnValue=false)
  ]}
 UpdateRoundDisplay {nodes [
   n1 = UpdateRoundDisplay!() -> (then, GameEnd:bool=~, ScoreboardState:EMinigameScoreboardStates Enum=RoundEndDisplayWinner) {is=title K2Node_FunctionEntry}
   > Branch(Condition=Get IsRoundsGame())
   n2 = Get CurrentMinigame() -> (CurrentMinigame:Fort Minigame Object Reference)
   n3 = Get CurrentRound!(self:Fort Minigame Object Reference=@n2) -> (CurrentRound:int=0)
   > "SetText (Text)"(
    self=Get GameStats() -> (GameStats:Common Text Object Reference)
    InText=Select(
     Option 0:Text=Format Text!(Format="LOCTEXT(\"ROUND {RndNum}\")", RndNum:int=@n3) -> (Result)
     Option 1:Text="LOCTEXT(\"FINAL SCORES\")"
     Index=@n1
    ) -> (ReturnValue:Text)
   )
   n4 = Get RoundsCountTotal() -> (RoundsCountTotal:Common Text Object Reference)
   > "SetText (Text)"(self=@n4, InText=Format Text(Format="LOCTEXT(\"OF {RndNum}\")", RndNum=Get TotalRounds(self=@n2) -> (TotalRounds=0)))
   > SetVisibility(
    self=@n4
    InVisibility=Select(Option 0:ESlateVisibility Enum=SelfHitTestInvisible, Option 1:ESlateVisibility Enum=Hidden, Index=@n1) -> (ReturnValue:ESlateVisibility Enum)
   )
   n5 = IsLastRound!(self:Fort Minigame Object Reference=Get CurrentMinigame()) -> (ReturnValue:bool)
   n6 = Get bInGameScoreboard()
   > n7 = Branch(Condition=AND Boolean!(A=NOT Boolean(A=@n6), B=NOT Boolean(A=@n5)) -> (ReturnValue))
   n8 = Get CountdownHeaderSmall() -> (CountdownHeaderSmall:Common Text Object Reference)
   > "SetText (Text)"(self=@n8, InText="LOCTEXT(\"Next Round\")")
   > SetVisibility(self=@n8, InVisibility=SelfHitTestInvisible)
   n9 = Get CountdownHeaderlarge() -> (CountdownHeaderlarge:Common Text Object Reference)
   > "SetText (Text)"(self=@n9, InText="LOCTEXT(\"Starts in\")")
   "SetText (Text)"(
    execute=@n7.else
    self=@n9
    InText=Select!(
     RoundEndDisplayWinner:Text
     GameEndDisplayWinner:Text
     RoundEndDisplayScoreboard:Text="LOCTEXT(\"Game Results In\")"
     GameEndDisplayScoreboard:Text="LOCTEXT(\"Game Ending In\")"
     Max:Text
     Index:EMinigameScoreboardStates Enum=@n1.ScoreboardState
    ) -> (ReturnValue:Text)
   )
  ]}
 UpdateStats {nodes [
   n1 = UpdateStats!() -> (then, ShowGameTotals:bool=~, ScoreboardState:EMinigameScoreboardStates Enum=RoundEndDisplayWinner) {is=title K2Node_FunctionEntry}
   > n2 = Get()
   > n3 = Get!(execute, self:Fort Minigame Object Reference=@n2) -> (then, else, ScoreboardComponent:Fort Minigame Scoreboard Component Object Reference)
   n4 = GetMinigameStatsComponent(ContextActor=Get Volume(self=@n2))
   > Is Valid(InputObject=@n4)
   n5 = Get bIsCareerScoreboard()
   > n6 = Set ScoreboardBucketRows(ScoreboardBucketRows:Array of Minigame Scoreboard Bucket Row Structures=GetColumnSortedScoreboardRows!(
     self:Fort Minigame Scoreboard Component Object Reference=@n3
     RowSortOrder:Array of Fort Minigame Stat Filter Object References=Get ScoreboardStatObjects!(self:Fort Minigame Stats Component Object Reference=@n4) -> (ScoreboardStatObjects:Array of Fort Minigame Stat Filter Object References)
     ColumnSortOrder:Array of Fort Minigame Stat Filter Object References=GetTrackedStatsForScoreboardColumns(self=@n4, bOnlyPersistentStats=@n5)
     bGameTotal:bool=OR Boolean(A=@n5, B=@n1)
     bHighlighting:bool=true
     bCountWins:bool=true
     bIsCareerScoreboard:bool=Get bIsCareerScoreboard()
     ScoreboardState:EMinigameScoreboardStates Enum=@n1.ScoreboardState
    ) -> (OutScoreboardRows:Array of Minigame Scoreboard Bucket Row Structures)) -> (Output_Get:Array of Minigame Scoreboard Bucket Row Structures)
   > n7 = Set NumBuckets(NumBuckets:int=Length(TargetArray:Array of Minigame Scoreboard Bucket Row Structures=@n6)) -> (Output_Get:int=0)
   > HideUnusedRows!(execute, self, bTeamGame:bool=Get bIsTeamGame(), NumBuckets:int=@n7) -> (then) {is=title K2Node_CallFunction}
   > n8 = For Each Loop(Array:Array of Minigame Scoreboard Bucket Row Structures=@n6) -> (Array Element:Minigame Scoreboard Bucket Row Structure)
   > Set CurrentBucket(CurrentBucket:Minigame Scoreboard Bucket Row Structure=@n8) -> (Output_Get:Minigame Scoreboard Bucket Row Structure)
   > Set CurrentIndex(CurrentIndex:int=@n8.Array Index) -> (Output_Get:int=0)
   > n9 = Branch(Condition=Get bIsTeamGame())
   > n10 = Branch(Condition=Is Valid Index(
     TargetArray:Array of Scoreboard Team Widget Object References=Get TeamScoreDisplays()
     IndexToTest=Get CurrentIndex() -> (CurrentIndex:int=0)
    ))
   > UpdateTeamData!(
    execute
    self:Scoreboard Team Widget Object Reference="Get (a copy)"(Array:Array of Scoreboard Team Widget Object References=Get TeamScoreDisplays(), Dimension 1=Get CurrentIndex() -> (CurrentIndex=0)) -> (Output:Scoreboard Team Widget Object Reference)
    TeamBucketData:Minigame Scoreboard Bucket Row Structure=Get CurrentBucket() -> (CurrentBucket:Minigame Scoreboard Bucket Row Structure)
    ShowVictoryScore:bool=Get IsRoundsGame()
   ) -> (then)
   n11 = MakeTeamScoreboardRow!(
    execute=@n10.else
    self
    MinigameScoreboardBucketRow:"Minigame Scoreboard Bucket Row Structure (by ref)"=Get CurrentBucket()
    IsRoundsGame:bool=Get IsRoundsGame()
   ) -> (then, ReturnValue:Scoreboard Team Widget Object Reference) {is=title K2Node_CallFunction}
   > Add(
    TargetArray:Array of Scoreboard Team Widget Object References=Get TeamScoreDisplays()
    NewItem:"Scoreboard Team Widget Object Reference (by ref)"=@n11
   )
   > AddChild(self=Get DataScrollBox(), Content=@n11)
   n12 = Break Minigame Scoreboard Bucket Row(MinigameScoreboardBucketRow=Get CurrentBucket()) -> (bHighlight=~)
   n13 = For Each Loop(Exec=@n9.else, Array:Array of Minigame Scoreboard Player Row Structures=@n12.PlayerRows) -> (Array Element:Minigame Scoreboard Player Row Structure)
   > n14 = Branch(Condition=Is Valid Index(TargetArray:Array of Scoreboard UI Row Object References=Get PlayerScoreDisplays(), IndexToTest=Get CurrentIndex() -> (CurrentIndex=0)))
   > UpdatePlayerData!(
    execute
    self:Scoreboard UI Row Object Reference="Get (a copy)"(Array:Array of Scoreboard UI Row Object References=Get PlayerScoreDisplays(), Dimension 1=Get CurrentIndex() -> (CurrentIndex=0)) -> (Output:Scoreboard UI Row Object Reference)
    PlayerBucketData:Minigame Scoreboard Player Row Structure=@n13
   ) -> (then)
   n15 = MakePlayerScoreboardRow(
    execute=@n14.else
    PlayerRow=@n13
    PlayerIndex=Get CurrentIndex() -> (CurrentIndex=0)
    ShowWins=ShowWinsColumn() {is=title K2Node_CallFunction}
    NumPlayers=Get NumBuckets() -> (NumBuckets=0)
    Standing=@n12
   ) {is=title K2Node_CallFunction}
   > Add(
    TargetArray:Array of Scoreboard UI Row Object References=Get PlayerScoreDisplays()
    NewItem:"Scoreboard UI Row Object Reference (by ref)"=@n15
   )
   > AddChild(self=Get DataScrollBox(), Content=@n15)
  ]}
]
variables [
 bInGameScoreboard {type=boolean}
 bIsCareerScoreboard {type=boolean}
 bIsTeamGame {type=boolean}
 bShowCumulativeScoreboard {type=boolean}
 IsRoundsGame {type=boolean}
 OnMinigameEnd {}
 PlayerScoreDisplays {type="array<Scoreboard_UI_Row_C>", default []}
 ScoreboardColumnWidget {type=Scoreboard_UI_ColumnLabels_C, default=null}
 StatUpdateRate {type=number, default=0.5}
 TeamColors {
  type="array<LinearColor>"
  default [
   <rgba 1 1 1 1>
   <rgba 0.2086 0.3763 1 1>
   <rgba 0.9301 0.1119 0.1095 1>
   <rgba 1 0.7231 0 1>
   <rgba 0.0742 0.5972 0.0976 1>
   <rgba 0.5457 0.0742 0.5972 1>
   <rgba 0.9823 0.2502 0.0252 1>
   <rgba 0.0382 0.8632 0.5972 1>
   <rgba 0.5972 0.4179 0.3095 1>
   <rgba 0.1812 0.0307 0.2831 1>
   <rgba 0.3763 0.8228 0.0908 1>
   <rgba 0.0152 0.1878 0.2961 1>
   <rgba 0.4342 0.4287 0.4342 1>
   <rgba 0.0685 0.6038 1 1>
   <rgba 1 0.3278 0.7529 1>
   <rgba 0.0033 0.1912 0.0252 1>
  ]
 }
 TeamScoreDisplays {type="array<Scoreboard_Team_Widget_C>", default []}
 TestData {
  type="array<MinigameScoreboardPlayerRow>"
  default [
   "" {bHasLeftGame=false, bHighlight=false, playerState=null, teamColorIndex=-1, winCount=0, values []}
   "" {bHasLeftGame=false, bHighlight=false, playerState=null, teamColorIndex=-1, winCount=0, values []}
   "" {bHasLeftGame=false, bHighlight=false, playerState=null, teamColorIndex=-1, winCount=0, values []}
   "" {bHasLeftGame=false, bHighlight=false, playerState=null, teamColorIndex=-1, winCount=0, values []}
   "" {bHasLeftGame=false, bHighlight=false, playerState=null, teamColorIndex=-1, winCount=0, values []}
   "" {bHasLeftGame=false, bHighlight=false, playerState=null, teamColorIndex=-1, winCount=0, values []}
   "" {bHasLeftGame=false, bHighlight=false, playerState=null, teamColorIndex=-1, winCount=0, values []}
   "" {bHasLeftGame=false, bHighlight=false, playerState=null, teamColorIndex=-1, winCount=0, values []}
   "" {bHasLeftGame=false, bHighlight=false, playerState=null, teamColorIndex=-1, winCount=0, values []}
   "" {bHasLeftGame=false, bHighlight=false, playerState=null, teamColorIndex=-1, winCount=0, values []}
   "" {bHasLeftGame=false, bHighlight=false, playerState=null, teamColorIndex=-1, winCount=0, values []}
   "" {bHasLeftGame=false, bHighlight=false, playerState=null, teamColorIndex=-1, winCount=0, values []}
   "" {bHasLeftGame=false, bHighlight=false, playerState=null, teamColorIndex=-1, winCount=0, values []}
   "" {bHasLeftGame=false, bHighlight=false, playerState=null, teamColorIndex=-1, winCount=0, values []}
   "" {bHasLeftGame=false, bHighlight=false, playerState=null, teamColorIndex=-1, winCount=0, values []}
   "" {bHasLeftGame=false, bHighlight=false, playerState=null, teamColorIndex=-1, winCount=0, values []}
   "" {bHasLeftGame=false, bHighlight=false, playerState=null, teamColorIndex=-1, winCount=0, values []}
   "" {bHasLeftGame=false, bHighlight=false, playerState=null, teamColorIndex=-1, winCount=0, values []}
  ]
 }
 UpdateDisplayTimerHandle {type=TimerHandle}
 WinCondition {type=EMinigameWinCondition, default=MostRoundWins}
]
widgets {
 info {inheritedWidgetCount=2, parentClass=/Script/FortniteUI.AthenaFullScreenScoreboardPanel, rootWidgetClass=&Border, widgetCount=44}
 widgets [
  Border_4 {
   widgetClassPath=&Border
   widget=Border_4
   properties {
    background {
     bIsDynamicallyLoaded=false
     drawAs=Image
     imageType=NoImage
     mirroring=NoMirror
     resourceName=None
     resourceObject=/Game/Items/Traps/Blueprints/Toys/UI_Art_Assets/M_UI_CreativeScoreboard_Backing_Inst
     tiling=NoTile
     imageSize=<32 32>
     margin=<ltrb 0 0 0 0>
     outlineSettings {
      bUseBrushTransparency=false
      roundingType=HalfHeightRadius
      width=0
      color {colorUseRule=UseColor_Specified, specifiedColor=<rgba 0 0 0 0>}
      cornerRadii=<0 0 0 0>
     }
     tintColor {colorUseRule=UseColor_Specified, specifiedColor=<rgba 1 1 1 0>}
     uVRegion {bIsValid=false, max=<0 0>, min=<0 0>}
    }
    padding=<ltrb 0 0 0 0>
   }
  }
  Overlay_1 {widgetClassPath=&Overlay, parent=Border_4, widget=Overlay_1, slot {class=BorderSlot, properties {padding=<ltrb 0 0 0 0>}}}
  BackingTris {
   widgetClassPath=&Image
   bIsVariable=true
   parent=Overlay_1
   widget=BackingTris
   properties {
    renderOpacity=0
    visibility=SelfHitTestInvisible
    brush {
     bIsDynamicallyLoaded=false
     drawAs=Image
     imageType=NoImage
     mirroring=NoMirror
     resourceName=None
     resourceObject=/Game/Items/Traps/Blueprints/Toys/UI_Art_Assets/M_UI_FinalScoreBoard_Tri_Backing
     tiling=NoTile
     imageSize=<32 32>
     margin=<ltrb 0 0 0 0>
     outlineSettings {
      bUseBrushTransparency=false
      roundingType=HalfHeightRadius
      width=0
      color {colorUseRule=UseColor_Specified, specifiedColor=<rgba 0 0 0 0>}
      cornerRadii=<0 0 0 0>
     }
     tintColor {colorUseRule=UseColor_Specified, specifiedColor=<rgba 1 1 1 1>}
     uVRegion {bIsValid=false, max=<0 0>, min=<0 0>}
    }
   }
   slot {class=OverlaySlot, properties {horizontalAlignment=HAlign_Fill, verticalAlignment=VAlign_Fill}}
  }
  SafeZone_0 {
   widgetClassPath=&SafeZone
   bIsVariable=true
   parent=Overlay_1
   widget=SafeZone_0
   slot {class=OverlaySlot, properties {horizontalAlignment=HAlign_Fill, verticalAlignment=VAlign_Fill}}
  }
  Overlay_5 {widgetClassPath=&Overlay, parent=SafeZone_0, widget=Overlay_5, slot {class=SafeZoneSlot}}
  All {
   widgetClassPath=&VerticalBox
   bIsVariable=true
   parent=Overlay_5
   widget=All
   slot {class=OverlaySlot, properties {horizontalAlignment=HAlign_Fill, padding=<ltrb 128 64 128 272>}}
  }
  HeaderDisplay {
   widgetClassPath=&VerticalBox
   parent=All
   widget=HeaderDisplay
   slot {class=VerticalBoxSlot, properties {size {sizeRule=Fill, value=1}}}
  }
  Header {widgetClassPath=&HorizontalBox, bIsVariable=true, parent=HeaderDisplay, widget=Header, slot {class=VerticalBoxSlot}}
  HeaderText {widgetClassPath=&HorizontalBox, parent=Header, widget=HeaderText, slot {class=HorizontalBoxSlot}}
  GameStats {
   widgetClassPath=&CommonTextBlock
   bIsVariable=true
   parent=HeaderText
   widget=GameStats
   properties {
    text=Scoreboard
    textTransformPolicy=ToUpper
    font {
     bForceMonospaced=false
     bMaterialIsStencil=false
     fontMaterial=null
     fontObject=&Burbank
     letterSpacing=0
     monospacedWidth=1
     size=56
     skewAmount=0
     typefaceFontName=Burbank-Black
     outlineSettings {
      bApplyOutlineToDropShadows=false
      bMiteredCorners=false
      bSeparateFillAlpha=false
      outlineMaterial=null
      outlineSize=0
      outlineColor=<rgba 0 0 0 1>
     }
    }
    renderTransform {angle=0, scale=<1 1>, shear=<-8 0>, translation=<0 0>}
    shadowOffset=<0 0>
   }
   slot {class=HorizontalBoxSlot}
  }
  Spacer_72 {widgetClassPath=&Spacer, parent=HeaderText, widget=Spacer_72, properties {size=<12 1>}, slot {class=HorizontalBoxSlot}}
  RoundsCountTotal {
   widgetClassPath=&CommonTextBlock
   bIsVariable=true
   parent=HeaderText
   widget=RoundsCountTotal
   properties {
    text=of 999
    textTransformPolicy=ToUpper
    visibility=Collapsed
    colorAndOpacity {colorUseRule=UseColor_Specified, specifiedColor=<rgba 0.25 1 1 1>}
    font {
     bForceMonospaced=false
     bMaterialIsStencil=false
     fontMaterial=null
     fontObject=&BurbankSmall
     letterSpacing=0
     monospacedWidth=1
     size=24
     skewAmount=0
     typefaceFontName=Black
     outlineSettings {
      bApplyOutlineToDropShadows=false
      bMiteredCorners=false
      bSeparateFillAlpha=false
      outlineMaterial=null
      outlineSize=0
      outlineColor=<rgba 0 0 0 1>
     }
    }
    renderTransform {angle=0, scale=<1 1>, shear=<-12 0>, translation=<0 -16>}
    shadowOffset=<0 0>
   }
   slot {class=HorizontalBoxSlot, properties {horizontalAlignment=HAlign_Left, verticalAlignment=VAlign_Bottom}}
  }
  WinnerDisplay {
   widgetClassPath=&Overlay
   bIsVariable=true
   parent=Header
   widget=WinnerDisplay
   properties {visibility=Hidden}
   slot {class=HorizontalBoxSlot, properties {horizontalAlignment=HAlign_Right, size {sizeRule=Fill, value=1}}}
  }
  BackingBox {
   widgetClassPath=&SizeBox
   parent=WinnerDisplay
   widget=BackingBox
   properties {bOverride_MinDesiredWidth=true, heightOverride=64, minDesiredWidth=256, widthOverride=384}
   slot {class=OverlaySlot, properties {horizontalAlignment=HAlign_Fill, verticalAlignment=VAlign_Fill}}
  }
  Overlay_0 {widgetClassPath=&Overlay, parent=BackingBox, widget=Overlay_0, slot {class=SizeBoxSlot}}
  Overlay_2 {
   widgetClassPath=&Overlay
   parent=Overlay_0
   widget=Overlay_2
   slot {class=OverlaySlot, properties {horizontalAlignment=HAlign_Fill, verticalAlignment=VAlign_Fill, padding=<ltrb 5 4 5 4>}}
  }
  Backing {
   widgetClassPath=&Image
   bIsVariable=true
   parent=Overlay_2
   widget=Backing
   properties {
    brush {
     bIsDynamicallyLoaded=false
     drawAs=Image
     imageType=NoImage
     mirroring=NoMirror
     resourceName=None
     resourceObject=/Game/Items/Traps/Blueprints/Toys/UI_Art_Assets/M_UI_Scoreboard_GameWinnerBacking_Inst
     tiling=NoTile
     imageSize=<32 32>
     margin=<ltrb 0 0 0 0>
     outlineSettings {
      bUseBrushTransparency=false
      roundingType=HalfHeightRadius
      width=0
      color {colorUseRule=UseColor_Specified, specifiedColor=<rgba 0 0 0 0>}
      cornerRadii=<0 0 0 0>
     }
     tintColor {colorUseRule=UseColor_Specified, specifiedColor=<rgba 1 1 1 1>}
     uVRegion {bIsValid=false, max=<0 0>, min=<0 0>}
    }
    renderTransform {angle=0, scale=<1 1>, shear=<0 0>, translation=<0 -4>}
   }
   slot {class=OverlaySlot, properties {horizontalAlignment=HAlign_Fill, verticalAlignment=VAlign_Fill}}
  }
  WinnerVBox {
   widgetClassPath=&VerticalBox
   bIsVariable=true
   parent=Overlay_2
   widget=WinnerVBox
   slot {class=OverlaySlot, properties {horizontalAlignment=HAlign_Center, verticalAlignment=VAlign_Center, padding=<ltrb 48 0 48 0>}}
  }
  WinnerNameColor {
   widgetClassPath=&Border
   bIsVariable=true
   parent=WinnerVBox
   widget=WinnerNameColor
   properties {
    background {
     bIsDynamicallyLoaded=false
     drawAs=NoDrawType
     imageType=NoImage
     mirroring=NoMirror
     resourceName=None
     resourceObject=null
     tiling=NoTile
     imageSize=<32 32>
     margin=<ltrb 0 0 0 0>
     outlineSettings {
      bUseBrushTransparency=false
      roundingType=HalfHeightRadius
      width=0
      color {colorUseRule=UseColor_Specified, specifiedColor=<rgba 0 0 0 0>}
      cornerRadii=<0 0 0 0>
     }
     tintColor {colorUseRule=UseColor_Specified, specifiedColor=<rgba 1 1 1 1>}
     uVRegion {bIsValid=false, max=<0 0>, min=<0 0>}
    }
    padding=<ltrb 0 0 0 0>
   }
   slot {class=VerticalBoxSlot, properties {horizontalAlignment=HAlign_Center}}
  }
  WinnerName {
   widgetClassPath=&CommonTextBlock
   bIsVariable=true
   parent=WinnerNameColor
   widget=WinnerName
   properties {
    justification=Center
    text=Player or Team
    textTransformPolicy=ToUpper
    font {
     bForceMonospaced=false
     bMaterialIsStencil=false
     fontMaterial=null
     fontObject=&BurbankSmall
     letterSpacing=0
     monospacedWidth=1
     size=20
     skewAmount=0
     typefaceFontName=Black
     outlineSettings {
      bApplyOutlineToDropShadows=false
      bMiteredCorners=false
      bSeparateFillAlpha=false
      outlineMaterial=null
      outlineSize=0
      outlineColor=<rgba 0 0 0 1>
     }
    }
    renderTransform {angle=0, scale=<1 1>, shear=<-8 0>, translation=<0 0>}
    shadowOffset=<0 0>
   }
   slot {class=BorderSlot, properties {padding=<ltrb 0 0 0 0>}}
  }
  WinDescriptionColor {
   widgetClassPath=&Border
   bIsVariable=true
   parent=WinnerVBox
   widget=WinDescriptionColor
   properties {
    background {
     bIsDynamicallyLoaded=false
     drawAs=NoDrawType
     imageType=NoImage
     mirroring=NoMirror
     resourceName=None
     resourceObject=null
     tiling=NoTile
     imageSize=<32 32>
     margin=<ltrb 0 0 0 0>
     outlineSettings {
      bUseBrushTransparency=false
      roundingType=HalfHeightRadius
      width=0
      color {colorUseRule=UseColor_Specified, specifiedColor=<rgba 0 0 0 0>}
      cornerRadii=<0 0 0 0>
     }
     tintColor {colorUseRule=UseColor_Specified, specifiedColor=<rgba 1 1 1 1>}
     uVRegion {bIsValid=false, max=<0 0>, min=<0 0>}
    }
    contentColorAndOpacity=<rgba 0.5029 0.117 0.0123 1>
    padding=<ltrb 0 0 0 0>
   }
   slot {class=VerticalBoxSlot, properties {horizontalAlignment=HAlign_Center}}
  }
  WinDescription {
   widgetClassPath=&CommonTextBlock
   bIsVariable=true
   parent=WinDescriptionColor
   widget=WinDescription
   properties {
    justification=Center
    text="Wins the game!"
    textTransformPolicy=ToUpper
    wrapTextAt=256
    font {
     bForceMonospaced=false
     bMaterialIsStencil=false
     fontMaterial=null
     fontObject=&BurbankSmall
     letterSpacing=0
     monospacedWidth=1
     size=16
     skewAmount=0
     typefaceFontName=Black
     outlineSettings {
      bApplyOutlineToDropShadows=false
      bMiteredCorners=false
      bSeparateFillAlpha=false
      outlineMaterial=null
      outlineSize=0
      outlineColor=<rgba 0 0 0 1>
     }
    }
    renderTransform {angle=0, scale=<1 1>, shear=<-12 0>, translation=<0 -4>}
    shadowOffset=<0 0>
   }
   slot {class=BorderSlot, properties {padding=<ltrb 0 0 0 0>}}
  }
  PlacementMarker {
   widgetClassPath=&Overlay
   bIsVariable=true
   parent=WinnerDisplay
   widget=PlacementMarker
   properties {renderTransform {angle=0, scale=<1 1>, shear=<0 0>, translation=<-88 -20>}}
   slot {class=OverlaySlot}
  }
  Glow {
   widgetClassPath=&Image
   bIsVariable=true
   parent=PlacementMarker
   widget=Glow
   properties {
    renderOpacity=0.25
    brush {
     bIsDynamicallyLoaded=false
     drawAs=Image
     imageType=NoImage
     mirroring=NoMirror
     resourceName=None
     resourceObject=/Game/Items/Traps/Blueprints/Toys/UI_Art_Assets/M_UI_RoundEnd_Marker_Glow
     tiling=NoTile
     imageSize=<192 32>
     margin=<ltrb 0 0 0 0>
     outlineSettings {
      bUseBrushTransparency=false
      roundingType=HalfHeightRadius
      width=0
      color {colorUseRule=UseColor_Specified, specifiedColor=<rgba 0 0 0 0>}
      cornerRadii=<0 0 0 0>
     }
     tintColor {colorUseRule=UseColor_Specified, specifiedColor=<rgba 1 1 1 1>}
     uVRegion {bIsValid=false, max=<0 0>, min=<0 0>}
    }
    renderTransform {angle=0, scale=<1.125 1.25>, shear=<0 0>, translation=<0 0>}
   }
   slot {class=OverlaySlot, properties {horizontalAlignment=HAlign_Center, verticalAlignment=VAlign_Fill}}
  }
  Triangles {
   widgetClassPath=&SizeBox
   parent=PlacementMarker
   widget=Triangles
   properties {bOverride_HeightOverride=true, bOverride_WidthOverride=true, heightOverride=56, widthOverride=192}
   slot {class=OverlaySlot, properties {horizontalAlignment=HAlign_Center, verticalAlignment=VAlign_Center}}
  }
  HorizontalBox_202 {widgetClassPath=&HorizontalBox, parent=Triangles, widget=HorizontalBox_202, slot {class=SizeBoxSlot}}
  Image_1007 {
   widgetClassPath=&Image
   bIsVariable=true
   parent=HorizontalBox_202
   widget=Image_1007
   properties {
    brush {
     bIsDynamicallyLoaded=false
     drawAs=Image
     imageType=NoImage
     mirroring=NoMirror
     resourceName=None
     resourceObject=/Game/UI/Foundation/Materials/M_FX_TriangleScroll_Scoreboard
     tiling=NoTile
     imageSize=<32 32>
     margin=<ltrb 0 0 0 0>
     outlineSettings {
      bUseBrushTransparency=false
      roundingType=HalfHeightRadius
      width=0
      color {colorUseRule=UseColor_Specified, specifiedColor=<rgba 0 0 0 0>}
      cornerRadii=<0 0 0 0>
     }
     tintColor {colorUseRule=UseColor_Specified, specifiedColor=<rgba 1 1 1 1>}
     uVRegion {bIsValid=false, max=<0 0>, min=<0 0>}
    }
    renderTransform {angle=0, scale=<-1 -1>, shear=<0 0>, translation=<0 0>}
   }
   slot {class=HorizontalBoxSlot, properties {size {sizeRule=Fill, value=1}}}
  }
  Image_1009 {
   widgetClassPath=&Image
   bIsVariable=true
   parent=HorizontalBox_202
   widget=Image_1009
   properties {brush {
     bIsDynamicallyLoaded=false
     drawAs=Image
     imageType=NoImage
     mirroring=NoMirror
     resourceName=None
     resourceObject=&M_FX_TriangleScroll_Scoreboard
     tiling=NoTile
     imageSize=<32 32>
     margin=<ltrb 0 0 0 0>
     outlineSettings {
      bUseBrushTransparency=false
      roundingType=HalfHeightRadius
      width=0
      color {colorUseRule=UseColor_Specified, specifiedColor=<rgba 0 0 0 0>}
      cornerRadii=<0 0 0 0>
     }
     tintColor {colorUseRule=UseColor_Specified, specifiedColor=<rgba 1 1 1 1>}
     uVRegion {bIsValid=false, max=<0 0>, min=<0 0>}
    }}
   slot {class=HorizontalBoxSlot, properties {size {sizeRule=Fill, value=1}}}
  }
  Circle {
   widgetClassPath=&SizeBox
   parent=PlacementMarker
   widget=Circle
   properties {bOverride_HeightOverride=true, bOverride_WidthOverride=true, heightOverride=56, widthOverride=56}
   slot {class=OverlaySlot, properties {horizontalAlignment=HAlign_Center, verticalAlignment=VAlign_Center}}
  }
  Image_900 {
   widgetClassPath=&Image
   bIsVariable=true
   parent=Circle
   widget=Image_900
   properties {brush {
     bIsDynamicallyLoaded=false
     drawAs=Image
     imageType=NoImage
     mirroring=NoMirror
     resourceName=None
     resourceObject=/Game/Items/Traps/Blueprints/Toys/UI_Art_Assets/M_UI_RoundEnd_Marker
     tiling=NoTile
     imageSize=<32 32>
     margin=<ltrb 0 0 0 0>
     outlineSettings {
      bUseBrushTransparency=false
      roundingType=HalfHeightRadius
      width=0
      color {colorUseRule=UseColor_Specified, specifiedColor=<rgba 0 0 0 0>}
      cornerRadii=<0 0 0 0>
     }
     tintColor {colorUseRule=UseColor_Specified, specifiedColor=<rgba 1 1 1 1>}
     uVRegion {bIsValid=false, max=<0 0>, min=<0 0>}
    }}
   slot {class=SizeBoxSlot}
  }
  1 {
   widgetClassPath=&CommonTextBlock
   parent=PlacementMarker
   widget=1
   properties {
    justification=Center
    text="##1"
    font {
     bForceMonospaced=false
     bMaterialIsStencil=false
     fontMaterial=null
     fontObject=&Burbank
     letterSpacing=0
     monospacedWidth=1
     size=26
     skewAmount=0
     typefaceFontName=Burbank-Black
     outlineSettings {
      bApplyOutlineToDropShadows=false
      bMiteredCorners=false
      bSeparateFillAlpha=false
      outlineMaterial=null
      outlineSize=0
      outlineColor=<rgba 0 0 0 1>
     }
    }
    renderTransform {angle=0, scale=<1 1>, shear=<-8 0>, translation=<-2 0>}
    shadowOffset=<0 0>
   }
   slot {class=OverlaySlot, properties {horizontalAlignment=HAlign_Center, verticalAlignment=VAlign_Center}}
  }
  ScoreboardDisplay {
   widgetClassPath=&VerticalBox
   bIsVariable=true
   parent=HeaderDisplay
   widget=ScoreboardDisplay
   slot {class=VerticalBoxSlot, properties {padding=<ltrb 0 40 0 0>, size {sizeRule=Fill, value=1}}}
  }
  ColumnHolder {
   widgetClassPath=&Border
   bIsVariable=true
   parent=ScoreboardDisplay
   widget=ColumnHolder
   properties {
    background {
     bIsDynamicallyLoaded=false
     drawAs=NoDrawType
     imageType=NoImage
     mirroring=NoMirror
     resourceName=None
     resourceObject=null
     tiling=NoTile
     imageSize=<32 32>
     margin=<ltrb 0 0 0 0>
     outlineSettings {
      bUseBrushTransparency=false
      roundingType=HalfHeightRadius
      width=0
      color {colorUseRule=UseColor_Specified, specifiedColor=<rgba 0 0 0 0>}
      cornerRadii=<0 0 0 0>
     }
     tintColor {colorUseRule=UseColor_Specified, specifiedColor=<rgba 1 1 1 1>}
     uVRegion {bIsValid=false, max=<0 0>, min=<0 0>}
    }
    padding=<ltrb 0 0 0 0>
   }
   slot {class=VerticalBoxSlot, properties {verticalAlignment=VAlign_Top}}
  }
  CanvasPanel_2 {
   widgetClassPath=&CanvasPanel
   parent=ScoreboardDisplay
   widget=CanvasPanel_2
   slot {class=VerticalBoxSlot, properties {size {sizeRule=Fill, value=1}}}
  }
  DataScrollBox {
   widgetClassPath=&ScrollBox
   bInherited=true
   bIsVariable=true
   parent=CanvasPanel_2
   widget=DataScrollBox
   properties {allowOverscroll=false}
   slot {
    class=CanvasPanelSlot
    properties {bAutoSize=true, layoutData {alignment=<0 0>, anchors {maximum=<1 1>, minimum=<0 0>}, offsets=<ltrb 0 0 0 0>}}
   }
  }
  TransitionInfo {
   widgetClassPath=&HorizontalBox
   bIsVariable=true
   parent=All
   widget=TransitionInfo
   slot {class=VerticalBoxSlot, properties {horizontalAlignment=HAlign_Right, verticalAlignment=VAlign_Bottom, padding=<ltrb 0 16 16 0>}}
  }
  CountdownHeader {
   widgetClassPath=&VerticalBox
   parent=TransitionInfo
   widget=CountdownHeader
   slot {class=HorizontalBoxSlot, properties {horizontalAlignment=HAlign_Right, verticalAlignment=VAlign_Center, padding=<ltrb 0 0 0 4>}}
  }
  CountdownHeaderSmall {
   widgetClassPath=&CommonTextBlock
   bIsVariable=true
   parent=CountdownHeader
   widget=CountdownHeaderSmall
   properties {
    justification=Right
    text=Next Round
    textTransformPolicy=ToUpper
    visibility=Collapsed
    colorAndOpacity {colorUseRule=UseColor_Specified, specifiedColor=<rgba 0 1 1 0.5>}
    font {
     bForceMonospaced=false
     bMaterialIsStencil=false
     fontMaterial=null
     fontObject=&BurbankSmall
     letterSpacing=0
     monospacedWidth=1
     size=20
     skewAmount=0
     typefaceFontName=Black
     outlineSettings {
      bApplyOutlineToDropShadows=false
      bMiteredCorners=false
      bSeparateFillAlpha=false
      outlineMaterial=null
      outlineSize=0
      outlineColor=<rgba 0 0 0 0>
     }
    }
    renderTransform {angle=0, scale=<1 1>, shear=<-12 0>, translation=<0 4>}
    shadowColorAndOpacity=<rgba 0 0 0 0.3>
    shadowOffset=<0 0>
   }
   slot {class=VerticalBoxSlot, properties {horizontalAlignment=HAlign_Right, verticalAlignment=VAlign_Center}}
  }
  CountdownHeaderlarge {
   widgetClassPath=&CommonTextBlock
   bIsVariable=true
   parent=CountdownHeader
   widget=CountdownHeaderlarge
   properties {
    justification=Right
    text=Game Ending In
    textTransformPolicy=ToUpper
    colorAndOpacity {colorUseRule=UseColor_Specified, specifiedColor=<rgba 0.25 1 1 1>}
    font {
     bForceMonospaced=false
     bMaterialIsStencil=false
     fontMaterial=null
     fontObject=&BurbankSmall
     letterSpacing=0
     monospacedWidth=1
     size=32
     skewAmount=0
     typefaceFontName=Black
     outlineSettings {
      bApplyOutlineToDropShadows=false
      bMiteredCorners=false
      bSeparateFillAlpha=false
      outlineMaterial=null
      outlineSize=0
      outlineColor=<rgba 0 0 0 0>
     }
    }
    renderTransform {angle=0, scale=<1 1>, shear=<-12 0>, translation=<0 0>}
    shadowColorAndOpacity=<rgba 0 0 0 0.3>
    shadowOffset=<0 0>
   }
   slot {class=VerticalBoxSlot, properties {padding=<ltrb 0 0 8 0>}}
  }
  Spacer_182 {widgetClassPath=&Spacer, parent=TransitionInfo, widget=Spacer_182, properties {size=<4 1>}, slot {class=HorizontalBoxSlot}}
  Image_110 {
   widgetClassPath=&Image
   bIsVariable=true
   parent=TransitionInfo
   widget=Image_110
   properties {brush {
     bIsDynamicallyLoaded=false
     drawAs=Image
     imageType=NoImage
     mirroring=NoMirror
     resourceName=None
     resourceObject=/Game/Items/Traps/Blueprints/Toys/UI_Art_Assets/M_UI_ActionDividerLine_DeepBlue
     tiling=NoTile
     imageSize=<24 32>
     margin=<ltrb 0 0 0 0>
     outlineSettings {
      bUseBrushTransparency=false
      roundingType=HalfHeightRadius
      width=0
      color {colorUseRule=UseColor_Specified, specifiedColor=<rgba 0 0 0 0>}
      cornerRadii=<0 0 0 0>
     }
     tintColor {colorUseRule=UseColor_Specified, specifiedColor=<rgba 1 1 1 1>}
     uVRegion {bIsValid=false, max=<0 0>, min=<0 0>}
    }}
   slot {class=HorizontalBoxSlot, properties {padding=<ltrb 0 4 0 2>}}
  }
  Spacer_289 {widgetClassPath=&Spacer, parent=TransitionInfo, widget=Spacer_289, properties {size=<10 1>}, slot {class=HorizontalBoxSlot}}
  CountdownText {
   widgetClassPath=&CommonTextBlock
   bIsVariable=true
   parent=TransitionInfo
   widget=CountdownText
   properties {
    minDesiredWidth=96
    text=88
    font {
     bForceMonospaced=false
     bMaterialIsStencil=false
     fontMaterial=null
     fontObject=&Burbank
     letterSpacing=0
     monospacedWidth=1
     size=72
     skewAmount=0
     typefaceFontName=Burbank-Black
     outlineSettings {
      bApplyOutlineToDropShadows=false
      bMiteredCorners=false
      bSeparateFillAlpha=false
      outlineMaterial=null
      outlineSize=0
      outlineColor=<rgba 0 0 0 0>
     }
    }
    renderTransform {angle=0, scale=<1 1>, shear=<-8 0>, translation=<0 0>}
    shadowColorAndOpacity=<rgba 0 0 0 0.3>
    shadowOffset=<0 0>
   }
   slot {class=HorizontalBoxSlot, properties {horizontalAlignment=HAlign_Center, verticalAlignment=VAlign_Center}}
  }
  TutorialHighlight {widgetClassPath=&TutorialHighlight}
 ]
}
