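# Map screen tab widget that shows the minigame scoreboard, or a "Scoreboard Not Available" message when it is disabled.
# On activation, IsScoreboardEnabled picks the regular or career tag from the minigame's FNE UI component, defaulting to enabled if any lookup fails.
# The result drives a visibility switcher and show/hide calls on Scoreboard_UI.
# Construct softens backdrop blur and gradient alpha when in zone, and PreConstruct sizes a tinted image for mobile.
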WidgetBlueprint /Game/Items/Traps/Blueprints/Toys/Scoreboard_MapWidget
graphs [
 EventGraph {nodes [
   Event Construct()
   > n1 = Branch(Condition=IsInZone!(self:Fort Global UIContext Object Reference=Get FortGlobalUIContext() -> (ReturnValue:Fort Global UIContext Object Reference) {is=title K2Node_GetSubsystem}) -> (ReturnValue:bool))
   n2 = Get MapFalloff() -> (MapFalloff:Image Object Reference)
   > n3 = GetDynamicMaterial(self=@n2)
   > SetScalarParameterValue(self=@n3, ParameterName=Alpha, Value=0.65)
   n4 = Get BackgroundBlur_89() -> (BackgroundBlur_89:Background Blur Object Reference)
   > SetBlurStrength!(execute, self:Background Blur Object Reference=@n4, InStrength:float=15) -> (then)
   n5 = GetDynamicMaterial(execute=@n1.else, self=@n2)
   > SetScalarParameterValue(self=@n5, ParameterName=Alpha, Value=0.75)
   > SetBlurStrength(self=@n4, InStrength=20)
   Event PreConstruct()
   > SetBrush!(
    execute
    self=Get Image_145() -> (Image_145:Image Object Reference)
    InBrush_TintColor_SpecifiedColor:Linear Color Structure=(R=0,G=0.0012,B=0.0999,A=0.4)
    InBrush_TintColor_ColorUseRule:ESlateColorStylingMode Enum
    InBrush_DrawAs:ESlateBrushDrawType Enum=Image
    InBrush_Tiling:ESlateBrushTileType Enum=NoTile
    InBrush_Mirroring:ESlateBrushMirrorType Enum=NoMirror
    InBrush_ImageSize:Vector 2D Structure=Select(Option 0:Vector 2D Structure=(X=130,Y=130), Option 1:Vector 2D Structure=(X=130,Y=160), Index=IsMobileGame()) -> (ReturnValue:Vector 2D Structure)
    InBrush_Margin:Margin Structure=(Left=0,Top=0,Right=0,Bottom=0)
    InBrush_ResourceObject:Object Reference
    InBrush_OutlineSettings:Slate Brush Outline Settings Structure=(CornerRadii=(X=0,Y=0,Z=0,W=0),Color=(SpecifiedColor=(R=0,G=0,B=0,A=0),ColorUseRule=UseColor_Specified),Width=0,RoundingType=HalfHeightRadius,bUseBrushTransparency=False)
   ) -> (then)
   Event On Activated()
   > n6 = Branch(Condition=IsScoreboardEnabled!() -> (bIsEnabled:bool))
   n7 = Get ScoreboardEnabledSwitcher() -> (ScoreboardEnabledSwitcher:Common Visibility Switcher Object Reference)
   > SetActiveWidget(self=@n7, Widget=Get Scoreboard_UI() -> (Scoreboard_UI:Scoreboard UI Object Reference))
   > DoShowUI(self:Scoreboard UI Object Reference=Get Scoreboard_UI())
   SetActiveWidget(execute=@n6.else, self=@n7, Widget=Get NotAvailableText() -> (NotAvailableText:Common Text Object Reference))
   Event On Deactivated()
   > Branch(Condition=IsScoreboardEnabled())
   > DoHideUI(self:Scoreboard UI Object Reference=Get Scoreboard_UI())
  ]}
 IsScoreboardEnabled {nodes [
   IsScoreboardEnabled() {is=title K2Node_FunctionEntry}
   > n1 = Cast To FortPlayerControllerAthena(Object=GetOwningPlayer())
   > n2 = Set Minigame(Minigame:Fort Minigame Object Reference=GetMinigame!(self:Fort Player Controller Gameplay Object Reference=@n1) -> (ReturnValue:Fort Minigame Object Reference)) -> (Output_Get:Fort Minigame Object Reference)
   > n3 = Is Valid(InputObject=@n2)
   > n4 = Set FNEUIComp(FNEUIComp:FNEUIComponent Object Reference=GetFNEUIComponent!(self:Fort Minigame Object Reference=Get Minigame() -> (Minigame:Fort Minigame Object Reference)) -> (ReturnValue:FNEUIComponent Object Reference)) -> (Output_Get:FNEUIComponent Object Reference)
   > n5 = Is Valid(InputObject=@n4)
   n6 = Get FNEUIComp() -> (FNEUIComp:FNEUIComponent Object Reference)
   > Set ShowScoreboard(ShowScoreboard:bool=HasTag!(
     TagContainer:"Gameplay Tag Container Structure (by ref)"=Get UIExtensionTags!(self:FNEUIComponent Object Reference=@n6) -> (UIExtensionTags:Gameplay Tag Container Structure)
     Tag:Gameplay Tag Structure=Get ShowScoreboardTag() -> (ShowScoreboardTag:Gameplay Tag Structure)
     bExactMatch:bool=true
    ) -> (ReturnValue:bool)) -> (Output_Get:bool)
   n7 = Get FNEUIComp()
   > Set ShowCareerScoreboard(ShowCareerScoreboard:bool=HasTag(
     TagContainer=Get UIExtensionTags(self=@n7)
     Tag=Get ShowCareerScoreboardTag() -> (ShowCareerScoreboardTag:Gameplay Tag Structure)
     bExactMatch=true
    )) -> (Output_Get:bool)
   n8 = Get Scoreboard_UI()
   > Return Node!(
    execute
    bIsEnabled:bool=Select(
     Option 0:bool=Get ShowScoreboard() -> (ShowScoreboard:bool)
     Option 1:bool=Get ShowCareerScoreboard() -> (ShowCareerScoreboard:bool)
     Index=Get bIsCareerScoreboard!(self:Scoreboard UI Object Reference=@n8) -> (bIsCareerScoreboard:bool)
    ) -> (ReturnValue:bool)
   )
   Return Node(execute=[@n5.Is Not Valid, @n3.Is Not Valid, @n1.CastFailed], bIsEnabled=true)
  ]}
]
variables []
widgets {
 info {inheritedWidgetCount=0, parentClass=/Script/FortniteUI.AthenaMapScreenTabBase, rootWidgetClass=&Overlay, widgetCount=9}
 widgets [
  Overlay_19 {widgetClassPath=&Overlay, widget=Overlay_19}
  BackgroundBlur_89 {
   widgetClassPath=&BackgroundBlur
   bIsVariable=true
   parent=Overlay_19
   widget=BackgroundBlur_89
   properties {blurStrength=20}
   slot {class=OverlaySlot, properties {horizontalAlignment=HAlign_Fill, verticalAlignment=VAlign_Fill}}
  }
  Border_235 {
   widgetClassPath=&Border
   parent=Overlay_19
   widget=Border_235
   properties {visibility=SelfHitTestInvisible, brushColor=<rgba 0 0 0 0.4>, padding=<ltrb 0 0 0 0>}
   slot {class=OverlaySlot, properties {horizontalAlignment=HAlign_Fill, verticalAlignment=VAlign_Fill, padding=<ltrb -2 -2 -2 -2>}}
  }
  MapFalloff {
   widgetClassPath=&Image
   bIsVariable=true
   parent=Border_235
   widget=MapFalloff
   properties {
    visibility=SelfHitTestInvisible
    brush {
     bIsDynamicallyLoaded=false
     drawAs=Image
     imageType=NoImage
     mirroring=NoMirror
     resourceName=None
     resourceObject=/Game/Athena/Apollo/Maps/UI/M_MapBacking_Gradient
     tiling=NoTile
     imageSize=<32 32>
     margin=<ltrb 0 0 0 0>
     outlineSettings {
      bUseBrushTransparency=false
      roundingType=HalfHeightRadius
      width=0
      color {colorUseRule=UseColor_Specified, specifiedColor=<rgba 0 0 0 0>}
      cornerRadii=<0 0 0 0>
     }
     tintColor {colorUseRule=UseColor_Specified, specifiedColor=<rgba 1 1 1 1>}
     uVRegion {bIsValid=false, max=<0 0>, min=<0 0>}
    }
   }
   slot {class=BorderSlot, properties {padding=<ltrb 0 0 0 0>}}
  }
  ScoreboardEnabledSwitcher {
   widgetClassPath=&CommonVisibilitySwitcher
   bIsVariable=true
   parent=Overlay_19
   widget=ScoreboardEnabledSwitcher
   slot {class=OverlaySlot, properties {horizontalAlignment=HAlign_Fill, verticalAlignment=VAlign_Fill}}
  }
  Scoreboard_UI {
   widgetClassPath=&/Scoreboard_UI
   bIsVariable=true
   parent=ScoreboardEnabledSwitcher
   widget=Scoreboard_UI
   properties {visibility=Hidden}
   slot {class=CommonVisibilitySwitcherSlot, properties {padding=<ltrb 0 75 0 0>}}
  }
  NotAvailableText {
   widgetClassPath=&CommonTextBlock
   bIsVariable=true
   parent=ScoreboardEnabledSwitcher
   widget=NotAvailableText
   properties {
    text=Scoreboard Not Available
    font {
     bForceMonospaced=false
     bMaterialIsStencil=false
     fontMaterial=null
     fontObject=&BurbankSmall
     letterSpacing=0
     monospacedWidth=1
     size=24
     skewAmount=0
     typefaceFontName=Bold
     outlineSettings {
      bApplyOutlineToDropShadows=false
      bMiteredCorners=false
      bSeparateFillAlpha=false
      outlineMaterial=null
      outlineSize=0
      outlineColor=<rgba 0 0 0 1>
     }
    }
    shadowOffset=<0 0>
   }
   slot {class=CommonVisibilitySwitcherSlot, properties {horizontalAlignment=HAlign_Center, verticalAlignment=VAlign_Center, padding=<ltrb 4 2 4 2>}}
  }
  SafeZone_0 {
   widgetClassPath=&SafeZone
   parent=Overlay_19
   widget=SafeZone_0
   slot {class=OverlaySlot, properties {horizontalAlignment=HAlign_Fill}}
  }
  Image_145 {
   widgetClassPath=&Image
   bIsVariable=true
   parent=SafeZone_0
   widget=Image_145
   properties {brush {
     bIsDynamicallyLoaded=false
     drawAs=Image
     imageType=NoImage
     mirroring=NoMirror
     resourceName=None
     resourceObject=null
     tiling=NoTile
     imageSize=<130 130>
     margin=<ltrb 0 0 0 0>
     outlineSettings {
      bUseBrushTransparency=false
      roundingType=HalfHeightRadius
      width=0
      color {colorUseRule=UseColor_Specified, specifiedColor=<rgba 0 0 0 0>}
      cornerRadii=<0 0 0 0>
     }
     tintColor {colorUseRule=UseColor_Specified, specifiedColor=<rgba 0 0.0012 0.0999 0.4>}
     uVRegion {bIsValid=false, max=<0 0>, min=<0 0>}
    }}
   slot {class=SafeZoneSlot, properties {padding=<ltrb -100 -60 -100 0>}}
  }
 ]
}
