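; Actor component for trap toys that opens a property editor widget when a player interacts locally.
; Local interaction calls DisplayPropertyEditor unless assets are loading or an editor is already cached, and always returns true.
; The editor class is pushed asynchronously as soft-class primary content, then cached and initialized with the owning actor.
; Deactivation or EndPlay unbinds the deactivation event and clears the cached widget, allowing it to reopen.
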(event Custom|DisplayPropertyEditor
  (params (InteractingPlayer :type Player%20Controller%20Object%20Reference))
  (Variables|Default|SetbLoadingAssets true)
  (bind _activatablewidget (Class|FortAsyncActionPushPrimaryContent|PushPrimaryContentSoftClassAsyncforPlayer InteractingPlayer (Variables|Default|GetPropertyEditorWidgetClass))
    (:OnPrimaryContentPushed
      (Variables|Default|SetbLoadingAssets false)
      (bind _output_get (Variables|Default|SetCachedPropertyEditorWidget _activatablewidget))
      (Init|InitfromObject(Message) _output_get (Components|GetOwner))
      (Events|BindEventtoOnWidgetDeactivated _output_get (EventDispatchers|CreateEvent :!event Custom|HandlePropertyEditorDeactivated)))))

(event Custom|HandlePropertyEditorDeactivated
  (bind _cachedpropertyeditorwidget (Variables|Default|GetCachedPropertyEditorWidget)
    (:then
      (Events|UnbindEventfromOnWidgetDeactivated _cachedpropertyeditorwidget (EventDispatchers|CreateEvent :!event Custom|HandlePropertyEditorDeactivated))
      (Variables|Default|SetCachedPropertyEditorWidget 0))
    (:else)))

(event EventEndPlay
  (params (EndPlayReason :type EEndPlayReason%20Enum :default "Destroyed"))
  (bind _cachedpropertyeditorwidget (Variables|Default|GetCachedPropertyEditorWidget)
    (:then
      (Events|UnbindEventfromOnWidgetDeactivated _cachedpropertyeditorwidget (EventDispatchers|CreateEvent :!event Custom|HandlePropertyEditorDeactivated))
      (Variables|Default|SetCachedPropertyEditorWidget 0))
    (:else)))

(fn BlueprintOnLocalInteract
  (params (InteractingPawn :type Fort%20Player%20Pawn%20Object%20Reference))
  (bind _returnvalue (Fortnite|GetFortPlayerControllerfromActor InteractingPawn))
  (Utilities|FlowControl|Sequence
    (:then_0
      (if (not (Variables|Default|GetbLoadingAssets))
        (Utilities|IsValid _returnvalue
          (:Is%20Valid
            (Variables|Default|GetCachedPropertyEditorWidget
              (:then)
              (:else
                (CallFunction|DisplayPropertyEditor _returnvalue)))))))
    (:then_1
      (return true))))
