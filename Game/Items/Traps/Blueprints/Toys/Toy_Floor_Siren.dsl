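; Creative timer device with a floor pad that launches players and toggles the timer, starting it or completing it.
; Receiver events also start, pause, reset, complete, enable or disable it, or record minigame lap times.
; Completion awards score, broadcasts an event and plays siren and color effects; expiry completes or fails, optionally looping.
; A replicated vector packs server start time or paused sentinel with elapsed seconds, letting machines phase-align local one-second display ticks.
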(event Custom|StartWhenReceivingFrom_Event
  (params (Instigator :type Controller%20Object%20Reference))
  ;[0.6648 0 1] Trigger and Receiver Logic
  (Utilities|CanTrigger Instigator (Math|Boolean|ANDBoolean (CallFunction|IsPaused) (Variables|Default|GetbEnabled) (not (Variables|TimerValues|GetTimer%20Completed)))
    (:True
      (TimerActions|SetTimerState false))))

(event Custom|PauseWhenReceivingFrom_Event
  (params (Instigator :type Controller%20Object%20Reference))
  ;[0.6648 0 1] Trigger and Receiver Logic
  (Utilities|CanTrigger Instigator (and (not (CallFunction|IsPaused)) (Variables|Default|GetbEnabled))
    (:True
      (TimerActions|SetTimerState true))))

(event EventOnPlaced
  ;[0 1 0.1154] Initialization
  (Audio|PlaySoundatLocation (Variables|Audio|GetSound_Placed) (Transformation|GetWorldLocation (Variables|Debug|GetStaticMeshComponent))))

(event Custom|ResetWhenReceivingFrom_Event
  (params (Instigator :type Controller%20Object%20Reference))
  ;[0.6648 0 1] Trigger and Receiver Logic
  (Utilities|CanTrigger Instigator (Variables|Default|GetbEnabled)
    (:True
      (Variables|TimerValues|SetTimer%20Completed false)
      (CallFunction|ResetDevice (Variables|UserOptions|GetAutoStartSetting)))))

(event EventBeginPlay
  ;[0 1 0.1154] Initialization
  (CallFunction|ResetDevice false)
  (CallFunction|SetPadVisibility)
  (VisualFX|UpdateDisplayTimer false))

(event OnPropertiesLoaded(ToyOptionsComponent)
  ;[0 1 0.1154] Initialization
  (CallFunction|ResetDevice false)
  (CallFunction|SetPadVisibility)
  (VisualFX|UpdateDisplayTimer false))

(event Custom|EnableFunction
  (params (Instigator :type Controller%20Object%20Reference))
  ;[0.6648 0 1] Trigger and Receiver Logic
  (Utilities|CanTrigger Instigator (not (Variables|Default|GetbEnabled))
    (:True
      (Variables|Default|SetbEnabled true))))

(event OnAnyPropertyChanged(ToyOptionsComponent)
  ;[0 1 0.1154] Initialization
  (CallFunction|ResetDevice false)
  (CallFunction|SetPadVisibility)
  (VisualFX|UpdateDisplayTimer false))

(event Custom|DisableFunction
  (params (Instigator :type Controller%20Object%20Reference))
  ;[0.6648 0 1] Trigger and Receiver Logic
  (Utilities|CanTrigger Instigator (Variables|Default|GetbEnabled)
    (:True
      (Variables|Default|SetbEnabled false))))

(event Custom|CompleteWhenReceivingFrom_Event
  (params (Instigator :type Controller%20Object%20Reference))
  ;[0.6648 0 1] Trigger and Receiver Logic
  (Utilities|CanTrigger Instigator (Variables|Default|GetbEnabled)
    (:True
      (TimerActions|CompleteTimer))))

(event OnComponentBeginOverlap(TimerTrigger)
  (params (OverlappedComponent :type Primitive%20Component%20Object%20Reference) (OtherActor :type Actor%20Object%20Reference) (OtherComp :type Primitive%20Component%20Object%20Reference) (OtherBodyIndex :type Integer) (bFromSweep :type Boolean) (SweepResult :type Hit%20Result%20Structure%20%28by%20ref%29))
  ;[1 1 1] Overlap Event for the Trigger.
  (bind _asfort_player_pawn (Utilities|Casting|CastToFortPlayerPawn OtherActor))
  (Utilities|CanTrigger (Pawn|GetController _asfort_player_pawn) (and (Variables|Default|GetbEnabled) (not (and (Variables|UserOptions|GetCompleteOnce) (Variables|TimerValues|GetTimer%20Completed))))
    (:True
      (CallFunction|LaunchPlayerFromPad) ; FORT-408252: Launch before we toggle the timer because if we complete and the player binds to reset we lose our instigator and then there's a client/server misalignment causing infinite launches on the client and none on the server
      (TimerActions|ToggleTimer)
      (Networking|CanExecuteCosmeticEvents
        (:True
          (Audio|PlaySoundatLocation (Variables|Audio|GetSound_Switch) (Transformation|GetWorldLocation (Variables|Debug|GetStaticMeshComponent))))))))

(event Custom|SetLapTimeWhenReceivingFrom_Event
  (params (Instigator :type Controller%20Object%20Reference))
  ;[0.6648 0 1] Trigger and Receiver Logic
  (Utilities|CanTrigger Instigator (Variables|Default|GetbEnabled)
    (:True
      (CallFunction|AttemptSetLapTime))))

(event Custom|Blend%20Display%20Brightness
  (params (On :type Boolean))
  ;[1 0.8902 0] Quickly Lerp between brightness values (nicer to look at than direct setting)
  (Networking|CanExecuteCosmeticEvents
    (:True
      (if On
        (bind (_direction _brightness) (Timeline|Brightness%20Blend Play)
          (:Update
            (Rendering|Material|SetScalarParameterValue (Variables|VisualFX|GetMID_Time) "Brightness" _brightness :!class /Script/Engine.MaterialInstanceDynamic)))
        (else
          (Timeline|Brightness%20Blend Reverse))))))

(event Custom|Tick%20Timer
  ;[0 0.432 1] Timer Tick (Every Second)
  (if (and (not (CallFunction|IsPaused)) (Variables|Default|GetbEnabled))
    (Math|Float|IncrementFloat (Variables|TimerValues|GetTime%20Elapsed)
      (:%20%20
        (Networking|SwitchHasAuthority
          (:Authority
            (if (<= (CallFunction|GetRemainingTime) 0.0)
              (if (Variables|UserOptions|GetCompleteOnEnd)
                (TimerActions|CompleteTimer)
                (else
                  (TimerActions|FailTimer)))
              (else
                _)))
          (:Remote
            (VisualFX|UpdateDisplayTimer true)))
        (bind _returnvalue (Utilities|Time|SetTimerbyEvent (delegate Custom|Tick%20Timer) 1.0))
        (Variables|Default|SetTimerHandle _returnvalue)))))

(event Custom|RPC_MULTICAST%20Play%20Complete%20Effects
  (params (Complete%20Time :type Float%20%28double-precision%29))
  ;[1 0.8902 0] Playing Effects on Complete
  (Networking|CanExecuteCosmeticEvents
    (:True
      (Audio|PlayAlarmAudio)
      (Variables|VisualFX|SetShowCompleteVisuals true) ; Makes sure the timer doesn't update\nwhile the complete effects are running.
      (VisualFX|SetMaterialTime Complete%20Time)
      (bind (_direction _colorlerp) (Timeline|ColorLerp PlayFromStart)
        (:Update
          (Rendering|Material|SetVectorParameterValue (Variables|VisualFX|GetMID_Time) "Color" (Math|Color|Lerp(LinearColor) (Variables|VisualFX|GetDefault%20Color) (Variables|VisualFX|GetComplete%20Color) _colorlerp) :!class /Script/Engine.MaterialInstanceDynamic))
        (:Finished
          (Variables|VisualFX|SetShowCompleteVisuals false)
          (VisualFX|UpdateDisplayTimer false))))))

(event MinigameStartedDelegate(FortMinigameProgress)
  ;[0 0.432 1] Minigame Progress
  (CallFunction|ResetDevice (Variables|UserOptions|GetAutoStartSetting)))

(event MinigameEndedDelegate(FortMinigameProgress)
  ;[0 0.432 1] Minigame Progress
  (CallFunction|ResetDevice false))

(fn AttemptSetLapTime
  (if (and (not (Creative|TeamOption|IsTeamType (Variables|UserOptions|GetLapTimeTeam) "None")) (Utilities|IsValid (Variables|Default|GetLast%20Instigator) :!class /Script/Engine.KismetSystemLibrary))
    (if (Creative|TeamOption|IsTeamType (Variables|UserOptions|GetLapTimeTeam) "ActivatingTeam")
      _
      (elif (Creative|TeamOption|IsTeamType (Variables|UserOptions|GetLapTimeTeam))
        (if (Math|Byte|Equal(Byte) (Team|GetActorTeam (Variables|Default|GetLast%20Instigator)) (Creative|TeamOption|ConverttoTeam (Variables|UserOptions|GetLapTimeTeam)))
          _)))
    (bind _asfort_player_controller (Utilities|Casting|CastToFortPlayerController (Variables|Default|GetLast%20Instigator)))
    (bind _currentminigame (Variables|/Script/FortniteGame.FortMinigameLogicComponent|GetCurrentMinigame (Variables|Default|GetFortMinigameProgress))
      (:then
        (Minigame|SetLapTimeStat (Minigame|Stats|GetMinigameStatsComponent _currentminigame) _asfort_player_controller (Utilities|Select :Index (Variables|UserOptions|GetLapTimeStyle) :NewEnumerator0 (- (CallFunction|GetRemainingTime) (CallFunction|GetTimerMsElapsed)) :NewEnumerator1 (+ (CallFunction|GetTimerMsElapsed) (Variables|TimerValues|GetTime%20Elapsed)))))
      (:else))))

(fn AwardScore
  (Networking|SwitchHasAuthority
    (:Authority
      (bind _last_instigator (Variables|Default|GetLast%20Instigator)
        (:then
          (bind _asfort_player_controller (Utilities|Casting|CastToFortPlayerController _last_instigator))
          (Utilities|IsValid _asfort_player_controller
            (:Is%20Valid
              (bind _returnvalue (CreativeDevices|GetMinigameStatsComp _asfort_player_controller "Athena"))
              (Utilities|IsValid _returnvalue
                (:Is%20Valid
                  (Class|FortMinigameStatsComponent|SendStatEventwithTags _returnvalue _asfort_player_controller "MinigameDynamicEvent" self (GameplayTags|MakeGameplayTagContainerfromTag "(TagName=\"Minigame.Stat.Score.Pinball\")") (Class|FortMinigameStatsComponent|GetSourceandContextTags _returnvalue _asfort_player_controller) (.OutContextTags (Class|FortMinigameStatsComponent|GetSourceandContextTags _returnvalue _asfort_player_controller)) (Variables|UserOptions|GetScoreValueSetting)))))))
        (:else)))))

(fn CompleteTimer
  ;[1 0 0] Complete
  (CallFunction|AttemptSetLapTime)
  (Variables|TimerValues|SetTimer%20Completed true)
  (CallFunction|RPCMULTICASTPlayCompleteEffects (Math|Float|SelectFloat (CallFunction|GetRemainingTime) (Variables|TimerValues|GetTime%20Elapsed) (Variables|UserOptions|GetKnob_CountDownDirection)))
  (CallFunction|AwardScore)
  (TimerActions|SetTimerState (or (not (Variables|UserOptions|GetbLoop)) (Variables|UserOptions|GetCompleteOnce)) (and (Variables|UserOptions|GetbLoop) (not (Variables|UserOptions|GetCompleteOnce))))
  (GameplayEvent|BroadcastEvent (Variables|UserOptions|Events|GetOn%20Complete) 0 (Utilities|Struct|MakeWhenCompleteTransmitonData (Variables|Default|GetLast%20Instigator))))

(fn FailTimer
  ;[1 0 0] Timer Failed
  (TimerActions|SetTimerState (or (not (Variables|UserOptions|GetbLoop)) (Variables|UserOptions|GetCompleteOnce)) (Variables|UserOptions|GetbLoop)))

(fn GetRemainingTime
  (return (Math|Conversions|ToFloat(Integer) (Math|Float|Round (Math|Float|Clamp(Float) (- (Variables|UserOptions|GetTimerValue) (Variables|TimerValues|GetTime%20Elapsed)) 0.0 (Variables|UserOptions|GetTimerValue))))))

(fn GetTimerMsElapsed
  (return (Math|Float|SelectFloat 0.0 (- 1.0 (Utilities|Time|GetTimerRemainingTimebyHandle (Variables|Default|GetTimerHandle))) (<= (Utilities|Time|GetTimerRemainingTimebyHandle (Variables|Default|GetTimerHandle)) 0.0))))

(fn isPaused
  (bind _paused_time_data_x (Math|Vector2D|BreakVector2D (Variables|GetPaused%20Time%20Data_X)))
  (return (<= _paused_time_data_x 0.0)))

(fn IsTargetableByLinearMover
  (return true))

(fn LaunchPlayerfromPad
  (bind _last_instigator (Variables|Default|GetLast%20Instigator))
  (bind _returnvalue (Pawn|GetControlledPawn _last_instigator))
  (Utilities|IsValid _returnvalue
    (:Is%20Valid
      (bind _asfort_player_pawn (Utilities|Casting|CastToFortPlayerPawn _returnvalue))
      (bind _output_get (LocalVariables|Setlocal%20Pawn _asfort_player_pawn))
      (if (not (Health|IsDBNO _output_get))
        (Pawn|Character|LaunchCharacterJump (LocalVariables|Getlocal%20Pawn) (* (Transformation|GetUpVector (Variables|Default|GetTimerTrigger)) (Variables|Default|GetImpulse%20Modifier)) true true true true)
        (Utilities|FlowControl|Sequence
          (:then_0
            (bind _asfort_player_controller (Utilities|Casting|CastToFortPlayerController (Variables|Default|GetLast%20Instigator)))
            (Game|Feedback|ClientPlayForceFeedback _asfort_player_controller "/Game/Weapons/ForceFeedback/FF_Generic_LaunchPawn.FF_Generic_LaunchPawn" "JumpPad" false false false))
          (:then_1
            (Networking|CanExecuteCosmeticEvents
              (:True
                (if (Transformation|DoesSocketExist (Variables|/Script/Engine.Character|GetMesh (LocalVariables|Getlocal%20Pawn)) "FX_foot_lSocket")
                  (Effects|SpawnEmitterAttached "/Game/Weapons/FORT_TRAPS/Floor_Player_Jump_Pad/Effects/P_PlayerJumpPadTrail.P_PlayerJumpPadTrail" (Variables|/Script/Engine.Character|GetMesh (LocalVariables|Getlocal%20Pawn)) "FX_foot_lSocket")
                  (if (Transformation|DoesSocketExist (Variables|/Script/Engine.Character|GetMesh (LocalVariables|Getlocal%20Pawn)) "FX_foot_rSocket")
                    (Effects|SpawnEmitterAttached "/Game/Weapons/FORT_TRAPS/Floor_Player_Jump_Pad/Effects/P_PlayerJumpPadTrail.P_PlayerJumpPadTrail" (Variables|/Script/Engine.Character|GetMesh (LocalVariables|Getlocal%20Pawn)) "FX_foot_rSocket")))))))))))

(fn OnRep_bEnabled
  (Networking|SwitchHasAuthority
    (:Authority
      (Collision|SetCollisionEnabled (Variables|Default|GetTimerTrigger) (select (and (Variables|UserOptions|GetbShowPad) (Variables|Default|GetbEnabled)) "QueryOnly" "NoCollision"))
      (TimerActions|SetTimerState (not (Variables|Default|GetbEnabled))))
    (:Remote
      (CallFunction|BlendDisplayBrightness (Variables|Default|GetbEnabled)))))

(fn OnRep_PausedTimeData
  (bind (_paused_time_data_x _paused_time_data_y) (Math|Vector2D|BreakVector2D (Variables|GetPaused%20Time%20Data_X)))
  (bind _returnvalue (Fortnite|Sync|Time|GetServerWorldTimeSeconds))
  (bind _returnvalue_2 (- _returnvalue _paused_time_data_x))
  (bind (_remainder _returnvalue_3) (Math|Float|Division(WholeandRemainder) _returnvalue_2 1.0))
  (bind (_paused_time_data_x_2 _paused_time_data_y_2) (Math|Vector2D|BreakVector2D (Variables|GetPaused%20Time%20Data_X)))
  (if (CallFunction|IsPaused)
    ;[1 0.8902 0] Immediately update timer with the elapsed time when it updates if paused..
    (Networking|CanExecuteCosmeticEvents
      (:True
        (Variables|TimerValues|SetTime%20Elapsed _paused_time_data_y_2)
        (VisualFX|UpdateDisplayTimer false)))
    (else
      (Utilities|FlowControl|Sequence
        (:then_0
          (Utilities|Time|ClearandInvalidateTimerbyHandle (Variables|Default|GetTimerHandle))
          (Variables|TimerValues|SetTime%20Elapsed (+ (Math|Conversions|ToFloat(Integer) _returnvalue_3) _paused_time_data_y))
          (LocalVariables|Setlocal%20Time%20to%20First%20Tick (- 1.0 _remainder))
          (VisualFX|UpdateDisplayTimer true)
          (if (>= (LocalVariables|Getlocal%20Time%20to%20First%20Tick) 0.0)
            (bind _returnvalue_4 (Utilities|Time|SetTimerbyEvent (EventDispatchers|CreateEvent :!event Tick%20Timer) (LocalVariables|Getlocal%20Time%20to%20First%20Tick)))
            (Variables|Default|SetTimerHandle _returnvalue_4)
            (else
              (CallFunction|TickTimer))))))))

(fn PlayAlarmAudio
  (bind _output_get (LocalVariables|Setlocal%20Siren (Utilities|Array|Get(acopy) (Variables|Audio|GetSound_Siren) (Variables|UserOptions|GetAlarmIndex))))
  (Utilities|IsValid _output_get
    (:Is%20Valid
      (Audio|PlaySoundatLocation (LocalVariables|Getlocal%20Siren) (Transformation|GetWorldLocation (Variables|Debug|GetStaticMeshComponent)) "0, 0, 0" (Utilities|Array|Get(acopy) (Variables|Audio|GetAlarmVolumeArray) (Variables|UserOptions|GetVolumeIndex)) 1.0 0.0 (Utilities|Array|Get(acopy) (Variables|Audio|GetAttenuation%20Settings) (Variables|UserOptions|GetVolumeIndex)) 0 self))))

(fn ResetDevice
  (params (Immediately%20Start :type Boolean))
  (Networking|SwitchHasAuthority
    (:Authority
      (Variables|Default|SetLast%20Instigator 0)
      (Variables|TimerValues|SetTimer%20Completed false)
      (Variables|Default|SetbEnabled true)
      (TimerActions|SetTimerState (not Immediately%20Start) true))))

(fn SetMaterialTime
  (params (Complete%20Seconds :type Float%20%28double-precision%29))
  (bind _returnvalue (Utilities|String|TimeSecondstoString Complete%20Seconds))
  (bind _returnvalue_2 (Utilities|String|LeftChop _returnvalue 3))
  (bind _timerdisplay (Variables|Default|GetTimerDisplay))
  (Rendering|Material|SetScalarParameterValueonMaterials _timerdisplay "TenMinute" (Utilities|String|StringToFloat (Utilities|String|LeftChop _returnvalue_2 4)))
  (Rendering|Material|SetScalarParameterValueonMaterials _timerdisplay "OneMinute" (Utilities|String|StringToFloat (Utilities|String|RightChop (Utilities|String|LeftChop _returnvalue_2 3) 1)))
  (Rendering|Material|SetScalarParameterValueonMaterials _timerdisplay "TenSecond" (Utilities|String|StringToFloat (Utilities|String|LeftChop (Utilities|String|RightChop _returnvalue_2 3) 1)))
  (Rendering|Material|SetScalarParameterValueonMaterials _timerdisplay "OneSecond" (Utilities|String|StringToFloat (Utilities|String|RightChop _returnvalue_2 4))))

(fn SetPadVisibility
  (bind _bshowpad (Variables|UserOptions|GetbShowPad))
  (Development|SetHiddeninGame (Variables|Default|GetTimerPad) (not (Variables|UserOptions|GetbShowPad)) false)
  (Collision|SetCollisionEnabled (Variables|Default|GetTimerPad) (select _bshowpad "QueryAndPhysics" "NoCollision"))
  (Collision|SetCollisionEnabled (Variables|Default|GetTimerTrigger) (select _bshowpad "QueryOnly" "NoCollision")))

(fn SetTimerState
  (params (Pause :type Boolean) (Reset :type Boolean))
  (bind _paused_time_data_x (Math|Vector2D|BreakVector2D (Variables|GetPaused%20Time%20Data_X)))
  (Networking|SwitchHasAuthority
    (:Authority
      (bind _output_get (Variables|TimerValues|SetTime%20Elapsed (select (not Reset) (Variables|TimerValues|GetTime%20Elapsed) 0.0)))
      (Variables|SetPaused%20Time%20Data_X (Math|Vector2D|MakeVector2D (Math|Float|SelectFloat (Math|Float|SelectFloat -1.0 0.0 (== _paused_time_data_x 0.0)) (Fortnite|Sync|Time|GetServerWorldTimeSeconds) Pause) _output_get)))))

(fn ShouldInvisiblePropsBeAllowed
  (return false))

(fn ToggleTimer
  (Networking|SwitchHasAuthority
    (:Authority
      (if (Variables|Default|GetbEnabled)
        (if (CallFunction|IsPaused)
          (TimerActions|SetTimerState false true)
          (else
            (TimerActions|CompleteTimer)))))))

(fn UpdateDisplayTimer
  (params (PlayAudio :type Boolean :default true))
  (if (not (Variables|VisualFX|GetShowCompleteVisuals))
    (Networking|CanExecuteCosmeticEvents
      (:True
        (Utilities|FlowControl|Sequence
          (:then_0
            (if PlayAudio
              (Audio|PlaySoundatLocation (Variables|Audio|GetSound_Tick) (Transformation|GetWorldLocation (Variables|Debug|GetStaticMeshComponent)))))
          (:then_1
            (VisualFX|SetMaterialTime (Math|Float|SelectFloat (Math|Float|SelectFloat (Variables|UserOptions|GetTimerValue) (CallFunction|GetRemainingTime) (<= (CallFunction|GetRemainingTime) 0.0)) (Variables|TimerValues|GetTime%20Elapsed) (Variables|UserOptions|GetKnob_CountDownDirection)))))))))

(fn ConstructionScript
  (|CallParentFunction)
  (bind _returnvalue (Rendering|Material|CreateDynamicMaterialInstance (Variables|Default|GetTimerDisplay) 0 "/Game/Environments/Sets/RaceTrack/Materials/M_Racetrack_StartLight_Timer.M_Racetrack_StartLight_Timer" "None" :!class /Script/Engine.PrimitiveComponent))
  (bind _output_get (Variables|VisualFX|SetMID_Time _returnvalue))
  (Rendering|Material|SetScalarParameterValue _output_get "Brightness" 1.0 :!class /Script/Engine.MaterialInstanceDynamic)
  (Rendering|Material|SetScalarParameterValue _output_get "LapTime" 1.0 :!class /Script/Engine.MaterialInstanceDynamic)
  (Rendering|Material|SetScalarParameterValue _output_get "Slider" 0.0 :!class /Script/Engine.MaterialInstanceDynamic)
  (Rendering|Material|SetVectorParameterValue _output_get "Color" (Variables|VisualFX|GetDefault%20Color) :!class /Script/Engine.MaterialInstanceDynamic)
  (TimerActions|SetTimerState true true))
